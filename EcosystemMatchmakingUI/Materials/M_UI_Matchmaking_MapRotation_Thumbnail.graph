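# Translucent UI material for a matchmaking map-rotation thumbnail that slides horizontally from TextureA to TextureB.
# OffsetToNext shifts both textures' fill-scaled UVs, with TextureB starting one width to the right so it follows TextureA in.
# An edge-pixel clip on TextureA's shifted UVs and alpha drives the blend between the two clamp-sampled images.
# A CoreUI rounded mask with a pixel-based radius shapes the corners.
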Material /EcosystemMatchmakingUI/Materials/M_UI_Matchmaking_MapRotation_Thumbnail
graph {
 properties {
  bAutomaticallySetUsageInEditor=false
  bUsedWithStaticMesh=false
  blendMode=BLEND_Translucent
  materialDomain=MD_UI
  usedShadingModels=Default Lit
 }
 nodes [
  n1 = AppendVector(A=ScalarParameter() {parameterName=OffsetToNext, sliderMax=1}, B=Constant())
  n2 = NamedRerouteDeclaration(None=MaterialFunctionCall() {materialFunction=&GetUserInterfaceUV}.Normalized UV) {name=Normalized UV, nodeColor=<rgba 1 0 0.7969 1>}
  n3 = TextureObjectParameter() {parameterName=TextureB}
  n4 = Subtract(
   A=MaterialFunctionCall(UV=@n2, TextureSize=TextureProperty(Tex=@n3)) {materialFunction=&MF_UI_UVScaling}.Scaled To Fill
   B=Constant2Vector() {r=1}
  )
  n5 = TextureObjectParameter() {parameterName=TextureA}
  n6 = Add(
   A=MaterialFunctionCall(UV=@n2, TextureSize=TextureProperty(Tex=@n5)) {materialFunction=&MF_UI_UVScaling}.Scaled To Fill
   B=@n1
  )
  n7 = TextureSample(UVs=@n6, Tex=@n5) {samplerSource=SSM_Clamp_WorldGroupSettings}
  LinearInterpolate(
   A=TextureSample(UVs=Add(A=@n1, B=@n4), Tex=@n3) {samplerSource=SSM_Clamp_WorldGroupSettings}.RGB
   B=@n7.RGB
   Alpha=MaterialFunctionCall!(UVs=@n6, Alpha=@n7.A) {materialFunction=/FortUILibrary/MaterialFunctions/Basics/MF_UI_ClipEdgePixel}.Result
  )
  MaterialFunctionCall(Corner Radius=Constant() {r=10}, Radius in Pixels=Constant() {r=1}) {materialFunction=&MF_UI_CoreUI_RoundedMask}
 ]
}
