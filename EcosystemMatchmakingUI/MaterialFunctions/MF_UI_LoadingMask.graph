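# Material function that builds an animated screen-space loading mask for matchmaking UI and also outputs the IsLoading flag.
# A delayed animation timer, eased with InOutSine, drives an offset added to viewport UV, moving a vertical gradient down the screen.
# Saturating the gradient and multiplying by the ceiling of its inverse leaves a band with a soft trailing fade and hard leading edge.
# IsLoading passes straight through from the matchmaking global parameter collection.
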MaterialFunction /EcosystemMatchmakingUI/MaterialFunctions/MF_UI_LoadingMask
graph {nodes [
  n1 = MaterialFunctionCall!(Animation Duration=Constant() {r=2}, Delay=Constant() {r=0.125}, Reset to Default) {materialFunction=/FortUILibrary/MaterialFunctions/Time/MF_UI_AnimationDelay}
  n2 = Add(
   A=ScreenPosition().ViewportUV
   B=LinearInterpolate(Alpha=MaterialFunctionCall("Value (0-1)"=@n1.Result) {materialFunction=&MF_UI_EaseCurves}.EaseInOutSine) {constA=1, constB=-1}
  )
  n3 = Saturate(None=ComponentMask(None=Multiply(A=@n2) {constB=2}) {g=true})
  n4 = Ceil(None=OneMinus(None=@n3))
  FunctionOutput(None=NamedRerouteDeclaration(None=Multiply(A=@n3, B=@n4)) {name=Loading Mask, nodeColor=<rgba 0 0 0 1>}) {outputName=Mask}
  FunctionOutput(None=CollectionParameter() {collection=&/MPC_UI_Matchmaking_GlobalVisualValues, parameterName=IsLoading}) {outputName=IsLoading, sortPriority=1}
 ]}
