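# Translucent UI material for a matchmaking thumbnail that crossfades between TextureA and TextureB via OffsetToNext, darkening TextureA during the transition.
# Both textures scale to fill, with IsTournamentTexture shifting the vertical pivot and lowering opacity.
# Vertical gradients darken the image, fade its bottom edge and limit a brightened rounded-mask outline to the upper half.
# MF_UI_LoadingMask swaps color and reduces opacity while loading.
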Material /EcosystemMatchmakingUI/Materials/M_UI_Matchmaking_Thumbnail
graph {
 properties {
  bAutomaticallySetUsageInEditor=false
  bUsedWithStaticMesh=false
  blendMode=BLEND_Translucent
  materialDomain=MD_UI
  usedShadingModels=Default Lit
 }
 nodes [
  n1 = NamedRerouteUsage()
  n2 = TextureObjectParameter() {parameterName=TextureB}
  n3 = TextureProperty(Tex=@n2)
  n4 = NamedRerouteUsage()
  n5 = MaterialFunctionCall() {materialFunction=&GetUserInterfaceUV}
  n6 = NamedRerouteDeclaration(None=@n5.Pixel Size) {name=Pixel Size, nodeColor=<rgba 0 0 0 1>}
  n7 = TextureObjectParameter() {parameterName=TextureA}
  n8 = TextureSample(
   UVs=MaterialFunctionCall(
    UV=NamedRerouteDeclaration(None=@n5.Normalized UV) {name=Normalized UV, nodeColor=<rgba 0 0 0 1>}
    TextureSize=TextureProperty(Tex=@n7)
    ScalePivotPoint=NamedRerouteUsage()
    ContainerSize=@n6
   ) {materialFunction=&MF_UI_UVScaling}.Scaled To Fill
   Tex=@n7
  ) {samplerSource=SSM_Clamp_WorldGroupSettings}
  n9 = Saturate(None=Multiply(A=@n8.RGB) {constB=0.001})
  n10 = Power(Base=NamedRerouteDeclaration(None=ScalarParameter() {parameterName=OffsetToNext, sliderMax=1}) {name=OffsetAnimValue, nodeColor=<rgba 0 0 0 1>}) {constExponent=0.25}
  n11 = LinearInterpolate(
   A=TextureSample(
    UVs=MaterialFunctionCall(UV=@n1, TextureSize=@n3, ScalePivotPoint=@n4, ContainerSize=@n6) {materialFunction=&MF_UI_UVScaling}.Scaled To Fill
    Tex=@n2
   ) {samplerSource=SSM_Clamp_WorldGroupSettings}.RGB
   B=Saturate(None=LinearInterpolate(A=@n8.RGB, B=@n9, Alpha=@n10))
   Alpha=OneMinus(None=NamedRerouteUsage())
  )
  n12 = ComponentMask(None=NamedRerouteUsage()) {g=true}
  n13 = Constant() {r=0.125}
  n14 = Constant() {r=1}
  n15 = OneMinus(None=Saturate(None=MaterialFunctionCall(Input=@n12, Input Low=@n13, Input High=@n14) {materialFunction=&RemapValue_0-1}.Result))
  n16 = LinearInterpolate(A=Multiply(A=@n11) {constB=0.007}, B=@n11, Alpha=Multiply(A=@n15, B=@n15))
  n17 = MaterialFunctionCall!(Base=@n16, Add=Constant() {r=2}) {materialFunction=/Engine/Functions/Engine_MaterialFunctions02/Math/MultiplyAdd}
  n18 = MaterialFunctionCall(
   Size=Constant() {r=1}
   Corner Radius=Constant4Vector() {constant=<rgba 16 0 16 0>}.RGBA
   Radius in Pixels=Constant() {r=1}
   Outline Thickness=Constant() {r=4}
  ) {materialFunction=&MF_UI_CoreUI_RoundedMask}
  n19 = OneMinus(None=@n18.Shape Mask)
  n20 = OneMinus(None=ComponentMask(None=NamedRerouteUsage()) {g=true})
  n21 = Multiply(
   A=Max(A=Ceil(None=@n19), B=@n18.Outline Mask)
   B=MaterialFunctionCall(Input=@n20, Input Low=Constant() {r=0.5}, Input High=Constant() {r=1}) {materialFunction=&RemapValue_0-1}.Result
  )
  n22 = MaterialFunctionCall() {materialFunction=/EcosystemMatchmakingUI/MaterialFunctions/MF_UI_LoadingMask}
  LinearInterpolate(
   A=LinearInterpolate(A=@n16, B=Saturate(None=@n17.Result), Alpha=Multiply(A=@n21) {constB=0.6})
   B=LinearInterpolate(Alpha=@n22.Mask) {constA=0.1}
   Alpha=@n22.IsLoading
  ) {constA=1}
  n23 = MaterialFunctionCall(Input=@n20, Input Low=Constant(), Input High=Constant() {r=0.25}) {materialFunction=&RemapValue_0-1}
  n24 = Multiply(A=@n18.Shape Mask, B=SmoothStep(Value=Saturate(None=@n23.Result)))
  Multiply(
   A=LinearInterpolate(Alpha=ScalarParameter() {parameterName=IsTournamentTexture}) {constA=1, constB=0.7}
   B=LinearInterpolate(A=@n24, B=Multiply(A=@n24) {constB=0.5}, Alpha=@n22.IsLoading) {constA=1}
  )
  n25 = ScalarParameter() {parameterName=IsTournamentTexture}
  NamedRerouteDeclaration(None=AppendVector(A=Constant() {r=0.5}, B=LinearInterpolate(Alpha=@n25) {constA=0.5, constB=0})) {name=ScalePivotPoint, nodeColor=<rgba 0 1 0.5313 1>}
 ]
}
