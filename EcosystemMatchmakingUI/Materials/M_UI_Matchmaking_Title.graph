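# Translucent UI material for the matchmaking title that dims it with a mask while loading.
# MF_UI_LoadingMask supplies Mask and IsLoading, with Mask blending alpha between a dimmed level and full.
# IsLoading blends from constant full alpha to that masked value, and the result is multiplied by vertex color alpha.
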Material /EcosystemMatchmakingUI/Materials/M_UI_Matchmaking_Title
graph {
 properties {
  bAutomaticallySetUsageInEditor=false
  bUsedWithStaticMesh=false
  blendMode=BLEND_Translucent
  materialDomain=MD_UI
  usedShadingModels=Default Lit
 }
 nodes [
  n1 = MaterialFunctionCall() {materialFunction=/EcosystemMatchmakingUI/MaterialFunctions/MF_UI_LoadingMask}
  Multiply(A=VertexColor().A, B=LinearInterpolate(B=LinearInterpolate(Alpha=@n1.Mask) {constA=0.25}, Alpha=@n1.IsLoading) {constA=1})
 ]
}
