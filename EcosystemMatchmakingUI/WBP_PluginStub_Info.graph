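# WBP_PluginStub_Info is a stub user widget whose root is a plugin injection slot for the SteelShark plugin's lobby widget.
# On PreConstruct it sets its own visibility to Collapsed.
# A Select on the SteelSharkDisplay feature flag, choosing between Collapsed and SelfHitTestInvisible, sits in the graph unconnected, so the flag does not affect visibility here.
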WidgetBlueprint /EcosystemMatchmakingUI/WBP_PluginStub_Info
graphs [EventGraph {nodes [
   Event PreConstruct()
   > SetVisibility(InVisibility=Collapsed)
   Select(
    Option 0=Collapsed
    Option 1=SelfHitTestInvisible
    Index=GetPluginFeatureFlag!(
     self:Fort Global UIContext Object Reference=Get FortGlobalUIContext() -> (ReturnValue:Fort Global UIContext Object Reference) {is=title K2Node_GetSubsystem}
     FlagName:Name=SteelSharkDisplay
    ) -> (ReturnValue:bool)
   )
  ]}]
variables []
widgets {
 info {
  inheritedWidgetCount=0
  parentClass=&UserWidget
  rootWidgetClass=/Game/UI/Plugin/FortPluginWidgetInjectionSlot.FortPluginWidgetInjectionSlot_C
  widgetCount=1
 }
 widgets [FortPluginWidgetInjectionSlot {
   widgetClassPath=/Game/UI/Plugin/FortPluginWidgetInjectionSlot
   bIsVariable=true
   widget=FortPluginWidgetInjectionSlot
   properties {pluginId=SteelShark, pluginWidgetId=SteelShark_Lobby}
  }]
}
