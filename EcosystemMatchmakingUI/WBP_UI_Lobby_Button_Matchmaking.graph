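# Lobby matchmaking button derived from WBP_UIKit_Button_Generic, with a PLAY default text, animated cancel state and text changes.
# PreConstruct styles background and outline material corners; presses pick a cancel, locked or unlocked sound.
# Animated float variables drive setters: Anim_Background fades the material's "Display Loud" for cancel toggles, Anim_TextSwap slides both text blocks.
# SetRegisterTextValuesThenAnimate caches previous and current texts, which sequencer events apply during the slide.
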WidgetBlueprint /EcosystemMatchmakingUI/WBP_UI_Lobby_Button_Matchmaking
animations [
 Anim_Background {
  binding_count=1
  length_seconds=0.2
  object=&self:Anim_Background
  sequence {
   bindings [WBP_UI_Lobby_Button_Matchmaking_C_5 {
     id=#1
     possessed_class=WBP_UI_Lobby_Button_Matchmaking_C
     tracks [{
       type=MovieSceneDoubleTrack
       display_name=Anim Value Background
       sections [{type=MovieSceneDoubleSection, end_frame=12, start_frame=0, channels [None {type=MovieSceneScriptingDoubleChannel, keys [[0,0],[12,1]]}]}]
      }]
    }]
   display_rate {numerator=60}
  }
 }
 Anim_TextSwap {
  binding_count=1
  length_seconds=0.2
  object=&self:Anim_TextSwap
  sequence {
   bindings [WBP_UI_Lobby_Button_Matchmaking_C_5 {
     id=#1
     possessed_class=WBP_UI_Lobby_Button_Matchmaking_C
     tracks [{
       type=MovieSceneDoubleTrack
       display_name=Anim Value Text
       sections [{type=MovieSceneDoubleSection, channels [None {type=MovieSceneScriptingDoubleChannel, keys [[0,0],[5,256],[5.038,-256],[12,0]]}]}]
      }]
    }]
   display_rate {numerator=60}
   tracks [{
     type=MovieSceneEventTrack
     sections [{
       type=MovieSceneEventTriggerSection
       channels [None {
         type=MovieSceneScriptingEventChannel
         keys [[0, "<Struct 'MovieSceneEvent' (0x466fff1e600) {}>"], [5, "<Struct 'MovieSceneEvent' (0x466fff1be00) {}>"]]
        }]
      }]
    }]
  }
 }
]
graphs [
 EventGraph {nodes [
   n1 = Event PreConstruct()
   > Parent: PreConstruct(IsDesignTime=@n1.IsDesignTime)
   > Set Clipping!(execute, Clipping:EWidgetClipping Enum=ClipToBounds, self:Widget Object Reference=Get Main() -> (Main:Grid Panel Object Reference)) -> (then, Output_Get:EWidgetClipping Enum=Inherit)
   > Set DefaultText(DefaultText:Text=Get Text!() -> (Text)) -> (Output_Get:Text)
   > n2 = Cast To WBP_UIKit_Block_BackgroundWithOutline!(execute, Object=Get Block_Background() -> (Block_Background:WBP UIKit Block Background Base Object Reference)) -> (then, CastFailed, AsWBP UIKit Block Background with Outline:WBP UIKit Block Background With Outline Object Reference) {is=title K2Node_DynamicCast}
   n3 = Get BackgroundDynamicMaterial!(self:WBP UIKit Block Background Base Object Reference=@n2) -> (BackgroundDynamicMaterial:Material Instance Dynamic Object Reference)
   n4 = Get OutlineDynamicMaterial!(self:WBP UIKit Block Background Base Object Reference=@n2) -> (OutlineDynamicMaterial:Material Instance Dynamic Object Reference)
   > SetScalarParameterValue(self=[@n3, @n4], ParameterName=Radius in Pixels, Value=1)
   > SetScalarParameterValue(self=@n3, ParameterName=Corner Radius, Value=16)
   > SetVectorParameterValue!(execute, self=@n4, ParameterName=Corner Radius, Value_R:float=30, Value_G:float=30, Value_B:float=30, Value_A:float=30) -> (then)
   Event Construct()
   > Parent: Construct()
   Event On Pressed() {is=title K2Node_Event}
   > n5 = Branch(Condition=Get IsCancel() -> (IsCancel:bool))
   > PlaySound2D(Sound=Get PressedCancelSound() -> (PressedCancelSound:Sound Base Object Reference))
   PlaySound2D(
    execute=@n5.else
    Sound=Select(
     Option 0:Sound Base Object Reference=Get PressedPlayUnlockedSound() -> (PressedPlayUnlockedSound:Sound Base Object Reference)
     Option 1:Sound Base Object Reference=Get PressedPlayLockedSound() -> (PressedPlayLockedSound:Sound Base Object Reference)
     Index=Get bLocked() -> (bLocked:bool)
    ) -> (ReturnValue:Sound Base Object Reference)
   )
   Event Destruct()
   > Parent: Destruct()
  ]}
 Sequencer Events {nodes [
   SetPreviousText() {is=title K2Node_CustomEvent}
   > SetText!(execute, self:WBP UIKit Button Generic Base Object Reference, Text:Text=Get PreviousText() -> (PreviousText:Text)) -> (then)
   > n1 = Get!(execute:Exec) -> (then:Exec, else:Exec, Block_SecondText:WBP UIKit Block Text Object Reference)
   > SetText(self:WBP UIKit Block Text Object Reference=@n1, Text=Get SecondPreviousText() -> (SecondPreviousText:Text))
   SetCurrentText() {is=title K2Node_CustomEvent}
   > SetText(self:WBP UIKit Button Generic Base Object Reference, Text=Get CurrentText() -> (CurrentText:Text))
   > n2 = Get()
   > SetText(self:WBP UIKit Block Text Object Reference=@n2, Text=Get SecondCurrentText() -> (SecondCurrentText:Text))
  ]}
 SetAnimValue_Background {nodes [
   n1 = SetAnimValue_Background!() -> (then, NewParam:double=~) {is=title K2Node_FunctionEntry}
   > SetScalarParameterValue(self=Get BackgroundDynamicMaterial(self=Get Block_Background()), ParameterName=Display Loud, Value=Lerp(A=1, B=0, Alpha=@n1))
  ]}
 SetAnimValue_Text {nodes [
   n1 = SetAnimValue_Text!() -> (then, NewParam:double=~) {is=title K2Node_FunctionEntry}
   > SetRenderTranslation!(execute, self=Get Block_Text() -> (Block_Text:WBP UIKit Block Text Object Reference), Translation_X:double=@n1, Translation_Y:double=0) -> (then)
   > n2 = Get()
   > SetRenderTranslation(self=@n2, Translation_X=@n1, Translation_Y=0)
  ]}
 SetIsCancelButtonAnim {nodes [
   n1 = SetIsCancelButtonAnim!() -> (then, IsCancel:bool=~) {is=title K2Node_FunctionEntry}
   > Branch(Condition="Not Equal (Boolean)"(A=@n1, B=Get IsCancel()) -> (ReturnValue=~))
   > n2 = Set IsCancel(IsCancel:bool=Get IsCancel()) -> (Output_Get:bool)
   > n3 = Branch(Condition=@n2)
   n4 = Get Anim_Background() -> (Anim_Background:Widget Animation Object Reference)
   > PlayAnimationForward(InAnimation=@n4)
   PlayAnimationReverse(execute=@n3.else, InAnimation=@n4)
  ]}
 SetRegisterTextValuesThenAnimate {nodes [
   n1 = SetRegisterTextValuesThenAnimate!() -> (then, TextValue:Text, SecondTextValue:Text) {is=title K2Node_FunctionEntry}
   > n2 = Branch(Condition=TextIsEmpty(InText=@n1))
   > SetText(self:WBP UIKit Button Generic Base Object Reference, Text=Get TextValue() -> (TextValue:Text))
   > n3 = Get()
   > SetText(self:WBP UIKit Block Text Object Reference=@n3, Text=Get SecondTextValue() -> (SecondTextValue:Text))
   n4 = Self-Reference()
   n5 = Branch(
    execute=@n2.else
    Condition="Equal, Case Insensitive (Text)"!(A:"Text (by ref)"=Get Text(None=@n4), B:"Text (by ref)"=Get DefaultText() -> (DefaultText:Text)) -> (ReturnValue:bool)
   )
   > SetText(self:WBP UIKit Button Generic Base Object Reference, Text=Get TextValue())
   > Set CurrentText(CurrentText:Text=Get TextValue()) -> (Output_Get:Text)
   > n6 = Get()
   > SetText(self:WBP UIKit Block Text Object Reference=@n6, Text=Get SecondTextValue())
   > Set SecondCurrentText(SecondCurrentText:Text=Get SecondTextValue()) -> (Output_Get:Text)
   Set PreviousText(execute=@n5.else, PreviousText:Text=Get CurrentText()) -> (Output_Get:Text)
   > Set CurrentText(CurrentText=Get TextValue())
   > Set SecondPreviousText(SecondPreviousText:Text=Get SecondCurrentText()) -> (Output_Get:Text)
   > Set SecondCurrentText(SecondCurrentText=Get SecondTextValue())
   > PlayAnimation(InAnimation=Get Anim_TextSwap() -> (Anim_TextSwap:Widget Animation Object Reference))
  ]}
]
variables [
 AnimValue_Background {type=number}
 AnimValue_Text {type=number}
 CurrentText {type=string, default=""}
 DefaultText {type=string, default=$PLAY}
 IsCancel {type=boolean}
 PressedCancelSound {type=SoundBase, default=/Game/Sounds/Fort_UI_Sounds/FrontEnd/SC_UI_MatchMaking_UnReady_Cue}
 PressedPlayLockedSound {type=SoundBase, default=/Game/Sounds/Fort_UI_Sounds/FrontEnd/DefaultSounds/PressedButtons/MSS_UI_Button_Pressed_Default}
 PressedPlayUnlockedSound {type=SoundBase, default=&MSS_UI_Button_Pressed_Default}
 PreviousText {type=string, default=""}
 SecondCurrentText {type=string, default=""}
 SecondPreviousText {type=string, default=""}
]
widgets {
 info {
  inheritedWidgetCount=0
  parentClass=/FortUILibrary/UIKit/Buttons/WBP_UIKit_Button_Generic.WBP_UIKit_Button_Generic_C
  rootWidgetClass=null
  widgetCount=0
 }
 widgets []
}
