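# Lobby mode-selection button widget, derived from WBP_UIKit_Button_Generic.
# PreConstruct overrides corner radii on the background and outline dynamic materials.
# A visibility event sets or clears the mode-selection icon material, then resets button state if disabled.
# AddThirdText builds and registers a left-justified "Squads - No Fill" text block with a separate visibility setter, but its EventGraph call has no triggering event.
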WidgetBlueprint /EcosystemMatchmakingUI/WBP_UI_Lobby_Button_ModeSelection
graphs [
 AddThirdText {nodes [
   AddThirdText() {is=title K2Node_FunctionEntry}
   > RemoveWidget!(execute, Widget:Widget Object Reference=Get Block_Third_Text() -> (Block_Third_Text:WBP UIKit Block Text Object Reference)) -> (then)
   > n1 = Create Widget!(
    execute:Exec
    Class:User Widget Class Reference=Get TextClassTertiary() -> (TextClassTertiary:WBP UIKit Block Text Class Reference)
    OwningPlayer:Player Controller Object Reference
   ) -> (then:Exec, ReturnValue:WBP UIKit Block Text Object Reference)
   > n2 = Set Block_Third_Text(Block_Third_Text:WBP UIKit Block Text Object Reference=@n1) -> (Output_Get:WBP UIKit Block Text Object Reference)
   > RegisterModularBlock!(execute, self:UIKit Modular Button Object Reference, Block:UIKit Block Object Reference=@n2) -> (then)
   > SetIsSelectable!(execute, self:WBP UIKit Block Base Object Reference=@n2, Is Selectable:bool=Get bSelectable() -> (bSelectable:bool)) -> (then)
   > AddChildToStackBox!(execute, self:Stack Box Object Reference=Get Main_Stack() -> (Main_Stack:Stack Box Object Reference), Content:Widget Object Reference=@n2) -> (then, ReturnValue:Stack Box Slot Object Reference)
   > SetTransformPolicy!(execute, self:WBP UIKit Block Text Object Reference=@n2, Transform Policy:ETextTransformPolicy Enum=None) -> (then)
   > SetJustification!(execute, self:WBP UIKit Block Text Object Reference=@n2, Justification:ETextJustify Enum=Left) -> (then)
   > SetText!(execute, self:WBP UIKit Block Text Object Reference=@n2, Text:Text="LOCTEXT(\"$Squads - No Fill\")") -> (then)
  ]}
 EventGraph {nodes [
   AddThirdText() {is=title K2Node_CallFunction}
   > SetFont!(
    execute
    self:WBP UIKit Block Text Object Reference=Get Block_Third_Text()
    Font:"Slate Font Info Structure (by ref)"=Get Font_SecondAndThirdText() -> (Font_SecondAndThirdText:Slate Font Info Structure)
   ) -> (then)
   > SetFont(self=Get Block_SecondText() -> (Block_SecondText:WBP UIKit Block Text Object Reference), Font=Get Font_SecondAndThirdText())
   n1 = Set ModeSelect Icon Visibility!() -> (OutputDelegate, then, Visible:bool=~) {is=title K2Node_CustomEvent}
   > SetIconMaterial!(
    execute
    self:WBP UIKit Button Generic Base Object Reference
    Material Instance:Material Instance Object Reference=Select(
     Option 0:Material Instance Object Reference
     Option 1:Material Instance Object Reference=/EcosystemMatchmakingUI/MaterialInstances/MI_ModeSelection_Icon
     Index=@n1.Visible
    ) -> (ReturnValue:Material Instance Object Reference)
   ) -> (then)
   > UpdateIcon(self:WBP UIKit Button Generic Base Object Reference)
   > n2 = Branch(Condition=GetIsEnabled!(self:Widget Object Reference) -> (ReturnValue:bool))
   ResetButtonState(execute=@n2.else, self:WBP UIKit Button Generic Base Object Reference)
   n3 = Event PreConstruct()
   > Parent: PreConstruct(IsDesignTime=@n3.IsDesignTime)
   > SetOverrideCornerRadius()
   Event Construct()
   > Parent: Construct()
   Event Destruct()
   > Parent: Destruct()
  ]}
 SetOverrideCornerRadius {nodes [
   SetOverrideCornerRadius() {is=title K2Node_FunctionEntry}
   > n1 = Cast To WBP_UIKit_Block_BackgroundWithOutline!(execute, Object=Get Block_Background() -> (Block_Background:WBP UIKit Block Background Base Object Reference)) -> (then, CastFailed, AsWBP UIKit Block Background with Outline:WBP UIKit Block Background With Outline Object Reference) {is=title K2Node_DynamicCast}
   n2 = Get BackgroundDynamicMaterial!(self:WBP UIKit Block Background Base Object Reference=@n1) -> (BackgroundDynamicMaterial:Material Instance Dynamic Object Reference)
   n3 = Get OutlineDynamicMaterial!(self:WBP UIKit Block Background Base Object Reference=@n1) -> (OutlineDynamicMaterial:Material Instance Dynamic Object Reference)
   > SetScalarParameterValue(self=[@n2, @n3], ParameterName=Radius in Pixels, Value=1)
   > SetScalarParameterValue(self=@n2, ParameterName=Corner Radius, Value=20)
   > SetVectorParameterValue!(execute, self=@n3, ParameterName=Corner Radius, Value_R:float=30, Value_G:float=30, Value_B:float=30, Value_A:float=30) -> (then)
  ]}
 SetThirdTextVisibility {nodes [
   n1 = SetThirdTextVisibility!() -> (then, ThirdTextVisibility:ESlateVisibility Enum=Visible) {is=title K2Node_FunctionEntry}
   > SetVisibility(self=Get Block_Third_Text(), InVisibility=@n1)
   Set IconClass(IconClass:WBP UIKit Block Image Class Reference) -> (Output_Get:WBP UIKit Block Image Class Reference)
  ]}
]
variables [
 Block_Third_Text {type=WBP_UIKit_Block_Text_C, default=null}
 Font_SecondAndThirdText {
  type=SlateFontInfo
  default {
   bForceMonospaced=false
   bMaterialIsStencil=false
   fontMaterial=/FortUILibrary/UIKit/Buttons/MaterialInstances/CTAButtons/MI_UIKit_TextQuiet
   fontObject=&HeadingNow
   letterSpacing=10
   monospacedWidth=1
   size=15.75
   skewAmount=0
   typefaceFontName=75 Medium
   outlineSettings {
    bApplyOutlineToDropShadows=false
    bMiteredCorners=false
    bSeparateFillAlpha=false
    outlineMaterial=null
    outlineSize=0
    outlineColor=<rgba 0 0 0 1>
   }
  }
 }
 TextClassTertiary {type=Class, default=/FortUILibrary/Systems/ModularButton/Blocks/WBP_UIKit_Block_Text.WBP_UIKit_Block_Text_C}
]
widgets {
 info {
  inheritedWidgetCount=0
  parentClass=/FortUILibrary/UIKit/Buttons/WBP_UIKit_Button_Generic.WBP_UIKit_Button_Generic_C
  rootWidgetClass=null
  widgetCount=0
 }
 widgets []
}
