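# WBP_UI_Lobby_MatchFound_Base is a UserWidget Blueprint base for the lobby match-found animation in the matchmaking UI.
# It exposes an empty Play MatchFound Animation custom event, with no logic attached.
# It holds a MatchFoundColor variable defaulting to blue and a MatchFoundText string defaulting to "Game Found!", and has no widget tree of its own.
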WidgetBlueprint /Game/UI/DiscoveryShared/Matchmaking/MatchFoundAnimation/WBP_UI_Lobby_MatchFound_Base
graphs [EventGraph {nodes [Play MatchFound Animation() {is=title K2Node_CustomEvent}]}]
variables [
 MatchFoundColor {type=LinearColor, default=<rgba 0.0566 0 0.8906 1>}
 MatchFoundText {type=string, default="Game Found!"}
]
widgets {info {inheritedWidgetCount=0, parentClass=&UserWidget, rootWidgetClass=null, widgetCount=0}, widgets []}
