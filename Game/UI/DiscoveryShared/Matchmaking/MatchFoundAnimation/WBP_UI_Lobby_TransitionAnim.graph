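# Lobby UserWidget showing the animated "Match Found" transition for matchmaking.
# Its 4-second Anim_MatchFound builds up a material-parameter-driven starburst, then flashes blue glows, a horizon line, rift wave and particles.
# "Match Found" text pops in with a horizontal shake as a styled outline copy expands and fades.
# Construct hides the root overlay; the animation restores its opacity, but nothing in this graph plays it.
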WidgetBlueprint /Game/UI/DiscoveryShared/Matchmaking/MatchFoundAnimation/WBP_UI_Lobby_TransitionAnim
animations [Anim_MatchFound {
  binding_count=13
  length_seconds=4
  object=&self:Anim_MatchFound
  sequence {bindings [
    Text_Outline {
     id=#1
     possessed_class=CommonTextBlock
     tracks [
      {
       type=MovieScene2DTransformTrack
       sections [{
         type=MovieScene2DTransformSection
         end_frame=41
         start_frame=0
         channels [
          Translation.X {type=MovieSceneScriptingFloatChannel}
          Translation.Y {type=MovieSceneScriptingFloatChannel}
          Angle {type=MovieSceneScriptingFloatChannel}
          Scale.X {type=MovieSceneScriptingFloatChannel, keys [[0,0.1],[12.8333,0.1],[14.053,0.6],[31.6667,1]]}
          Scale.Y {type=MovieSceneScriptingFloatChannel, keys [[0,0.1],[12.8333,0.1],[13.053,0.3],[14.053,0.5],[31.6667,1]]}
          Shear.X {type=MovieSceneScriptingFloatChannel}
          Shear.Y {type=MovieSceneScriptingFloatChannel}
         ]
        }]
      }
      {
       type=MovieSceneFloatTrack
       display_name=Render Opacity
       sections [{
         type=MovieSceneFloatSection
         end_frame=31
         start_frame=0
         channels [None {type=MovieSceneScriptingFloatChannel, keys [[0,0],[12.8333,0],[15,1],[31.6667,0]]}]
        }]
      }
     ]
    }
    Image_Starburst {
     id=#2
     possessed_class=Image
     tracks [
      {
       type=MovieSceneWidgetMaterialTrack
       sections [{
         type=MovieSceneParameterSection
         end_frame=46
         start_frame=0
         channels [
          FalloffDensity {type=MovieSceneScriptingFloatChannel, keys [[0,0],[7.6667,2.9255]]}
          SpeedSwitch {type=MovieSceneScriptingFloatChannel, keys [[12.6667,4],[12.927,0],[13.166,-2.5]]}
         ]
        }]
      }
      {
       type=MovieSceneFloatTrack
       display_name=Render Opacity
       sections [{
         type=MovieSceneFloatSection
         end_frame=45
         start_frame=0
         channels [None {type=MovieSceneScriptingFloatChannel, thinned_from=6, keys [[0,1],[12.6667,1],[12.927,0],[13.166,1],[20,0]]}]
        }]
      }
      {
       type=MovieScene2DTransformTrack
       sections [{
         type=MovieScene2DTransformSection
         end_frame=45
         start_frame=0
         channels [
          Translation.X {type=MovieSceneScriptingFloatChannel, const [0,1]}
          Translation.Y {type=MovieSceneScriptingFloatChannel, const [0,1]}
          Angle {type=MovieSceneScriptingFloatChannel, const [0,1]}
          Scale.X {type=MovieSceneScriptingFloatChannel, keys [[0,1],[7.6667,2],[12.3333,0.7],[13.166,2.8],[13.1667,2.5]]}
          Scale.Y {type=MovieSceneScriptingFloatChannel, keys [[0,1],[7.6667,2],[12.3333,0.7],[13.166,2.8],[13.1667,2.5]]}
         ]
        }]
      }
     ]
    }
    Image_Glow_1_Horizon {
     id=#3
     possessed_class=Image
     tracks [
      {
       type=MovieSceneFloatTrack
       display_name=Render Opacity
       sections [{
         type=MovieSceneFloatSection
         end_frame=64
         start_frame=0
         channels [None {type=MovieSceneScriptingFloatChannel, keys [[0,0],[11.3263,0],[13.6667,0.75],[20.873,0.75],[64.075,0.5]]}]
        }]
      }
      {
       type=MovieScene2DTransformTrack
       sections [{
         type=MovieScene2DTransformSection
         end_frame=77
         start_frame=0
         channels [
          Translation.X {type=MovieSceneScriptingFloatChannel}
          Translation.Y {type=MovieSceneScriptingFloatChannel}
          Angle {type=MovieSceneScriptingFloatChannel}
          Scale.X {type=MovieSceneScriptingFloatChannel, thinned_from=5, keys [[0,8],[19.4167,8],[77.6667,6]]}
          Scale.Y {type=MovieSceneScriptingFloatChannel, keys [[0,0.5],[11.3263,0.5],[11.9733,1],[40.775,0.4618],[77.6667,0.4]]}
          Shear.X {type=MovieSceneScriptingFloatChannel}
          Shear.Y {type=MovieSceneScriptingFloatChannel}
         ]
        }]
      }
      {
       type=MovieSceneColorTrack
       display_name=Color and Opacity
       sections [{
         type=MovieSceneColorSection
         end_frame=30
         start_frame=11
         channels [
          Color.R {type=MovieSceneScriptingFloatChannel, keys [[13.5917,0.11],[17.7987,0.021],[30.743,0.009]]}
          Color.G {type=MovieSceneScriptingFloatChannel, const [0,3]}
          Color.B {type=MovieSceneScriptingFloatChannel, keys [[13.5917,1],[17.7987,0.6719],[30.743,0.9219]]}
          Color.A {type=MovieSceneScriptingFloatChannel, keys [[13.5917,1],[17.7987,1],[30.743,0.25]]}
         ]
        }]
      }
     ]
    }
    Img_Particles_M {
     id=#4
     possessed_class=Image
     tracks [{
       type=MovieSceneFloatTrack
       display_name=Render Opacity
       sections [{
         type=MovieSceneFloatSection
         end_frame=45
         start_frame=0
         channels [None {type=MovieSceneScriptingFloatChannel, keys [[0,0],[12.3333,0],[18.3333,2]]}]
        }]
      }]
    }
    Image_Glow_2_Circle {
     id=#5
     possessed_class=Image
     tracks [
      {
       type=MovieSceneFloatTrack
       display_name=Render Opacity
       sections [{
         type=MovieSceneFloatSection
         end_frame=60
         start_frame=0
         channels [None {type=MovieSceneScriptingFloatChannel, keys [[0,0],[11.3333,0],[12,0.75],[17,0.75],[28,0.5],[60,0]]}]
        }]
      }
      {
       type=MovieScene2DTransformTrack
       sections [{
         type=MovieScene2DTransformSection
         end_frame=37
         start_frame=0
         channels [
          Translation.X {type=MovieSceneScriptingFloatChannel}
          Translation.Y {type=MovieSceneScriptingFloatChannel}
          Angle {type=MovieSceneScriptingFloatChannel}
          Scale.X {type=MovieSceneScriptingFloatChannel, thinned_from=5, keys [[0,1],[12.333,1],[14.3333,3],[37.3333,3]]}
          Scale.Y {type=MovieSceneScriptingFloatChannel, thinned_from=5, keys [[0,4],[12.333,4],[14.3333,5],[37.3333,3]]}
          Shear.X {type=MovieSceneScriptingFloatChannel, keys [[0,-32],[11.6667,-32],[37.3333,0]]}
          Shear.Y {type=MovieSceneScriptingFloatChannel}
         ]
        }]
      }
      {
       type=MovieSceneColorTrack
       display_name=Color and Opacity
       sections [{
         type=MovieSceneColorSection
         end_frame=37
         start_frame=11
         channels [
          Color.R {type=MovieSceneScriptingFloatChannel, keys [[13.6667,0.0316],[20,0.0177],[37,0]]}
          Color.G {type=MovieSceneScriptingFloatChannel, keys [[13.6667,0.0608],[20,0],[37,0.1037]]}
          Color.B {type=MovieSceneScriptingFloatChannel, const [1,3]}
          Color.A {type=MovieSceneScriptingFloatChannel, const [1,3]}
         ]
        }]
      }
     ]
    }
    Text {
     id=#6
     possessed_class=CommonTextBlock
     tracks [
      {
       type=MovieSceneFloatTrack
       display_name=Render Opacity
       sections [{
         type=MovieSceneFloatSection
         end_frame=80
         start_frame=0
         channels [None {type=MovieSceneScriptingFloatChannel, keys [[0,0],[11.955,0],[12.4367,1],[80.1217,1]]}]
        }]
      }
      {
       type=MovieScene2DTransformTrack
       sections [{
         type=MovieScene2DTransformSection
         end_frame=46
         start_frame=0
         channels [
          Translation.X {type=MovieSceneScriptingFloatChannel, keys [[0,0],[13.3333,0],[14,-24],[14.6667,23.6881],[15.3333,1]]}
          Angle {type=MovieSceneScriptingFloatChannel}
          Scale.X {type=MovieSceneScriptingFloatChannel, keys [[8.6667,0.6],[41.3333,1]]}
          Scale.Y {type=MovieSceneScriptingFloatChannel, keys [[8.6667,0.6],[41.3333,1]]}
          Shear.X {type=MovieSceneScriptingFloatChannel}
          Shear.Y {type=MovieSceneScriptingFloatChannel}
         ]
        }]
      }
      {
       type=MovieSceneFloatTrack
       display_name=Line Height Percentage
       sections [{
         type=MovieSceneFloatSection
         end_frame=59
         start_frame=0
         channels [None {type=MovieSceneScriptingFloatChannel, keys [[0,0.2738],[12.8533,0.2738],[53.05,0.6885]]}]
        }]
      }
     ]
    }
    Img_Particles_XS {
     id=#7
     possessed_class=Image
     tracks [{
       type=MovieSceneFloatTrack
       display_name=Render Opacity
       sections [{
         type=MovieSceneFloatSection
         end_frame=64
         start_frame=0
         channels [None {type=MovieSceneScriptingFloatChannel, keys [[12.3333,0],[18.3333,3]]}]
        }]
      }]
    }
    Image_Glow_3_BigBG {
     id=#8
     possessed_class=Image
     tracks [{
       type=MovieSceneFloatTrack
       display_name=Render Opacity
       sections [{
         type=MovieSceneFloatSection
         end_frame=63
         start_frame=0
         channels [None {type=MovieSceneScriptingFloatChannel, keys [[0,0],[11.6667,0],[12.333,0.25],[60,0.025]]}]
        }]
      }]
    }
    Image_PulsingBG {
     id=#9
     possessed_class=Image
     tracks [{
       type=MovieSceneFloatTrack
       display_name=Render Opacity
       sections [{
         type=MovieSceneFloatSection
         end_frame=18
         start_frame=0
         channels [None {type=MovieSceneScriptingFloatChannel, keys [[0,0],[12,0],[13.923,0.1],[18.3333,0.05]]}]
        }]
      }]
    }
    Image_HorizoneFineLine {
     id=#10
     possessed_class=Image
     tracks [
      {
       type=MovieSceneFloatTrack
       display_name=Render Opacity
       sections [{
         type=MovieSceneFloatSection
         channels [None {type=MovieSceneScriptingFloatChannel, keys [[0.3333,0],[11.6667,0],[14,2],[18,0.8109],[34.6667,0]]}]
        }]
      }
      {
       type=MovieScene2DTransformTrack
       sections [{
         type=MovieScene2DTransformSection
         channels [
          Translation.X {type=MovieSceneScriptingFloatChannel}
          Translation.Y {type=MovieSceneScriptingFloatChannel}
          Angle {type=MovieSceneScriptingFloatChannel}
          Scale.X {type=MovieSceneScriptingFloatChannel, keys [[0,1],[11.6667,1],[12.6667,2],[19.6667,2.2]]}
          Scale.Y {type=MovieSceneScriptingFloatChannel, keys [[0,1],[11.6667,1],[12.6667,0.05],[19.6667,0.01]]}
          Shear.X {type=MovieSceneScriptingFloatChannel}
          Shear.Y {type=MovieSceneScriptingFloatChannel}
         ]
        }]
      }
      {
       type=MovieSceneColorTrack
       display_name=Color and Opacity
       sections [{
         type=MovieSceneColorSection
         end_frame=26
         start_frame=0
         channels [
          Color.R {type=MovieSceneScriptingFloatChannel, keys [[0,1],[18,1],[26.6667,0]]}
          Color.G {type=MovieSceneScriptingFloatChannel, keys [[0,1],[18,1],[26.6667,0.2999]]}
          Color.B {type=MovieSceneScriptingFloatChannel, const [1,3]}
          Color.A {type=MovieSceneScriptingFloatChannel, const [1,3]}
         ]
        }]
      }
     ]
    }
    Image_PulsingFG {
     id=#11
     possessed_class=Image
     tracks [
      {
       type=MovieSceneFloatTrack
       display_name=Render Opacity
       sections [{
         type=MovieSceneFloatSection
         end_frame=25
         start_frame=0
         channels [None {type=MovieSceneScriptingFloatChannel, keys [[0,0],[11.3333,0],[13.94,0.75],[20.679,0.025]]}]
        }]
      }
      {
       type=MovieSceneColorTrack
       display_name=Color and Opacity
       sections [{
         type=MovieSceneColorSection
         end_frame=22
         start_frame=8
         channels [
          Color.R {type=MovieSceneScriptingFloatChannel, keys [[11.4097,0.9216],[16.3497,0],[20.702,0]]}
          Color.G {type=MovieSceneScriptingFloatChannel, keys [[11.1743,0.9216],[16.3497,0.1645],[20.702,0.0207]]}
          Color.B {type=MovieSceneScriptingFloatChannel, keys [[11.4097,0.9216],[16.3497,1],[20.702,1]]}
          Color.A {type=MovieSceneScriptingFloatChannel, const [1,3]}
         ]
        }]
      }
      {
       type=MovieScene2DTransformTrack
       sections [{
         type=MovieScene2DTransformSection
         end_frame=19
         start_frame=11
         channels [
          Translation.X {type=MovieSceneScriptingFloatChannel}
          Translation.Y {type=MovieSceneScriptingFloatChannel}
          Angle {type=MovieSceneScriptingFloatChannel}
          Scale.X {type=MovieSceneScriptingFloatChannel, keys [[11.3333,2.45],[16.274,4]]}
          Scale.Y {type=MovieSceneScriptingFloatChannel, keys [[11.3333,1.7],[16.274,4]]}
          Shear.X {type=MovieSceneScriptingFloatChannel}
          Shear.Y {type=MovieSceneScriptingFloatChannel}
         ]
        }]
      }
     ]
    }
    Img_RiftWave {
     id=#12
     possessed_class=Image
     tracks [
      {
       type=MovieSceneFloatTrack
       display_name=Render Opacity
       sections [{
         type=MovieSceneFloatSection
         end_frame=78
         start_frame=0
         channels [None {type=MovieSceneScriptingFloatChannel, keys [[0,0],[12,0],[12.3333,0.005],[12.6667,1.005],[67.6667,0.25]]}]
        }]
      }
      {
       type=MovieScene2DTransformTrack
       sections [{
         type=MovieScene2DTransformSection
         end_frame=80
         start_frame=0
         channels [
          Translation.X {type=MovieSceneScriptingFloatChannel}
          Translation.Y {type=MovieSceneScriptingFloatChannel}
          Angle {type=MovieSceneScriptingFloatChannel}
          Scale.X {type=MovieSceneScriptingFloatChannel, keys [[0,0.001],[12.3333,0.001],[13.697,2.65],[67.6667,3],[80,3]]}
          Scale.Y {type=MovieSceneScriptingFloatChannel, keys [[0,0.001],[12.3333,0],[13.0717,3.6198],[61.5,1.9779],[80,2]]}
          Shear.X {type=MovieSceneScriptingFloatChannel}
          Shear.Y {type=MovieSceneScriptingFloatChannel}
         ]
        }]
      }
     ]
    }
    Default_MatchFound_Animation {
     id=#13
     possessed_class=Overlay
     tracks [{
       type=MovieSceneFloatTrack
       display_name=Render Opacity
       sections [{type=MovieSceneFloatSection, end_frame=2, start_frame=0, channels [None {type=MovieSceneScriptingFloatChannel, const [1,2]}]}]
      }]
    }
   ]}
 }]
graphs [EventGraph {nodes [
   Event Construct()
   > Set RenderOpacity!(
    execute
    RenderOpacity:float=0
    self:Widget Object Reference=Get Default_MatchFound_Animation() -> (Default_MatchFound_Animation:Overlay Object Reference)
   ) -> (then, Output_Get:float)
  ]}]
variables []
widgets {
 info {inheritedWidgetCount=0, parentClass=&UserWidget, rootWidgetClass=&Overlay, widgetCount=19}
 widgets [
  Default_MatchFound_Animation {widgetClassPath=&Overlay, bIsVariable=true, widget=Default_MatchFound_Animation}
  FortVisualAttachment_0 {
   widgetClassPath=&FortVisualAttachment
   parent=Default_MatchFound_Animation
   widget=FortVisualAttachment_0
   properties {visibility=HitTestInvisible, contentAnchor=<0.5 0.5>}
   slot {class=OverlaySlot, properties {horizontalAlignment=HAlign_Center, verticalAlignment=VAlign_Center}}
  }
  Border {
   widgetClassPath=&Border
   bIsVariable=true
   parent=FortVisualAttachment_0
   widget=Border
   properties {
    visibility=HitTestInvisible
    background {
     bIsDynamicallyLoaded=false
     drawAs=Image
     imageType=NoImage
     mirroring=NoMirror
     resourceName=None
     resourceObject=null
     tiling=NoTile
     imageSize=<32 32>
     margin=<ltrb 0 0 0 0>
     outlineSettings {
      bUseBrushTransparency=false
      roundingType=HalfHeightRadius
      width=0
      color {colorUseRule=UseColor_Specified, specifiedColor=<rgba 0 0 0 0>}
      cornerRadii=<0 0 0 0>
     }
     tintColor {colorUseRule=UseColor_Specified, specifiedColor=<rgba 1 1 1 0>}
     uVRegion {bIsValid=false, max=<0 0>, min=<0 0>}
    }
    padding=<ltrb 0 0 0 0>
   }
   slot {class=SizeBoxSlot}
  }
  Overlay_285 {
   widgetClassPath=&Overlay
   parent=Border
   widget=Overlay_285
   properties {visibility=HitTestInvisible, renderTransform {angle=0, scale=<0.75 0.75>, shear=<0 0>, translation=<0 0>}}
   slot {class=BorderSlot, properties {padding=<ltrb 0 0 0 0>}}
  }
  Overlay_RiftVFX {
   widgetClassPath=&Overlay
   parent=Overlay_285
   widget=Overlay_RiftVFX
   slot {class=OverlaySlot, properties {horizontalAlignment=HAlign_Center, verticalAlignment=VAlign_Center}}
  }
  Img_RiftWave {
   widgetClassPath=&Image
   bIsVariable=true
   parent=Overlay_RiftVFX
   widget=Img_RiftWave
   properties {
    visibility=SelfHitTestInvisible
    brush {
     bIsDynamicallyLoaded=false
     drawAs=Image
     imageType=FullColor
     mirroring=NoMirror
     resourceName=None
     resourceObject=/Game/UI/DiscoveryShared/Matchmaking/MatchFoundAnimation/MaterialInstances/MI_WhispyNoise_Full
     tiling=NoTile
     imageSize=<230 60>
     margin=<ltrb 0 0 0 0>
     outlineSettings {
      bUseBrushTransparency=false
      roundingType=HalfHeightRadius
      width=0
      color {colorUseRule=UseColor_Specified, specifiedColor=<rgba 0 0 0 0>}
      cornerRadii=<0 0 0 0>
     }
     tintColor {colorUseRule=UseColor_Specified, specifiedColor=<rgba 1 1 1 1>}
     uVRegion {bIsValid=false, max=<0 0>, min=<0 0>}
    }
    renderTransform {angle=0, scale=<3 2>, shear=<0 0>, translation=<0 0>}
   }
   slot {class=OverlaySlot, properties {horizontalAlignment=HAlign_Center, verticalAlignment=VAlign_Center}}
  }
  Img_Particles_M {
   widgetClassPath=&Image
   bIsVariable=true
   parent=Overlay_RiftVFX
   widget=Img_Particles_M
   properties {
    visibility=SelfHitTestInvisible
    brush {
     bIsDynamicallyLoaded=false
     drawAs=Image
     imageType=FullColor
     mirroring=NoMirror
     resourceName=None
     resourceObject=/Game/UI/DiscoveryShared/Matchmaking/MatchFoundAnimation/MaterialInstances/MI_RiftParticles_Medium
     tiling=NoTile
     imageSize=<230 200>
     margin=<ltrb 0 0 0 0>
     outlineSettings {
      bUseBrushTransparency=false
      roundingType=HalfHeightRadius
      width=0
      color {colorUseRule=UseColor_Specified, specifiedColor=<rgba 0 0 0 0>}
      cornerRadii=<0 0 0 0>
     }
     tintColor {colorUseRule=UseColor_Specified, specifiedColor=<rgba 1 1 1 1>}
     uVRegion {bIsValid=false, max=<0 0>, min=<0 0>}
    }
    renderTransform {angle=0, scale=<2 1>, shear=<0 0>, translation=<0 0>}
   }
   slot {class=OverlaySlot, properties {horizontalAlignment=HAlign_Center, verticalAlignment=VAlign_Center}}
  }
  Img_Particles_XS {
   widgetClassPath=&Image
   bIsVariable=true
   parent=Overlay_RiftVFX
   widget=Img_Particles_XS
   properties {
    visibility=SelfHitTestInvisible
    brush {
     bIsDynamicallyLoaded=false
     drawAs=Image
     imageType=FullColor
     mirroring=NoMirror
     resourceName=None
     resourceObject=/Game/UI/DiscoveryShared/Matchmaking/MatchFoundAnimation/MaterialInstances/MI_RiftParticles_Tiny
     tiling=NoTile
     imageSize=<230 100>
     margin=<ltrb 0 0 0 0>
     outlineSettings {
      bUseBrushTransparency=false
      roundingType=HalfHeightRadius
      width=0
      color {colorUseRule=UseColor_Specified, specifiedColor=<rgba 0 0 0 0>}
      cornerRadii=<0 0 0 0>
     }
     tintColor {colorUseRule=UseColor_Specified, specifiedColor=<rgba 1 1 1 1>}
     uVRegion {bIsValid=false, max=<0 0>, min=<0 0>}
    }
    renderTransform {angle=0, scale=<2 1>, shear=<0 0>, translation=<0 0>}
   }
   slot {class=OverlaySlot, properties {horizontalAlignment=HAlign_Center, verticalAlignment=VAlign_Center}}
  }
  Overlay_Heavy_GlowFX {
   widgetClassPath=&Overlay
   parent=Overlay_285
   widget=Overlay_Heavy_GlowFX
   slot {class=OverlaySlot, properties {horizontalAlignment=HAlign_Center, verticalAlignment=VAlign_Center}}
  }
  Image_Glow_3_BigBG {
   widgetClassPath=&Image
   parent=Overlay_Heavy_GlowFX
   widget=Image_Glow_3_BigBG
   properties {
    renderOpacity=0
    visibility=SelfHitTestInvisible
    brush {
     bIsDynamicallyLoaded=false
     drawAs=Image
     imageType=FullColor
     mirroring=NoMirror
     resourceName=None
     resourceObject=/Game/UI/DiscoveryShared/Matchmaking/MatchFoundAnimation/MaterialInstances/MI_UI_Glow_Matchmaking
     tiling=NoTile
     imageSize=<256 256>
     margin=<ltrb 0 0 0 0>
     outlineSettings {
      bUseBrushTransparency=false
      roundingType=HalfHeightRadius
      width=0
      color {colorUseRule=UseColor_Specified, specifiedColor=<rgba 0 0 0 0>}
      cornerRadii=<0 0 0 0>
     }
     tintColor {colorUseRule=UseColor_Specified, specifiedColor=<rgba 0.045 0.0684 0.717 0.5>}
     uVRegion {bIsValid=false, max=<0 0>, min=<0 0>}
    }
    colorAndOpacity=<rgba 1 1 1 0.1>
    renderTransform {angle=0, scale=<8 2>, shear=<0 0>, translation=<0 0>}
   }
   slot {class=OverlaySlot, properties {horizontalAlignment=HAlign_Center, verticalAlignment=VAlign_Center}}
  }
  Image_Glow_1_Horizon {
   widgetClassPath=&Image
   parent=Overlay_Heavy_GlowFX
   widget=Image_Glow_1_Horizon
   properties {
    renderOpacity=0
    visibility=SelfHitTestInvisible
    brush {
     bIsDynamicallyLoaded=false
     drawAs=Image
     imageType=FullColor
     mirroring=NoMirror
     resourceName=None
     resourceObject=&MI_UI_Glow_Matchmaking
     tiling=NoTile
     imageSize=<256 256>
     margin=<ltrb 0 0 0 0>
     outlineSettings {
      bUseBrushTransparency=false
      roundingType=HalfHeightRadius
      width=0
      color {colorUseRule=UseColor_Specified, specifiedColor=<rgba 0 0 0 0>}
      cornerRadii=<0 0 0 0>
     }
     tintColor {colorUseRule=UseColor_Specified, specifiedColor=<rgba 0.7135 0.7135 0.7135 1>}
     uVRegion {bIsValid=false, max=<0 0>, min=<0 0>}
    }
    colorAndOpacity=<rgba 0.009 0 0.9219 0.25>
    renderTransform {angle=0, scale=<8 0.5>, shear=<0 0>, translation=<0 0>}
   }
   slot {class=OverlaySlot, properties {horizontalAlignment=HAlign_Center, verticalAlignment=VAlign_Center}}
  }
  Image_Glow_2_Circle {
   widgetClassPath=&Image
   parent=Overlay_Heavy_GlowFX
   widget=Image_Glow_2_Circle
   properties {
    visibility=SelfHitTestInvisible
    brush {
     bIsDynamicallyLoaded=false
     drawAs=Image
     imageType=FullColor
     mirroring=NoMirror
     resourceName=None
     resourceObject=&MI_UI_Glow_Matchmaking
     tiling=NoTile
     imageSize=<256 256>
     margin=<ltrb 0 0 0 0>
     outlineSettings {
      bUseBrushTransparency=false
      roundingType=HalfHeightRadius
      width=0
      color {colorUseRule=UseColor_Specified, specifiedColor=<rgba 0 0 0 0>}
      cornerRadii=<0 0 0 0>
     }
     tintColor {colorUseRule=UseColor_Specified, specifiedColor=<rgba 1 1 1 1>}
     uVRegion {bIsValid=false, max=<0 0>, min=<0 0>}
    }
    colorAndOpacity=<rgba 0.0316 0.0608 1 1>
    renderTransform {angle=0, scale=<3 4.8013>, shear=<-13 0>, translation=<0 0>}
   }
   slot {class=OverlaySlot, properties {horizontalAlignment=HAlign_Center, verticalAlignment=VAlign_Center}}
  }
  Image_Starburst {
   widgetClassPath=&Image
   bIsVariable=true
   parent=Overlay_285
   widget=Image_Starburst
   properties {
    visibility=SelfHitTestInvisible
    brush {
     bIsDynamicallyLoaded=false
     drawAs=Image
     imageType=FullColor
     mirroring=NoMirror
     resourceName=None
     resourceObject=/Game/UI/DiscoveryShared/Matchmaking/MatchFoundAnimation/MaterialInstances/M_UI_Starburst_MatchFoundAnim
     tiling=NoTile
     imageSize=<200 200>
     margin=<ltrb 0 0 0 0>
     outlineSettings {
      bUseBrushTransparency=false
      roundingType=HalfHeightRadius
      width=0
      color {colorUseRule=UseColor_Specified, specifiedColor=<rgba 0 0 0 0>}
      cornerRadii=<0 0 0 0>
     }
     tintColor {colorUseRule=UseColor_Specified, specifiedColor=<rgba 1 1 1 1>}
     uVRegion {bIsValid=false, max=<0 0>, min=<0 0>}
    }
   }
   slot {class=OverlaySlot, properties {horizontalAlignment=HAlign_Center, verticalAlignment=VAlign_Center}}
  }
  Image_PulsingBG {
   widgetClassPath=&Image
   bIsVariable=true
   parent=Overlay_285
   widget=Image_PulsingBG
   properties {
    renderOpacity=0.1
    visibility=SelfHitTestInvisible
    brush {
     bIsDynamicallyLoaded=false
     drawAs=Image
     imageType=FullColor
     mirroring=NoMirror
     resourceName=None
     resourceObject=&MI_UI_Glow_Matchmaking
     tiling=NoTile
     imageSize=<320 128>
     margin=<ltrb 0 0 0 0>
     outlineSettings {
      bUseBrushTransparency=false
      roundingType=HalfHeightRadius
      width=0
      color {colorUseRule=UseColor_Specified, specifiedColor=<rgba 0 0 0 0>}
      cornerRadii=<0 0 0 0>
     }
     tintColor {colorUseRule=UseColor_Specified, specifiedColor=<rgba 1 1 1 1>}
     uVRegion {bIsValid=false, max=<0 0>, min=<0 0>}
    }
    renderTransform {angle=0, scale=<5 0.5>, shear=<0 0>, translation=<0 0>}
   }
   slot {class=OverlaySlot, properties {horizontalAlignment=HAlign_Center, verticalAlignment=VAlign_Center}}
  }
  Image_PulsingFG {
   widgetClassPath=&Image
   bIsVariable=true
   parent=Overlay_285
   widget=Image_PulsingFG
   properties {
    renderOpacity=0.5
    visibility=SelfHitTestInvisible
    brush {
     bIsDynamicallyLoaded=false
     drawAs=Image
     imageType=FullColor
     mirroring=NoMirror
     resourceName=None
     resourceObject=&MI_UI_Glow_Matchmaking
     tiling=NoTile
     imageSize=<320 128>
     margin=<ltrb 0 0 0 0>
     outlineSettings {
      bUseBrushTransparency=false
      roundingType=HalfHeightRadius
      width=0
      color {colorUseRule=UseColor_Specified, specifiedColor=<rgba 0 0 0 0>}
      cornerRadii=<0 0 0 0>
     }
     tintColor {colorUseRule=UseColor_Specified, specifiedColor=<rgba 1 1 1 1>}
     uVRegion {bIsValid=false, max=<0 0>, min=<0 0>}
    }
    colorAndOpacity=<rgba 0.0498 0 1 1>
    renderTransform {angle=0, scale=<4.1027 4.15>, shear=<0 0>, translation=<0 0>}
   }
   slot {class=OverlaySlot, properties {horizontalAlignment=HAlign_Center, verticalAlignment=VAlign_Center}}
  }
  Image_HorizoneFineLine {
   widgetClassPath=&Image
   bIsVariable=true
   parent=Overlay_285
   widget=Image_HorizoneFineLine
   properties {
    visibility=SelfHitTestInvisible
    brush {
     bIsDynamicallyLoaded=false
     drawAs=Image
     imageType=FullColor
     mirroring=NoMirror
     resourceName=None
     resourceObject=&MI_UI_Glow_Matchmaking
     tiling=NoTile
     imageSize=<32 32>
     margin=<ltrb 0 0 0 0>
     outlineSettings {
      bUseBrushTransparency=false
      roundingType=HalfHeightRadius
      width=0
      color {colorUseRule=UseColor_Specified, specifiedColor=<rgba 0 0 0 0>}
      cornerRadii=<0 0 0 0>
     }
     tintColor {colorUseRule=UseColor_Specified, specifiedColor=<rgba 1 1 1 1>}
     uVRegion {bIsValid=false, max=<0 0>, min=<0 0>}
    }
    renderTransform {angle=0, scale=<1.5666 0.0454>, shear=<0 0>, translation=<0 0>}
   }
   slot {class=OverlaySlot, properties {horizontalAlignment=HAlign_Fill, verticalAlignment=VAlign_Fill, padding=<ltrb -80 0 -80 0>}}
  }
  Text {
   widgetClassPath=&CommonTextBlock
   bIsVariable=true
   parent=Default_MatchFound_Animation
   widget=Text
   properties {
    applyLineHeightToBottomLine=false
    bIsScrollingEnabled=false
    justification=Center
    text="Match\r\nFound"
    textTransformPolicy=ToUpper
    font {
     bForceMonospaced=false
     bMaterialIsStencil=false
     fontMaterial=null
     fontObject=&HeadingNow
     letterSpacing=0
     monospacedWidth=1
     size=27
     skewAmount=0
     typefaceFontName=86 Bold Italic
     outlineSettings {
      bApplyOutlineToDropShadows=false
      bMiteredCorners=false
      bSeparateFillAlpha=false
      outlineMaterial=null
      outlineSize=0
      outlineColor=<rgba 0 0 0 1>
     }
    }
    renderTransform {angle=0, scale=<0.8034 0.8034>, shear=<0 0>, translation=<0 0>}
    shadowOffset=<0 0>
   }
   slot {class=OverlaySlot, properties {horizontalAlignment=HAlign_Center, verticalAlignment=VAlign_Center}}
  }
  Text_Outline {
   widgetClassPath=&CommonTextBlock
   bIsVariable=true
   parent=Default_MatchFound_Animation
   widget=Text_Outline
   properties {
    applyLineHeightToBottomLine=false
    bIsScrollingEnabled=false
    justification=Center
    text="Match\r\nFound"
    textTransformPolicy=ToUpper
    colorAndOpacity {colorUseRule=UseColor_Specified, specifiedColor=<rgba 1 1 1 0.25>}
    font {
     bForceMonospaced=false
     bMaterialIsStencil=false
     fontMaterial=/Game/UI/DiscoveryShared/Matchmaking/MatchFoundAnimation/MaterialInstances/MI_UI_MatchFound_TextStyle
     fontObject=&HeadingNow
     letterSpacing=-40
     monospacedWidth=1
     size=46.5
     skewAmount=0
     typefaceFontName=86 Bold Italic
     outlineSettings {
      bApplyOutlineToDropShadows=false
      bMiteredCorners=false
      bSeparateFillAlpha=true
      outlineMaterial=null
      outlineSize=0
      outlineColor=<rgba 1 1 1 1>
     }
    }
    renderTransform {angle=0, scale=<1.0353 1.0353>, shear=<0 0>, translation=<0 0>}
    shadowOffset=<0 0>
   }
   slot {class=OverlaySlot, properties {horizontalAlignment=HAlign_Center, verticalAlignment=VAlign_Center}}
  }
  Spacer_100 {
   widgetClassPath=&Spacer
   parent=Default_MatchFound_Animation
   widget=Spacer_100
   properties {size=<1 320>}
   slot {class=OverlaySlot, properties {horizontalAlignment=HAlign_Center, verticalAlignment=VAlign_Center}}
  }
 ]
}
