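# Default lobby match-found widget, a child of WBP_UI_Lobby_MatchFound_Base that shows a "Game Found!" transition animation.
# Its only widget is a shared WBP_UI_Lobby_TransitionAnim as root.
# PreConstruct fills that widget's text and outline text and tints its border with MatchFoundColor.
# The Play MatchFound Animation event plays the nested widget's Anim_MatchFound and makes it visible without blocking input.
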WidgetBlueprint /EcosystemMatchmakingUI/WBP_UI_Lobby_MatchFound_Default
graphs [EventGraph {nodes [
   Event Play MatchFound Animation() {is=title K2Node_Event}
   n1 = Get WBP_UI_Lobby_TransitionAnim() -> (WBP_UI_Lobby_TransitionAnim:WBP UI Lobby Transition Anim Object Reference)
   > PlayAnimation(
    self=@n1
    InAnimation=Get Anim_MatchFound!(self:WBP UI Lobby Transition Anim Object Reference=@n1) -> (Anim_MatchFound:Widget Animation Object Reference)
   )
   > SetVisibility(self=Get WBP_UI_Lobby_TransitionAnim(), InVisibility=HitTestInvisible)
   Event PreConstruct()
   n2 = Get WBP_UI_Lobby_TransitionAnim()
   > n3 = Set Text(
    Text="LOCTEXT(\"Game\\r\\nFound!\")"
    self=Get Text(self:WBP UI Lobby Transition Anim Object Reference=@n2) -> (Text:Common Text Object Reference)
   )
   > Set Text(Text=@n3, self=Get Text_Outline!(self:WBP UI Lobby Transition Anim Object Reference=@n2) -> (Text_Outline:Common Text Object Reference))
   > Set ContentColorAndOpacity!(
    execute
    ContentColorAndOpacity:Linear Color Structure=Get MatchFoundColor() -> (MatchFoundColor:Linear Color Structure)
    self:Border Object Reference=Get Border!(self:WBP UI Lobby Transition Anim Object Reference=Get WBP_UI_Lobby_TransitionAnim()) -> (Border:Border Object Reference)
   ) -> (then, Output_Get:Linear Color Structure)
  ]}]
variables []
widgets {
 info {
  inheritedWidgetCount=0
  parentClass=/Game/UI/DiscoveryShared/Matchmaking/MatchFoundAnimation/WBP_UI_Lobby_MatchFound_Base.WBP_UI_Lobby_MatchFound_Base_C
  rootWidgetClass=/Game/UI/DiscoveryShared/Matchmaking/MatchFoundAnimation/WBP_UI_Lobby_TransitionAnim.WBP_UI_Lobby_TransitionAnim_C
  widgetCount=1
 }
 widgets [WBP_UI_Lobby_TransitionAnim {
   widgetClassPath=/Game/UI/DiscoveryShared/Matchmaking/MatchFoundAnimation/WBP_UI_Lobby_TransitionAnim
   bIsVariable=true
   widget=WBP_UI_Lobby_TransitionAnim
  }]
}
