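# Pilgrim coachmark tooltip widget that introduces Festival's curated setlists in a sidebar onboarding flow.
# It wraps a single WBP_UIKit_CoachMark, sets its header and description text in PreConstruct, and shows a "Got it" button without a dismiss button.
# OnShowTooltip activates the coachmark.
# The button click (via HandleOtherActionClicked) and OnEndTooltip both notify that the close animation finished, then deactivate the coachmark.
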WidgetBlueprint /FortniteCoreUI/UI/Coachmarks/WBP_PilgrimCoachmark_CuratedSetlist
graphs [EventGraph {nodes [
   "Button Other Action Clicked (WBP_UIKit_CoachMark)"() {is=title K2Node_ComponentBoundEvent}
   > n1 = HandleOtherActionClicked(self:Fort Sidebar Onboard Tooltip Widget Object Reference)
   NotifyCloseAnimationFinished(execute=[@n1.then, @n3.then], self:Fort Sidebar Onboard Tooltip Widget Object Reference)
   > DeactivateWidget(self=Get WBP_UIKit_CoachMark() -> (WBP_UIKit_CoachMark:WBP UIKit Coach Mark Object Reference))
   Event PreConstruct()
   n2 = Get WBP_UIKit_CoachMark()
   > SetTextHeader!(execute, self:WBP UIKit Coach Mark Base Object Reference=@n2, Text:Text="LOCTEXT(\"New! Festival Setlists\")") -> (then, Empty:bool)
   > SetTextDescription!(
    execute
    self:WBP UIKit Coach Mark Base Object Reference=@n2
    Text:Text="LOCTEXT(\"These ready-made setlists let you hit the stage faster with an ever-changing mix of Spotlight, Featured, and New tracks. Pick one and play now!\")"
   ) -> (then, Empty:bool)
   Event OnShowTooltip() {is=title K2Node_Event}
   > ActivateWidget(self=Get WBP_UIKit_CoachMark())
   n3 = Event OnEndTooltip() {is=title K2Node_Event}
  ]}]
variables []
widgets {
 info {
  inheritedWidgetCount=2
  parentClass=/Script/FortniteUI.FortSidebarOnboardTooltipWidget
  rootWidgetClass=/FortUILibrary/UIKit/CoachMark/WBP_UIKit_CoachMark.WBP_UIKit_CoachMark_C
  widgetCount=3
 }
 widgets [
  WBP_UIKit_CoachMark {
   widgetClassPath=/FortUILibrary/UIKit/CoachMark/WBP_UIKit_CoachMark
   bIsVariable=true
   widget=WBP_UIKit_CoachMark
   properties {
    arrowPosition=Left_Down
    bIsFocusable=true
    shouldShowDismissButton=false
    textOtherActionButton=Got it
    widget Width=630
    input Action Other {dataTable=&InputActionDataTable, rowName=GenericForward}
   }
  }
  Dismiss_Button {widgetClassPath=&CommonButtonLegacy, bInherited=true, widget=null}
  OtherAction_Button {widgetClassPath=&CommonButtonLegacy, bInherited=true, widget=null}
 ]
}
