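# Sound cue for the tray notification that plays when matchmaking finds a game.
# A Mixer layers the Start wave at reduced volume with the main GameFound wave at full volume, leaving a third input empty.
# It routes to the application tray notification sound class and uses the shared CONC_TrayNotification concurrency set.
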SoundCue /Game/Sounds/Fort_UI_Sounds/TrayNotifications/SC_TrayNotification_MatchFound
root=Mixer_0
nodes [
 {id=Mixer_0, children [WavePlayer_2,WavePlayer_0,null], properties {inputVolume [0.5,1,0.9]}}
 {id=WavePlayer_0, properties {soundWaveAssetPtr=&/UI_MatchMaking_GameFound_TrayNotification_01}}
 {
  id=WavePlayer_2
  properties {soundWaveAssetPtr=&/UI_MatchMaking_GameFound_TrayNotification_Start_01}
 }
]
properties {
 maxDistance=2097152
 soundClassObject=/Game/Sounds/FortSoundClassesAndModes/Fort_Application_TrayNotification
 volumeMultiplier=1
 concurrencyOverrides {maxCount=1, resolutionRule=StopOldest}
 concurrencySet [&/CONC_TrayNotification]
}
