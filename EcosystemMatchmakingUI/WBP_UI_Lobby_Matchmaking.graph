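# Lobby matchmaking widget that shows matchmaking status, map rotation and tournament countdowns, and the play and mode selection buttons.
# It gets a selected-island view model from the matchmaking subsystem to drive ready-up, unready, match flushing and mode settings.
# Status changes select UMG animations; map rotation blends two thumbnail textures in a dynamic material.
# It also manages onboarding coachmarks, async-loads a mode-specific button and hosts extension content by gameplay tag.
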WidgetBlueprint /EcosystemMatchmakingUI/WBP_UI_Lobby_Matchmaking
animations [
 -NotUsed-Anim_OptionsBtnIn {
  binding_count=6
  length_seconds=0.35
  object=&self:-NotUsed-Anim_OptionsBtnIn
  sequence {
   bindings [
    Switcher_StatusIcons {
     id=#1
     possessed_class=CommonVisibilitySwitcher
     tracks [{
       type=MovieSceneFloatTrack
       display_name=Render Opacity
       sections [{type=MovieSceneFloatSection, end_frame=21, start_frame=0, channels [None {type=MovieSceneScriptingFloatChannel, const [0,2]}]}]
      }]
    }
    WBP_UI_Button_ModeSelection {
     id=#2
     possessed_class=WBP_UI_Lobby_Button_ModeSelection_C
     tracks [{
       type=MovieSceneFloatTrack
       display_name=Render Opacity
       sections [{type=MovieSceneFloatSection, end_frame=21, start_frame=0, channels [None {type=MovieSceneScriptingFloatChannel, keys [[0,0],[7,1],[18,1]]}]}]
      }]
    }
    GridSlot_2 {
     id=#3
     parent_id=#2
     possessed_class=GridSlot
     tracks [{
       type=MovieSceneDoubleVectorTrack
       display_name=Nudge
       sections [{
         type=MovieSceneDoubleVectorSection
         end_frame=21
         start_frame=0
         channels [
          Vector.X {type=MovieSceneScriptingDoubleChannel, keys [[0,-16],[21,0]]}
          Vector.Y {type=MovieSceneScriptingDoubleChannel}
         ]
        }]
      }]
    }
    VerticalBoxSlot_1 {
     id=#4
     parent_id=#5
     possessed_class=VerticalBoxSlot
     tracks [{
       type=MovieSceneMarginTrack
       display_name=Padding
       sections [{
         type=MovieSceneMarginSection
         end_frame=21
         start_frame=0
         channels [
          Left {type=MovieSceneScriptingFloatChannel}
          Top {type=MovieSceneScriptingFloatChannel}
          Right {type=MovieSceneScriptingFloatChannel}
          Bottom {type=MovieSceneScriptingFloatChannel, const [-72,1]}
         ]
        }]
      }]
    }
    Grid_StatusInfo {
     id=#5
     possessed_class=GridPanel
     tracks [{
       type=MovieSceneFloatTrack
       display_name=Render Opacity
       sections [{type=MovieSceneFloatSection, end_frame=21, start_frame=0, channels [None {type=MovieSceneScriptingFloatChannel}]}]
      }]
    }
    VB_TextLeft {
     id=#6
     possessed_class=VerticalBox
     tracks [{
       type=MovieSceneFloatTrack
       display_name=Render Opacity
       sections [{type=MovieSceneFloatSection, end_frame=21, start_frame=0, channels [None {type=MovieSceneScriptingFloatChannel, const [0,2]}]}]
      }]
    }
   ]
   display_rate {numerator=60}
  }
 }
 -NotUsed-Anim_OptionsBtnOut {
  binding_count=6
  length_seconds=0.35
  object=&self:-NotUsed-Anim_OptionsBtnOut
  sequence {
   bindings [
    Switcher_StatusIcons {
     id=#1
     possessed_class=CommonVisibilitySwitcher
     tracks [{
       type=MovieSceneFloatTrack
       display_name=Render Opacity
       sections [{type=MovieSceneFloatSection, end_frame=21, start_frame=0, channels [None {type=MovieSceneScriptingFloatChannel, const [0,2]}]}]
      }]
    }
    WBP_UI_Button_ModeSelection {
     id=#2
     possessed_class=WBP_UI_Lobby_Button_ModeSelection_C
     tracks [{
       type=MovieSceneFloatTrack
       display_name=Render Opacity
       sections [{
         type=MovieSceneFloatSection
         end_frame=21
         start_frame=0
         channels [None {type=MovieSceneScriptingFloatChannel, keys [[0,1],[7.5,1],[18,0]]}]
        }]
      }]
    }
    GridSlot_2 {
     id=#3
     parent_id=#2
     possessed_class=GridSlot
     tracks [{
       type=MovieSceneDoubleVectorTrack
       display_name=Nudge
       sections [{
         type=MovieSceneDoubleVectorSection
         end_frame=21
         start_frame=0
         channels [
          Vector.X {type=MovieSceneScriptingDoubleChannel, keys [[0,0],[21,16]]}
          Vector.Y {type=MovieSceneScriptingDoubleChannel}
         ]
        }]
      }]
    }
    VerticalBoxSlot_1 {
     id=#4
     parent_id=#5
     possessed_class=VerticalBoxSlot
     tracks [{
       type=MovieSceneMarginTrack
       display_name=Padding
       sections [{
         type=MovieSceneMarginSection
         end_frame=21
         start_frame=0
         channels [
          Left {type=MovieSceneScriptingFloatChannel}
          Top {type=MovieSceneScriptingFloatChannel}
          Right {type=MovieSceneScriptingFloatChannel}
          Bottom {type=MovieSceneScriptingFloatChannel, const [-72,1]}
         ]
        }]
      }]
    }
    Grid_StatusInfo {
     id=#5
     possessed_class=GridPanel
     tracks [{
       type=MovieSceneFloatTrack
       display_name=Render Opacity
       sections [{type=MovieSceneFloatSection, end_frame=21, start_frame=0, channels [None {type=MovieSceneScriptingFloatChannel}]}]
      }]
    }
    VB_TextLeft {
     id=#6
     possessed_class=VerticalBox
     tracks [{
       type=MovieSceneFloatTrack
       display_name=Render Opacity
       sections [{type=MovieSceneFloatSection, end_frame=21, start_frame=0, channels [None {type=MovieSceneScriptingFloatChannel, const [0,2]}]}]
      }]
    }
   ]
   display_rate {numerator=60}
  }
 }
 ---Used--- {binding_count=0, length_seconds=5, object=&self:---Used---, sequence {bindings []}}
 Anim_Widget_Intro {
  binding_count=1
  length_seconds=0.5
  object=&self:Anim_Widget_Intro
  sequence {
   bindings [Switcher_GameIslandInfo {
     id=#7
     possessed_class=CommonVisibilitySwitcher
     tracks [{
       type=MovieSceneFloatTrack
       display_name=Render Opacity
       sections [{type=MovieSceneFloatSection, end_frame=30, start_frame=0, channels [None {type=MovieSceneScriptingFloatChannel, const [1,2]}]}]
      }]
    }]
   display_rate {numerator=60}
  }
 }
 Anim_StatusIn {
  binding_count=3
  length_seconds=0.3
  object=&self:Anim_StatusIn
  sequence {
   bindings [
    GridSlot_0 {
     id=#8
     parent_id=#6
     possessed_class=GridSlot
     tracks [{
       type=MovieSceneDoubleVectorTrack
       display_name=Nudge
       sections [{
         type=MovieSceneDoubleVectorSection
         end_frame=18
         start_frame=0
         channels [
          Vector.X {type=MovieSceneScriptingDoubleChannel, keys [[0,-20],[4.867,-20],[16,0]]}
          Vector.Y {type=MovieSceneScriptingDoubleChannel}
         ]
        }]
      }]
    }
    Switcher_StatusIcons {
     id=#9
     possessed_class=CommonVisibilitySwitcher
     tracks [
      {
       type=MovieSceneFloatTrack
       display_name=Render Opacity
       sections [{type=MovieSceneFloatSection, end_frame=18, start_frame=0, channels [None {type=MovieSceneScriptingFloatChannel, keys [[0,0],[13.845,1]]}]}]
      }
      {
       type=MovieScene2DTransformTrack
       sections [{
         type=MovieScene2DTransformSection
         end_frame=13
         start_frame=0
         channels [
          Translation.X {type=MovieSceneScriptingFloatChannel}
          Translation.Y {type=MovieSceneScriptingFloatChannel}
          Angle {type=MovieSceneScriptingFloatChannel}
          Scale.X {type=MovieSceneScriptingFloatChannel, keys [[0,0.6],[13.844,1]]}
          Scale.Y {type=MovieSceneScriptingFloatChannel, keys [[0,0.6],[13.844,1]]}
          Shear.X {type=MovieSceneScriptingFloatChannel}
          Shear.Y {type=MovieSceneScriptingFloatChannel}
         ]
        }]
      }
     ]
    }
    VB_TextLeft {
     id=#6
     possessed_class=VerticalBox
     tracks [{
       type=MovieSceneFloatTrack
       display_name=Render Opacity
       sections [{
         type=MovieSceneFloatSection
         end_frame=11
         start_frame=0
         channels [None {type=MovieSceneScriptingFloatChannel, keys [[0,0],[4.867,0],[9,1]]}]
        }]
      }]
    }
   ]
   display_rate {numerator=60}
  }
 }
 Anim_StatusOut {
  binding_count=3
  length_seconds=0.3
  object=&self:Anim_StatusOut
  sequence {
   bindings [
    GridSlot_0 {
     id=#8
     parent_id=#6
     possessed_class=GridSlot
     tracks [{
       type=MovieSceneDoubleVectorTrack
       display_name=Nudge
       sections [{
         type=MovieSceneDoubleVectorSection
         end_frame=18
         start_frame=0
         channels [
          Vector.X {type=MovieSceneScriptingDoubleChannel, keys [[0,0],[9.474,0],[17.461,32]]}
          Vector.Y {type=MovieSceneScriptingDoubleChannel}
         ]
        }]
      }]
    }
    Switcher_StatusIcons {
     id=#9
     possessed_class=CommonVisibilitySwitcher
     tracks [
      {
       type=MovieSceneFloatTrack
       display_name=Render Opacity
       sections [{type=MovieSceneFloatSection, end_frame=18, start_frame=9, channels [None {type=MovieSceneScriptingFloatChannel, keys [[9,1],[14.845,0]]}]}]
      }
      {
       type=MovieScene2DTransformTrack
       sections [{
         type=MovieScene2DTransformSection
         end_frame=18
         start_frame=5
         channels [
          Translation.X {type=MovieSceneScriptingFloatChannel}
          Translation.Y {type=MovieSceneScriptingFloatChannel}
          Angle {type=MovieSceneScriptingFloatChannel}
          Scale.X {type=MovieSceneScriptingFloatChannel, keys [[5.844,1],[14.845,1.4]]}
          Scale.Y {type=MovieSceneScriptingFloatChannel, keys [[5.844,1],[14.845,1.4]]}
          Shear.X {type=MovieSceneScriptingFloatChannel}
          Shear.Y {type=MovieSceneScriptingFloatChannel}
         ]
        }]
      }
     ]
    }
    VB_TextLeft {
     id=#6
     possessed_class=VerticalBox
     tracks [{
       type=MovieSceneFloatTrack
       display_name=Render Opacity
       sections [{
         type=MovieSceneFloatSection
         end_frame=18
         start_frame=0
         channels [None {type=MovieSceneScriptingFloatChannel, keys [[14.156,1],[17.478,0]]}]
        }]
      }]
    }
   ]
   display_rate {numerator=60}
  }
 }
 Anim_StatusIntro {
  binding_count=5
  length_seconds=0.6
  object=&self:Anim_StatusIntro
  sequence {
   bindings [
    VerticalBoxSlot_1 {
     id=#10
     parent_id=#11
     possessed_class=VerticalBoxSlot
     tracks [{
       type=MovieSceneMarginTrack
       display_name=Padding
       sections [{
         type=MovieSceneMarginSection
         end_frame=36
         start_frame=0
         channels [
          Left {type=MovieSceneScriptingFloatChannel}
          Top {type=MovieSceneScriptingFloatChannel}
          Right {type=MovieSceneScriptingFloatChannel}
          Bottom {type=MovieSceneScriptingFloatChannel, keys [[0,-60],[18,24],[36,24]]}
         ]
        }]
      }]
    }
    GridSlot_0 {
     id=#12
     parent_id=#13
     possessed_class=GridSlot
     tracks [{
       type=MovieSceneDoubleVectorTrack
       display_name=Nudge
       sections [{
         type=MovieSceneDoubleVectorSection
         end_frame=36
         start_frame=0
         channels [
          Vector.X {type=MovieSceneScriptingDoubleChannel, keys [[0,-20],[10,-20],[28,0],[36,0]]}
          Vector.Y {type=MovieSceneScriptingDoubleChannel}
         ]
        }]
      }]
    }
    Switcher_StatusIcons {
     id=#14
     possessed_class=CommonVisibilitySwitcher
     tracks [
      {
       type=MovieScene2DTransformTrack
       sections [{
         type=MovieScene2DTransformSection
         end_frame=36
         start_frame=0
         channels [
          Translation.X {type=MovieSceneScriptingFloatChannel}
          Translation.Y {type=MovieSceneScriptingFloatChannel}
          Angle {type=MovieSceneScriptingFloatChannel}
          Scale.X {type=MovieSceneScriptingFloatChannel, keys [[0,0.6],[2.999,0.6],[9.999,1],[36,1]]}
          Scale.Y {type=MovieSceneScriptingFloatChannel, keys [[0,0.6],[2.999,0.6],[9.999,1],[36,1]]}
          Shear.X {type=MovieSceneScriptingFloatChannel}
          Shear.Y {type=MovieSceneScriptingFloatChannel}
         ]
        }]
      }
      {
       type=MovieSceneFloatTrack
       display_name=Render Opacity
       sections [{
         type=MovieSceneFloatSection
         end_frame=36
         start_frame=0
         channels [None {type=MovieSceneScriptingFloatChannel, keys [[0,0],[2.999,0],[9.999,1],[36,1]]}]
        }]
      }
     ]
    }
    Grid_StatusInfo {id=#11, possessed_class=GridPanel, tracks []}
    VB_TextLeft {
     id=#13
     possessed_class=VerticalBox
     tracks [{
       type=MovieSceneFloatTrack
       display_name=Render Opacity
       sections [{
         type=MovieSceneFloatSection
         end_frame=36
         start_frame=0
         channels [None {type=MovieSceneScriptingFloatChannel, keys [[0,0],[10,0],[18,1],[36,1]]}]
        }]
      }]
    }
   ]
   display_rate {numerator=60}
  }
 }
 Anim_StatusOutro {
  binding_count=5
  length_seconds=0.6
  object=&self:Anim_StatusOutro
  sequence {
   bindings [
    VerticalBoxSlot_1 {
     id=#10
     parent_id=#11
     possessed_class=VerticalBoxSlot
     tracks [{
       type=MovieSceneMarginTrack
       display_name=Padding
       sections [{
         type=MovieSceneMarginSection
         end_frame=36
         start_frame=0
         channels [
          Left {type=MovieSceneScriptingFloatChannel}
          Top {type=MovieSceneScriptingFloatChannel}
          Right {type=MovieSceneScriptingFloatChannel}
          Bottom {type=MovieSceneScriptingFloatChannel, keys [[0,24],[18,24],[36,-60]]}
         ]
        }]
      }]
    }
    GridSlot_0 {
     id=#12
     parent_id=#13
     possessed_class=GridSlot
     tracks [{
       type=MovieSceneDoubleVectorTrack
       display_name=Nudge
       sections [{
         type=MovieSceneDoubleVectorSection
         end_frame=36
         start_frame=0
         channels [
          Vector.X {type=MovieSceneScriptingDoubleChannel, keys [[0,0],[8,0],[27,24],[36,24]]}
          Vector.Y {type=MovieSceneScriptingDoubleChannel}
         ]
        }]
      }]
    }
    Switcher_StatusIcons {
     id=#14
     possessed_class=CommonVisibilitySwitcher
     tracks [
      {
       type=MovieScene2DTransformTrack
       sections [{
         type=MovieScene2DTransformSection
         end_frame=36
         start_frame=0
         channels [
          Translation.X {type=MovieSceneScriptingFloatChannel}
          Translation.Y {type=MovieSceneScriptingFloatChannel}
          Angle {type=MovieSceneScriptingFloatChannel}
          Scale.X {type=MovieSceneScriptingFloatChannel, keys [[0,1],[13,1.4]]}
          Scale.Y {type=MovieSceneScriptingFloatChannel, keys [[0,1],[13,1.4]]}
          Shear.X {type=MovieSceneScriptingFloatChannel}
          Shear.Y {type=MovieSceneScriptingFloatChannel}
         ]
        }]
      }
      {
       type=MovieSceneFloatTrack
       display_name=Render Opacity
       sections [{
         type=MovieSceneFloatSection
         end_frame=36
         start_frame=0
         channels [None {type=MovieSceneScriptingFloatChannel, keys [[0,1],[5,1],[11,0],[36,0]]}]
        }]
      }
     ]
    }
    Grid_StatusInfo {id=#11, possessed_class=GridPanel, tracks []}
    VB_TextLeft {
     id=#13
     possessed_class=VerticalBox
     tracks [{
       type=MovieSceneFloatTrack
       display_name=Render Opacity
       sections [{
         type=MovieSceneFloatSection
         end_frame=36
         start_frame=0
         channels [None {type=MovieSceneScriptingFloatChannel, keys [[0,1],[8,1],[15,0],[36,0]]}]
        }]
      }]
    }
   ]
   display_rate {numerator=60}
  }
 }
 Anim_MatchFound {
  binding_count=1
  length_seconds=0.2512
  object=&self:Anim_MatchFound
  sequence {
   bindings [Content_Main {
     id=#15
     possessed_class=VerticalBox
     tracks [{
       type=MovieSceneFloatTrack
       display_name=Render Opacity
       sections [{type=MovieSceneFloatSection, end_frame=15, start_frame=0, channels [None {type=MovieSceneScriptingFloatChannel, keys [[0,1],[15.073,0]]}]}]
      }]
    }]
   display_rate {numerator=60}
  }
 }
 Anim_MapRotation {
  binding_count=2
  length_seconds=0.3167
  object=&self:Anim_MapRotation
  sequence {
   bindings [
    BackgroundThumbnail {
     id=#16
     possessed_class=Image
     tracks [{
       type=MovieSceneWidgetMaterialTrack
       sections [{type=MovieSceneParameterSection, channels [OffsetToNext {type=MovieSceneScriptingFloatChannel, keys [[0,0],[19,1]]}]}]
      }]
    }
    Txt_MapName {
     id=#17
     possessed_class=CommonTextBlock
     tracks [
      {
       type=MovieSceneFloatTrack
       display_name=Render Opacity
       sections [{type=MovieSceneFloatSection, channels [None {type=MovieSceneScriptingFloatChannel, keys [[0,0],[0.6,0],[12.003,1]]}]}]
      }
      {
       type=MovieScene2DTransformTrack
       sections [{type=MovieScene2DTransformSection, channels [Translation.X {type=MovieSceneScriptingFloatChannel, keys [[0,400],[12,0]]}]}]
      }
     ]
    }
   ]
   display_rate {numerator=60}
  }
 }
]
graphs [
 AnimateOldToNewPlayButtonText {nodes [
   n1 = AnimateOldToNewPlayButtonText!() -> (then, NewPlayButtonText:Text) {is=title K2Node_FunctionEntry}
   > n2 = Get!(
    execute:Exec
    self:Selected Island View Model Object Reference=Get FortMatchmakingViewModel() -> (FortMatchmakingViewModel:Selected Island View Model Object Reference)
   ) -> (then:Exec, else:Exec, GroupError:Ready Up Error UI Object Reference)
   > n3 = Set LocalSecondText(LocalSecondText:Text=Play Button Subtitle!(self:Ready Up Error UI Object Reference=@n2) -> (ReturnValue:Text)) -> (Output_Get:Text)
   n4 = Branch(execute=[@n3.then, @n2.else], Condition=TextIsEmpty(InText=Get LocalSecondText() -> (LocalSecondText:Text)))
   n5 = Find(
    TargetMap:Map of Unique Net Id Structures to EFortUIReadyUpStatus Enums=Get ReadyUpStatusByLocalPlayerId!(self:Selected Island View Model Object Reference=Get FortMatchmakingViewModel()) -> (ReadyUpStatusByLocalPlayerId:Map of Unique Net Id Structures to EFortUIReadyUpStatus Enums)
    Key:"Unique Net Id Structure (by ref)"=GetNetIDFromContextObject!(ContextObject:Object Reference=GetOwningPlayer()) -> (ReturnValue:Unique Net Id Structure)
   ) -> (Value:EFortUIReadyUpStatus Enum=Invalid)
   > n6 = Branch(Condition=OR Boolean(A="Equal (Enum)"(A:EFortUIReadyUpStatus Enum=@n5, B:EFortUIReadyUpStatus Enum=Play), B="Equal (Enum)"(A=@n5, B=Ready)))
   > n7 = GetTeamRankedText!(execute, self) -> (then, SecondaryText:Text)
   > n8 = Set LocalSecondText(LocalSecondText=@n7)
   SetRegisterTextValuesThenAnimate!(
    execute=[@n4.else, @n6.else, @n8.then]
    self:WBP UI Lobby Button Matchmaking Object Reference=Get WBP_UI_Button_Matchmaking() -> (WBP_UI_Button_Matchmaking:WBP UI Lobby Button Matchmaking Object Reference)
    TextValue:Text=@n1
    SecondTextValue:Text=Get LocalSecondText()
   ) -> (then)
  ]}
 ApplyButtonCornerRadius {nodes [
   n1 = ApplyButtonCornerRadius!() -> (then, BackgroundBlock:WBP UIKit Block Background Base Object Reference) {is=title K2Node_FunctionEntry}
   > n2 = Cast To WBP_UIKit_Block_BackgroundWithOutline!(execute, Object=@n1) -> (then, CastFailed, AsWBP UIKit Block Background with Outline:WBP UIKit Block Background With Outline Object Reference) {is=title K2Node_DynamicCast}
   n3 = Get BackgroundDynamicMaterial!(self:WBP UIKit Block Background Base Object Reference=@n2) -> (BackgroundDynamicMaterial:Material Instance Dynamic Object Reference)
   n4 = Get OutlineDynamicMaterial!(self:WBP UIKit Block Background Base Object Reference=@n2) -> (OutlineDynamicMaterial:Material Instance Dynamic Object Reference)
   > SetScalarParameterValue(self=[@n3, @n4], ParameterName=Radius in Pixels, Value=1)
   > SetScalarParameterValue(self=@n3, ParameterName=Corner Radius, Value=16)
   > SetVectorParameterValue!(execute, self=@n4, ParameterName=Corner Radius, Value_R:float=30, Value_G:float=30, Value_B:float=30, Value_A:float=30) -> (then)
  ]}
 BP_GetDesiredFocusTarget {nodes [
   Get Desired Focus Target()
   > Return Node!(execute, ReturnValue:Widget Object Reference=GetLastFocusedWidget!(self) -> (Widget:Widget Object Reference))
  ]}
 CheckCVarStartMatch {nodes [
   Inputs()
   > n1 = Branch(Condition=GetCanStartMatchEnabled!(self:Selected Island View Model Object Reference=Get FortMatchmakingViewModel()) -> (ReturnValue:bool))
   > Outputs!(true:Exec, false:Exec=@n1.else)
  ]}
 ClearFlushAudio {nodes [
   ClearFlushAudio() {is=title K2Node_FunctionEntry}
   n1 = Get FlushAudioTimer() -> (FlushAudioTimer:Timer Handle Structure)
   > Branch(Condition=Is Valid Timer Handle!(Handle:Timer Handle Structure=@n1) -> (ReturnValue:bool))
   > Clear and Invalidate Timer by Handle(Handle=@n1)
  ]}
 CloseMatchmakingSettings {nodes [
   CloseMatchmakingSettings() {is=title K2Node_FunctionEntry}
   > n1 = Get!(execute) -> (then, else, SettingsModal:Common Activatable Widget Object Reference)
   > DeactivateWidget(self=@n1)
  ]}
 EventGraph {nodes [
   n1 = "On Clicked (WBP_UI_Button_Matchmaking)"!() -> (OutputDelegate, then, Button:Common Button Base Object Reference) {is=title K2Node_ComponentBoundEvent}
   n2 = Get FortMatchmakingViewModel()
   n3 = Branch(
    execute=[@n1.then, @n12.then]
    Condition="Equal (Enum)"(
     A:EFortUIMatchmakingStatus Enum=Get MatchmakingStatus!(self:Selected Island View Model Object Reference=@n2) -> (MatchmakingStatus:EFortUIMatchmakingStatus Enum=InvalidIsland)
     B:EFortUIMatchmakingStatus Enum=Unready
    )
   )
   > ReadyUp(self:Selected Island View Model Object Reference=Get FortMatchmakingViewModel())
   > Set Last Focused Widget(Last Focused Widget:Widget Object Reference=Get WBP_UI_Button_Matchmaking()) -> (Output_Get:Widget Object Reference)
   Unready(execute=@n3.else, self:Selected Island View Model Object Reference=Get FortMatchmakingViewModel())
   > SetVisibility(self=Get WBP_UI_Button_Flush() -> (WBP_UI_Button_Flush:WBP UI Lobby Button Matchmaking Object Reference), InVisibility=Collapsed)
   > Set CanStartMatchShouldBeVisible(CanStartMatchShouldBeVisible:bool=false) -> (Output_Get:bool)
   n4 = "On Clicked (DecoyButton_TournamentDetails)"!() -> (OutputDelegate, then, Button:Common Button Base Object Reference) {is=title K2Node_ComponentBoundEvent}
   > GoToTournamentDetails!(
    execute
    self:Selected Island View Model Object Reference=Get FortMatchmakingViewModel()
    LocalPlayer:Local Player Object Reference=GetOwningLocalPlayer!(self:Widget Object Reference) -> (ReturnValue:Local Player Object Reference)
   ) -> (then)
   > Fire Event Element Interacted(Element=@n4.Button)
   n5 = "On Clicked (WBP_UI_Button_ModeSelection)"!() -> (OutputDelegate, then, Button:Common Button Base Object Reference) {is=title K2Node_ComponentBoundEvent}
   n6 = OnActivityModeSetSelectionModalOpened!() -> (OutputDelegate, then, SettingsModal:Common Activatable Widget Object Reference) {is=title K2Node_CustomEvent}
   OpenModeSettings!(
    execute=@n5.then
    self:Selected Island View Model Object Reference=Get FortMatchmakingViewModel()
    LocalPlayer:Local Player Object Reference=GetOwningLocalPlayer()
    OnModalOpened:"Delegate (by ref)"=@n6
   ) -> (then)
   n7 = "On Visibility Changed (WBP_UI_Button_ModeSelection)"!() -> (OutputDelegate, then, InVisibility:ESlateVisibility Enum=Visible) {is=title K2Node_ComponentBoundEvent}
   n8 = Get WBP_PilgrimCoachmark_CuratedSetlist() -> (WBP_PilgrimCoachmark_CuratedSetlist:WBP Pilgrim Coachmark Curated Setlist Object Reference)
   > n9 = Branch(Condition=AND Boolean(
     A="Equal (Enum)"(A:ESlateVisibility Enum=@n7.InVisibility, B:ESlateVisibility Enum=Visible)
     B=ShouldShowOnboarding!(self:Fort Sidebar Onboard Tooltip Widget Object Reference=@n8) -> (ReturnValue:bool)
    ))
   n10 = Get WBP_PilgrimCoachmark_CuratedSetlist()
   > ActivateWidget(self=@n10)
   DeactivateWidget(execute=@n9.else, self=@n10)
   "On Clicked (WBP_UI_Button_Secondary)"!() -> (OutputDelegate, then, Button:Common Button Base Object Reference) {is=title K2Node_ComponentBoundEvent}
   n11 = Get GroupError!(self:Selected Island View Model Object Reference=Get FortMatchmakingViewModel()) -> (GroupError:Ready Up Error UI Object Reference)
   > Is Valid(InputObject=@n11)
   > SecondaryRedirect(self:Ready Up Error UI Object Reference=@n11)
   n12 = "On Clicked (WBP_UI_Button_Primary)"!() -> (OutputDelegate, then, Button:Common Button Base Object Reference) {is=title K2Node_ComponentBoundEvent}
   Animate Matchmaking Status() {is=title K2Node_CustomEvent}
   n13 = Get Anim_StatusOut() -> (Anim_StatusOut:Widget Animation Object Reference)
   n14 = Select!(
    InvalidIsland:Widget Animation Object Reference
    Unready:Widget Animation Object Reference=Get Anim_StatusOutro() -> (Anim_StatusOutro:Widget Animation Object Reference)
    WaitingForPlayersReadyUp:Widget Animation Object Reference=@n13
    FindingServer:Widget Animation Object Reference=@n13
    MatchFound:Widget Animation Object Reference=@n13
    Num:Widget Animation Object Reference
    Index:EFortUIMatchmakingStatus Enum=Get Current Matchmaking Status() -> (Current Matchmaking Status:EFortUIMatchmakingStatus Enum=InvalidIsland)
   ) -> (ReturnValue:Widget Animation Object Reference)
   > n15 = PlayAnimation(InAnimation=@n14)
   n16 = On Animation Finished() {is=title K2Node_CustomEvent}
   BindToAnimationFinished(execute=@n15.then, Animation=@n14, Delegate=@n16)
   > n17 = Switch on EFortUIMatchmakingStatus!(execute, Selection:EFortUIMatchmakingStatus Enum=Get Current Matchmaking Status() -> (Current Matchmaking Status=InvalidIsland)) -> (InvalidIsland:Exec, Unready:Exec, WaitingForPlayersReadyUp:Exec, FindingServer:Exec, MatchFound:Exec, Num:Exec) {is=title K2Node_SwitchEnum}
   Set Previous MM Data Text(
    execute=[@n17.FindingServer, @n17.WaitingForPlayersReadyUp]
    Previous MM Data Text:Text=Get Text(self:Text Object Reference=Get Text_MMData() -> (Text_MMData:Common Text Object Reference))
   ) -> (Output_Get:Text)
   > Set Previous MM Status Text(Previous MM Status Text:Text=Get Text(self=Get Text_MMStatus() -> (Text_MMStatus:Common Text Object Reference))) -> (Output_Get:Text)
   > Set Previous MM Additional Data Text(Previous MM Additional Data Text:Text=Get Text(self=Get Text_MMAdditionalData() -> (Text_MMAdditionalData:Common Text Object Reference))) -> (Output_Get:Text)
   > Set with Broadcast isAnimating!(execute, isAnimating:bool=true) -> (then, Output_Get:bool)
   PlayAnimation(execute=@n17.MatchFound, InAnimation=Get Anim_MatchFound() -> (Anim_MatchFound:Widget Animation Object Reference))
   n18 = Start TournamentTimer!() -> (OutputDelegate, then, Date Time:Date Time Structure) {is=title K2Node_CustomEvent}
   > PrintString!(execute, InString=Hello, bPrintToScreen, bPrintToLog, TextColor, Duration, Key) -> (then)
   > Setup Tournament Countdown Timer VM!(execute, self, DataTime:Date Time Structure=@n18.Date Time) -> (then)
   > n19 = Branch(Condition=Get bIsTournamentLive!(self:Selected Island View Model Object Reference=Get FortMatchmakingViewModel()) -> (bIsTournamentLive:bool))
   > SetVisibility(self=Get Grid_LiveRankedCountdown() -> (Grid_LiveRankedCountdown:Grid Panel Object Reference), InVisibility=HitTestInvisible)
   SetVisibility(execute=@n19.else, self=Get MapRotationCountdown() -> (MapRotationCountdown:WBP UIKit Badge Object Reference), InVisibility=HitTestInvisible)
   Match Flushed() {is=title K2Node_CustomEvent}
   n20 = Get WBP_UI_Button_Flush()
   > Set Visibility(Visibility=Collapsed, self=@n20)
   > Set CanStartMatchShouldBeVisible(CanStartMatchShouldBeVisible=false)
   n21 = OnButtonBaseClicked_Event!() -> (OutputDelegate, then, Button:Common Button Base Object Reference) {is=title K2Node_CustomEvent}
   > Set FlushClicked(FlushClicked:bool=true) -> (Output_Get:bool)
   > FlushMMS(self:Selected Island View Model Object Reference=Get FortMatchmakingViewModel())
   > Set Visibility(Visibility=Collapsed, self=Get WBP_UI_Button_Flush())
   > Set CanStartMatchShouldBeVisible(CanStartMatchShouldBeVisible=false)
   n22 = AddWidgetAlias_ModeSpecificWidget!() -> (OutputDelegate, then, IsVisible:bool=~) {is=title K2Node_CustomEvent}
   n23 = GetSoftWidgetClassForAlias!(
    self:Fort UIData Configuration Context Object Reference=Get FortUIDataConfigurationContext!() -> (ReturnValue:Fort UIData Configuration Context Object Reference) {is=title K2Node_GetBlueprintContextEx}
    WidgetAlias:"Gameplay Tag Structure (by ref)"=MakeLiteralGameplayTag!(Value:Gameplay Tag Structure=(TagName="MashFrontendUI.UI.FrontendButton")) -> (ReturnValue:Gameplay Tag Structure)
   ) -> (ReturnValue:User Widget Soft Class Reference)
   > Branch(Condition=IsValidSoftClassReference!(SoftClassReference:"Object Soft Class Reference (by ref)"=@n23) -> (ReturnValue:bool))
   > n24 = CreateWidgetAsync!(
    execute:Exec
    UserWidgetSoftClass:User Widget Soft Class Reference=@n23
    OwningPlayer:Player Controller Object Reference
    bSuspendInputUntilComplete:bool=true
   ) -> (then:Exec, OnComplete:Exec, OnFailure:Exec, UserWidget:User Widget Object Reference) {is=title K2Node_AsyncAction}
   n25 = Set ModeSpecificWidget(execute=@n24.OnComplete, ModeSpecificWidget:User Widget Object Reference=@n24) -> (Output_Get:User Widget Object Reference)
   > AddChild(self=Get Slot_ModeSpecificButtonSlot() -> (Slot_ModeSpecificButtonSlot:Named Slot Object Reference), Content=@n25)
   > SetVisibility(
    self=Get ModeSpecificWidget() -> (ModeSpecificWidget:User Widget Object Reference)
    InVisibility=Select!(Option 0:ESlateVisibility Enum=Collapsed, Option 1:ESlateVisibility Enum=Visible, Index:bool=@n22.IsVisible) -> (ReturnValue:ESlateVisibility Enum)
   )
   > SetIsEnabled(self=Get ModeSpecificWidget(), bInIsEnabled=NOT Boolean(A=Get IsPlaytimeOut() -> (IsPlaytimeOut:bool)))
   CheckPlayerSittingOutTooltipVisibility() {is=title K2Node_CustomEvent}
   n26 = Get WBP_Lobby_Coachmark_AFKSittingOut() -> (WBP_Lobby_Coachmark_AFKSittingOut:WBP Lobby Coachmark AFKSitting Out Object Reference)
   n27 = GetOwningLocalPlayer()
   > n28 = Branch(Condition=AND Boolean(A=ShouldShowOnboarding(self=@n26), B=DidAutoSitOut!(LocalPlayerController:Local Player Object Reference=@n27) -> (ReturnValue:bool)))
   n29 = Get WBP_Lobby_Coachmark_AFKSittingOut()
   > ActivateWidget(self=@n29)
   DeactivateWidget(execute=@n28.else, self=@n29)
   n30 = AthenaLobbyDeactivated() {is=title K2Node_CustomEvent}
   > n31 = Branch(Condition=Get is MOTD Activated!(self:Athena Lobby Object Reference=Get AthenaLobby() -> (AthenaLobby:Athena Lobby Object Reference)) -> (is MOTD Activated:bool))
   Set Last Focused Widget(execute=@n31.else, Last Focused Widget=Get WBP_UI_Button_Matchmaking())
   n32 = OnSelectedIslandViewModelSet!() -> (OutputDelegate, then, ViewModel:Selected Island View Model Object Reference) {is=title K2Node_CustomEvent}
   > n33 = SetMatchMakingVM!(execute, self, FortMatchmakingViewModel:Selected Island View Model Object Reference=@n32.ViewModel) -> (then)
   Set Last Focused Widget(execute=[@n57.then, @n33.then], Last Focused Widget=Get WBP_UI_Button_Matchmaking())
   > OnMatchmakingWidgetInitialized(self:Matchmaking UISubsystem Object Reference=Get MatchmakingUISubsystem() -> (ReturnValue:Matchmaking UISubsystem Object Reference) {is=title K2Node_GetSubsystem})
   n34 = Start Rotation Timer!() -> (OutputDelegate, then, Date Time:Date Time Structure) {is=title K2Node_CustomEvent}
   > Setup Map Rotation Countdown Timer VM!(execute, self, DataTime:Date Time Structure=@n34.Date Time) -> (then)
   > SetVisibility(self=[Get MapRotationCountdown(), Get BackgroundThumbnail() -> (BackgroundThumbnail:Image Object Reference)], InVisibility=HitTestInvisible)
   > Set PrevMapTexture(PrevMapTexture:Texture Object Reference=Get CurrentMapTexture() -> (CurrentMapTexture:Texture Object Reference)) -> (Output_Get:Texture Object Reference)
   > Set Background Thumbnail!(
    execute
    self
    HasThumbnail:bool=HasThumbnail!(self:Selected Island View Model Object Reference=Get FortMatchmakingViewModel()) -> (ReturnValue:bool)
   ) -> (then) {is=title K2Node_CallFunction}
   > SetVisibility(
    self=Get Switcher_Countdown() -> (Switcher_Countdown:Common Visibility Switcher Object Reference)
    InVisibility=Select(Option 0=SelfHitTestInvisible, Option 1=Collapsed, Index=Get IsPlaytimeOut())
   )
   On Island Changed() {is=title K2Node_CustomEvent}
   > n35 = Cast To FortPlayerController(Object=GetOwningPlayer())
   > n36 = Branch(Condition=AreModalsActive!(LocalPlayerController:Fort Player Controller Object Reference=@n35) -> (ReturnValue:bool))
   n37 = Tell Lobby Navigate To Matchmaking() {is=title K2Node_CustomEvent}
   n38 = Get Anim_Widget_Intro() -> (Anim_Widget_Intro:Widget Animation Object Reference)
   BindToAnimationFinished(execute=[@n53.then, @n36.else], Animation=@n38, Delegate=@n37)
   > PlayAnimation(InAnimation=@n38, PlaybackSpeed=1.5)
   OnMapRotation() {is=title K2Node_CustomEvent}
   n39 = Get PrevMapTexture() -> (PrevMapTexture:Texture Object Reference)
   n40 = Get CurrentMapTexture()
   > Branch(Condition=AND Boolean(
     A=IsValid(Object=@n39)
     B="Not Equal (Object)"(A:Object Reference=@n39, B:Object Reference=@n40) -> (ReturnValue:bool=~) {is=title K2Node_PromotableOperator}
    ))
   > SetTextureParameterValue(
    self=Get ThumbnailDynamicMaterial() -> (ThumbnailDynamicMaterial:Material Instance Dynamic Object Reference)
    ParameterName=TextureA
    Value=Get PrevMapTexture()
   )
   > SetTextureParameterValue(self=Get ThumbnailDynamicMaterial(), ParameterName=TextureB, Value=Get CurrentMapTexture())
   > n41 = "Play Animation with Finished event (legacy)"!(
    execute:Exec
    Widget:User Widget Object Reference=Self-Reference()
    InAnimation:Widget Animation Object Reference=Get Anim_MapRotation() -> (Anim_MapRotation:Widget Animation Object Reference)
    StartAtTime:float=0
    NumLoopsToPlay:int=1
    PlayMode:EUMGSequencePlayMode Enum
    PlaybackSpeed:float=1
   ) -> (then:Exec, Result:UMGSequence Player Object Reference, Finished:Exec) {is=title K2Node_PlayAnimation}
   n42 = Get CurrentMapTexture()
   SetTextureParameterValue(execute=@n41.Finished, self=Get ThumbnailDynamicMaterial(), ParameterName=TextureA, Value=@n42)
   > SetTextureParameterValue(self=Get ThumbnailDynamicMaterial(), ParameterName=TextureB, Value=@n42)
   Match Found() {is=title K2Node_CustomEvent}
   > n43 = Branch(Condition=Get FlushClicked() -> (FlushClicked:bool))
   > Set CanStartMatchShouldBeVisible(CanStartMatchShouldBeVisible=false)
   n44 = Check CVar Start Match!(execute:Exec=@n43.else) -> (true:Exec, false:Exec) {is=title K2Node_MacroInstance}
   > n45 = Set Visibility(Visibility=Visible, self=@n20)
   Set CanStartMatchShouldBeVisible(execute=[@n45.then, @n44.false], CanStartMatchShouldBeVisible=true)
   n46 = OnWidgetDeactivated() {is=title K2Node_CustomEvent}
   > Set SettingsModal(SettingsModal:Common Activatable Widget Object Reference) -> (Output_Get:Common Activatable Widget Object Reference)
   > DelayUntilNextTick!(execute) -> (then)
   > Tell Lobby Navigate To Matchmaking() {is=title K2Node_CallFunction}
   > Unbind all Events from On Widget Deactivated!(execute:Exec, self:Common Activatable Widget Object Reference=@n6.SettingsModal) -> (then:Exec) {is=title K2Node_ClearDelegate}
   Event Construct()
   > RegisterWidgetHost!(
    execute
    WidgetHost:Widget Object Reference
    WidgetHostTag:"Gameplay Tag Structure (by ref)"=Get ExtensionHolderSlotTag() -> (ExtensionHolderSlotTag:Gameplay Tag Structure)
    IsPlayerSpecific:bool=true
   ) -> (then)
   > n47 = Branch(Condition=IsValid(Object=Get FortMatchmakingViewModel()))
   > n48 = Set Map Rotation Date!(
    execute
    self
    Date Time:Date Time Structure=Next Map Rotation Date!(self:Selected Island View Model Object Reference=Get FortMatchmakingViewModel()) -> (ReturnValue:Date Time Structure)
   ) -> (then) {is=title K2Node_CallFunction}
   n49 = GetDynamicMaterial(execute=[@n48.then, @n47.else], self=Get BackgroundThumbnail())
   > Set ThumbnailDynamicMaterial(ThumbnailDynamicMaterial:Material Instance Dynamic Object Reference=@n49) -> (Output_Get:Material Instance Dynamic Object Reference)
   > n50 = FindParentWidgetOfType!(
    execute
    StartingWidget:Widget Object Reference=Self-Reference()
    Type:Widget Class Reference=/FortniteCoreUI/Frontend/AthenaLobby.AthenaLobby_C
   ) -> (then, ReturnValue:Athena Lobby Object Reference)
   > n51 = Set AthenaLobby(AthenaLobby:Athena Lobby Object Reference=@n50) -> (Output_Get:Athena Lobby Object Reference)
   > Bind Event to On MMWidget Display Changed(self:Athena Lobby Object Reference=@n51, Delegate=Create Event()) {is=title K2Node_AddDelegate}
   n52 = Get AthenaLobby()
   > Bind Event to On Widget Deactivated(self:Common Activatable Widget Object Reference=@n52, Delegate=@n30) {is=title K2Node_AddDelegate}
   > n53 = Branch(Condition=Get is MM Widget on Screen!(self:Athena Lobby Object Reference=@n52) -> (is MM Widget on Screen:bool))
   Event OnInitialized()
   > Assign On Clicked!(execute:Exec, self:Common Button Base Object Reference=Get WBP_UI_Button_Flush(), Delegate:"Delegate (by ref)"=@n21) -> (then:Exec) {is=title K2Node_AssignDelegate}
   n54 = Get MatchmakingUISubsystem()
   n55 = GetSelectedIslandViewModel!(self:Matchmaking UISubsystem Object Reference=@n54) -> (ReturnValue:Selected Island View Model Object Reference)
   > n56 = Branch(Condition=IsValid(Object=@n55))
   > n57 = SetMatchMakingVM(FortMatchmakingViewModel=@n55)
   Bind Event to On Selected Island View Model Set(execute=@n56.else, self:Matchmaking UISubsystem Object Reference=@n54, Delegate=@n32) {is=title K2Node_AddDelegate}
   Event OnAddedToFocusPath!() -> (OutputDelegate, then, InFocusEvent:Focus Event Structure) {is=title K2Node_Event}
   n58 = Get AthenaLobby()
   > Set Last Focused Widget!(
    execute
    Last Focused Widget=Get MatchmakingWidget!(self:Athena Lobby Base Object Reference=@n58) -> (MatchmakingWidget:Common User Widget Object Reference)
    self:Athena Lobby Object Reference=@n58
   ) -> (then, Output_Get)
   Event PreConstruct()
   > SetCompactUIStyle() {is=title K2Node_CallFunction}
   > SetModeSelectionButtonTextStyle()
   n59 = Get MapRotationCountdown()
   > Set HorizontalAlignment!(
    execute
    HorizontalAlignment:EHorizontalAlignment Enum=HAlign_Center
    self:Grid Slot Object Reference=SlotAsGridSlot(Widget=Get Content!(self:WBP UIKit Badge Base Object Reference=@n59) -> (Content:Grid Panel Object Reference))
   ) -> (then, Output_Get:EHorizontalAlignment Enum=HAlign_Fill)
   n60 = Get WBP_UI_Button_Secondary() -> (WBP_UI_Button_Secondary:WBP UIKit Button Quiet Object Reference)
   n61 = Get WBP_UI_Button_Primary() -> (WBP_UI_Button_Primary:WBP UIKit Button Regular Object Reference)
   > n62 = For Each Loop(Array:Array of WBP UIKit Block Background With Outline Object References=Make Array!(
     "[0]":WBP UIKit Block Background With Outline Object Reference=Get Block_Background!(self:WBP UIKit Button CTA Base Object Reference=@n61) -> (Block_Background:WBP UIKit Block Background With Outline Object Reference)
     "[1]":WBP UIKit Block Background With Outline Object Reference=Get Block_Background(self=@n60)
    ) -> (Array:Array of WBP UIKit Block Background With Outline Object References)) -> (Array Element:WBP UIKit Block Background With Outline Object Reference)
   > ApplyButtonCornerRadius!(execute, self, BackgroundBlock:WBP UIKit Block Background Base Object Reference=@n62) -> (then) {is=title K2Node_CallFunction}
   n63 = Event Add Content To Host Widget!() -> (
    OutputDelegate
    then
    WidgetHostTag:"Gameplay Tag Structure (by ref)"
    ContentWidget:Widget Object Reference
    Padding:"Margin Structure (by ref)"
    bFill:bool
   ) {is=title K2Node_Event}
   > Branch(Condition=MatchesTag!(TagOne:Gameplay Tag Structure=@n63.WidgetHostTag, TagTwo:Gameplay Tag Structure=Get ExtensionHolderSlotTag(), bExactMatch:bool=true) -> (ReturnValue:bool))
   n64 = Get VerticalBox_ExtensionHolder() -> (VerticalBox_ExtensionHolder:Vertical Box Object Reference)
   > AddChild(self=@n64, Content=@n63.ContentWidget)
   > SetPadding!(
    execute
    self:Vertical Box Slot Object Reference=SlotAsVerticalBoxSlot(Widget=@n64)
    InPadding_Left:float=0
    InPadding_Top:float=0
    InPadding_Right:float=0
    InPadding_Bottom:float=16
   ) -> (then)
   > SetIsEnabled(self=Get VerticalBox_ExtensionHolder(), bInIsEnabled=NOT Boolean(A=Get IsPlaytimeOut()))
   Event Destruct()
   > UnregisterWidgetHost!(execute, WidgetHost:Widget Object Reference, WidgetHostTag:"Gameplay Tag Structure (by ref)"=Get ExtensionHolderSlotTag()) -> (then)
   Event On Activated()
   > Init(self:Player Playtime View Model Object Reference=Get PlayerPlaytimeVM() -> (PlayerPlaytimeVM:Player Playtime View Model Object Reference))
   n65 = Set SettingsModal(execute=@n6.then, SettingsModal=@n6.SettingsModal)
   > Bind Event to On Widget Deactivated(self=@n65, Delegate=@n46)
   n66 = Get AthenaLobby()
   Navigate To Last Focused Widget(execute=@n67.then, self:Athena Lobby Object Reference=@n66)
   "Update Matchmaking State Appearance "!(
    execute=[@n68.then, @n70.then, @n69.Unready]
    self
    NewMatchmakingState:EFortUIMatchmakingStatus Enum=Get Current Matchmaking Status() -> (Current Matchmaking Status=InvalidIsland)
   ) -> (then)
   Get Anim_StatusOutro()
   UnbindAllFromAnimationFinished!(execute=@n37.then, self:User Widget Object Reference, Animation:Widget Animation Object Reference=@n38) -> (then)
   > Set Last Focused Widget(Last Focused Widget=Get MatchmakingWidget(self=@n66), self=@n66)
   > n67 = DelayUntilNextTick()
   n68 = SetActiveWidget(
    execute=@n72.then
    self=Get Switcher_MainOrMatchFound() -> (Switcher_MainOrMatchFound:Common Visibility Switcher Object Reference)
    Widget=Get SB_MatchFoundWidget() -> (SB_MatchFoundWidget:Size Box Object Reference)
   )
   UnbindAllFromAnimationFinished(execute=@n16.then, Animation=@n14)
   > n69 = Switch on EFortUIMatchmakingStatus(Selection=Get Current Matchmaking Status() -> (Current Matchmaking Status=InvalidIsland))
   Set with Broadcast isAnimating(execute=[@n69.FindingServer, @n69.WaitingForPlayersReadyUp], isAnimating=false)
   > n70 = PlayAnimation(InAnimation=Get Anim_StatusIn() -> (Anim_StatusIn:Widget Animation Object Reference))
   n71 = Get Content_Main() -> (Content_Main:Vertical Box Object Reference)
   n72 = SetMaxDesiredHeight!(
    execute=@n69.MatchFound
    self:Size Box Object Reference=Get SB_MatchFoundWidget()
    InMaxDesiredHeight:float=GetLocalSize!(Geometry:"Geometry Structure (by ref)"=GetPaintSpaceGeometry!(self:Widget Object Reference=@n71) -> (ReturnValue:"Geometry Structure (by ref)")) -> (ReturnValue_X:double, ReturnValue_Y:double).ReturnValue_Y
   ) -> (then)
  ]}
 GetCreatetimeOut {nodes [
   n1 = GetCreatetimeOut!() -> (then, PlaytimeState:EPlaytimeState Enum=Unblocked) {is=title K2Node_FunctionEntry}
   > Return Node!(execute, IsCreatetimeOut:bool="Equal (Enum)"(A:EPlaytimeState Enum=@n1, B:EPlaytimeState Enum=CreateTime_Blocked))
  ]}
 GetLastFocusedWidget {nodes [
   GetLastFocusedWidget() {is=title K2Node_FunctionEntry}
   n1 = Get Switcher_PrimaryButton() -> (Switcher_PrimaryButton:Common Visibility Switcher Object Reference)
   n2 = Get Last Focused Widget() -> (Last Focused Widget:Widget Object Reference)
   n3 = Get WBP_UI_Button_Flush()
   > Return Node!(
    execute
    Widget:Widget Object Reference=Select(
     Option 0:Widget Object Reference=GetActiveWidget!(self:Common Visibility Switcher Object Reference=@n1) -> (ReturnValue:Widget Object Reference)
     Option 1:Widget Object Reference=@n2
     Index="Equal (Object)"(A=@n2, B=@n3)
    ) -> (ReturnValue:Widget Object Reference)
   )
  ]}
 GetPlay Button Enabled {nodes [
   n1 = GetPlay Button Enabled!() -> (then, ErrorUI:Ready Up Error UI Object Reference, MatchmakingStatus:EFortUIMatchmakingStatus Enum=Unready) {is=title K2Node_FunctionEntry}
   > n2 = Branch(Condition=IsValid(Object=@n1))
   n3 = Get ErrorUI() -> (ErrorUI:Ready Up Error UI Object Reference)
   > n4 = Set bIsPlayButtonEnabled(bIsPlayButtonEnabled:bool=AND Boolean(
     A=HasRedirect!(self:Ready Up Error UI Object Reference=@n3) -> (ReturnValue:bool)
     B=CanRedirect!(self:Ready Up Error UI Object Reference=@n3) -> (ReturnValue:bool)
    )) -> (Output_Get:bool)
   n5 = Set with Broadcast IsBottomBarVisible!(execute=[@n7.then, @n4.then], IsBottomBarVisible:bool=Get bIsPlayButtonEnabled() -> (bIsPlayButtonEnabled:bool)) -> (then, Output_Get:bool)
   > Return Node!(execute, IsPlayLocked:bool=NOT Boolean(A=@n5))
   n6 = Get MatchmakingStatus!() -> (MatchmakingStatus=InvalidIsland)
   n7 = Set bIsPlayButtonEnabled(
    execute=@n2.else
    bIsPlayButtonEnabled=AND Boolean(
     A="Not Equal (Enum)"!(A:EFortUIMatchmakingStatus Enum=MatchFound, B:EFortUIMatchmakingStatus Enum=@n6) -> (ReturnValue:bool)
     B="Not Equal (Enum)"(A=@n6, B=InvalidIsland)
    )
   )
  ]}
 GetPlaytimeOut {nodes [
   n1 = GetPlaytimeOut!() -> (then, PlaytimeState:EPlaytimeState Enum=Unblocked) {is=title K2Node_FunctionEntry}
   > Return Node!(execute, IsPlaytimeOut:bool="Equal (Enum)"(A=@n1, B=PlayTime_Blocked))
  ]}
 GetTeamRankedText {nodes [
   GetTeamRankedText() {is=title K2Node_FunctionEntry}
   n1 = Get FortMatchmakingViewModel()
   > n2 = Branch(Condition=IsRanked!(self:Selected Island View Model Object Reference=@n1) -> (ReturnValue:bool))
   n3 = Get PartyRank!(self:Selected Island View Model Object Reference=@n1) -> (PartyRank:int=0)
   > n4 = Branch(Condition="Not Equal (Integer)"(A=@n3, B=-1))
   n5 = String To Name(InString=Get Ranked String From Island!(self:Selected Island View Model Object Reference=@n1) -> (ReturnValue:String) {is=title K2Node_CallFunction})
   n6 = Get PartyLowestPosition!(self:Selected Island View Model Object Reference=@n1) -> (PartyLowestPosition:int=0)
   n7 = Get PartyHighestMMR!(self:Selected Island View Model Object Reference=@n1) -> (PartyHighestMMR:int=0)
   > n8 = Set RankName(RankName:Text=GetHabaneroDisplayDataOverride!(
     InDisplayData:Fort Habanero Display Data Object Reference=Get HabaneroDisplayData() -> (HabaneroDisplayData:Fort Habanero Display Data Object Reference)
     InOverrideData:Fort Compete UIHabanero Tier Override List Object Reference=Get RankInfoOverrides() -> (RankInfoOverrides:Fort Compete UIHabanero Tier Override List Object Reference)
     InKey:"Fort Compete UIHabanero Tier Override Key Structure (by ref)"=Make Fort Compete UIHabanero Tier Override Key!(RankedTrack:Name=@n5, Season:String, Rank:int=@n3, MatchmakingRating:int=@n7, CurrentPosition:int=@n6) -> (FortCompeteUIHabaneroTierOverrideKey:Fort Compete UIHabanero Tier Override Key Structure) {is=title K2Node_MakeStruct}
    ) -> (
     OutData_ForegroundIcon:Material Instance Soft Object Reference
     OutData_ForegroundIconSmall:Material Instance Soft Object Reference
     OutData_BackgroundIcon:Material Instance Soft Object Reference
     OutData_BackgroundIconSmall:Material Instance Soft Object Reference
     OutData_WhiteIcon:Texture 2D Soft Object Reference
     OutData_SimpleIconSmall:Texture 2D Soft Object Reference
     OutData_Name:Text
     OutData_TierType:Fort Habanero Tier Type Data Object Reference
     OutData_AdditionalLeftIcon:Material Instance Soft Object Reference
     OutData_AdditionalRightIcon:Material Instance Soft Object Reference
     ReturnValue:bool
    ).OutData_Name) -> (Output_Get:Text)
   n9 = Get RankName() -> (RankName:Text)
   n10 = Get TotalPlayerCount!(self:Selected Island View Model Object Reference=@n1) -> (TotalPlayerCount:int=0)
   Return Node!(
    execute=[@n11.then, @n8.then]
    SecondaryText:Text=Select(
     Option 0:Text=Format Text!(Format="LOCTEXT(\"Rank: {0}\")", 0:Text=@n9) -> (Result)
     Option 1:Text=Format Text(Format="LOCTEXT(\"Party Rank: {0}\")", 0=@n9)
     Index="integer > integer"(A=@n10, B=1)
    ) -> (ReturnValue:Text)
   )
   n11 = Set RankName(execute=@n4.else, RankName="LOCTEXT(\"Pending\")")
   Return Node(execute=@n2.else)
  ]}
 HandleSecondaryButtonTextChanged {nodes [
   n1 = HandleSecondaryButtonTextChanged!() -> (then, SecondaryText:Text) {is=title K2Node_FunctionEntry}
   > n2 = Set IsSecondaryButtonVisible(IsSecondaryButtonVisible:bool=NOT Boolean(A=TextIsEmpty(InText=@n1))) -> (Output_Get:bool)
   > SetActiveWidget(
    self=Get Switcher_PrimaryButton()
    Widget=Select(Option 0:Widget Object Reference=Get WBP_UI_Button_Matchmaking(), Option 1:Widget Object Reference=Get WBP_UI_Button_Primary(), Index=@n2) -> (ReturnValue:Widget Object Reference)
   )
   n3 = Get IsSecondaryButtonVisible() -> (IsSecondaryButtonVisible:bool)
   n4 = Get WBP_UI_Button_Secondary()
   > SetVisibility(
    self=@n4
    InVisibility=Select(Option 0:ESlateVisibility Enum=Collapsed, Option 1:ESlateVisibility Enum=SelfHitTestInvisible, Index=@n3) -> (ReturnValue:ESlateVisibility Enum)
   )
   > Branch(Condition=@n3)
   > Set Text(Text=Get SecondaryText() -> (SecondaryText:Text), self:Fort CTAButton Object Reference=@n4)
  ]}
 NavigateDownFromSubmodeButton {nodes [
   NavigateDownFromSubmodeButton!() -> (then, Navigation:EUINavigation Enum) {is=title K2Node_FunctionEntry}
   > n1 = Branch(Condition=IsVisible(self=Get WBP_UI_Button_Flush()))
   > n2 = Set Last Focused Widget!(execute, Last Focused Widget=Get WBP_UI_Button_Flush()) -> (then, Output_Get)
   > Return Node!(execute, ReturnValue:Widget Object Reference=@n2)
   n3 = Set Last Focused Widget(execute=@n1.else, Last Focused Widget=Get WBP_UI_Button_Matchmaking())
   > Return Node(ReturnValue=@n3)
  ]}
 NavigateUpFromPlayButton {nodes [
   NavigateUpFromPlayButton!() -> (then, Navigation:EUINavigation Enum) {is=title K2Node_FunctionEntry}
   > n1 = Branch(Condition=IsVisible(self=Get WBP_UI_Button_Flush()))
   > n2 = Set Last Focused Widget(Last Focused Widget=Get WBP_UI_Button_Flush())
   > Return Node(ReturnValue=@n2)
   n3 = Branch(
    execute=@n1.else
    Condition=IsVisible(self=Get WBP_UI_Button_ModeSelection() -> (WBP_UI_Button_ModeSelection:WBP UI Lobby Button Mode Selection Object Reference))
   )
   > n4 = Set Last Focused Widget(Last Focused Widget=Get WBP_UI_Button_ModeSelection())
   > Return Node(ReturnValue=@n4)
   HandleNavigateUpToMOTDWidget!(execute=@n3.else, self:Athena Lobby Object Reference=Get AthenaLobby(), Navigation:EUINavigation Enum=Up) -> (then, ReturnValue:Widget Object Reference)
  ]}
 NavigateUpFromStartMatch {nodes [
   NavigateUpFromStartMatch!() -> (then, Navigation:EUINavigation Enum) {is=title K2Node_FunctionEntry}
   n1 = Get WBP_UI_Button_ModeSelection()
   > Branch(Condition=IsVisible(self=@n1))
   > Return Node(ReturnValue=@n1)
  ]}
 OnCanStartMatchChange {nodes [
   n1 = OnCanStartMatchChange!() -> (then, CanStartMatch:bool=~) {is=title K2Node_FunctionEntry}
   n2 = Get FlushClicked()
   > n3 = Branch(Condition=AND Boolean!(A=@n1, B=Get CanStartMatchShouldBeVisible() -> (CanStartMatchShouldBeVisible:bool), C:bool=NOT Boolean(A=@n2)) -> (ReturnValue))
   n4 = Get WBP_UI_Button_Flush()
   > Set Visibility(Visibility=Visible, self=@n4)
   Set Visibility(execute=@n3.else, Visibility=Collapsed, self=@n4)
  ]}
 OnFocusReceived {nodes [
   OnFocusReceived!() -> (then, MyGeometry:Geometry Structure, InFocusEvent:Focus Event Structure) {is=title K2Node_FunctionEntry}
   > n1 = Branch(Condition=Get is MM Widget on Screen(self=Get AthenaLobby()))
   > Return Node(ReturnValue:Event Reply Structure=SetUserFocus!(
     Reply:"Event Reply Structure (by ref)"=Handled()
     FocusWidget:Widget Object Reference=GetLastFocusedWidget() {is=title K2Node_CallFunction}
     bInAllUsers:bool=false
    ) -> (ReturnValue:Event Reply Structure))
   Return Node(execute=@n1.else, ReturnValue=Unhandled())
  ]}
 OnIsRankedChanged {nodes [
   OnIsRankedChanged!() -> (then, Is Ranked:bool=~) {is=title K2Node_FunctionEntry}
   > AnimateOldToNewPlayButtonText!(execute, self, NewPlayButtonText:Text=Get Text(self:WBP UIKit Button Generic Base Object Reference=Get WBP_UI_Button_Matchmaking())) -> (then) {is=title K2Node_CallFunction}
  ]}
 OnMatchmakingWidgetDisplayChanged {nodes [
   n1 = OnMatchmakingWidgetDisplayChanged!() -> (then, Visible:bool=~) {is=title K2Node_FunctionEntry}
   n2 = Get FortMatchmakingViewModel()
   > SetBottomButtonVisibility!(
    execute
    self:WBP Bottom Bar Decoy Button Object Reference=Get DecoyButton_TournamentDetails() -> (DecoyButton_TournamentDetails:WBP Bottom Bar Decoy Button Object Reference)
    ShouldButtonBeVisible:bool=AND Boolean!(A=@n1, B=Is The Island A Tournament Island!(self:Selected Island View Model Object Reference=@n2) -> (ReturnValue:bool)) -> (ReturnValue)
   ) -> (then)
  ]}
 OnRankChanged {nodes [
   OnRankChanged!() -> (then, Rank:int) {is=title K2Node_FunctionEntry}
   > AnimateOldToNewPlayButtonText(NewPlayButtonText=Get Text(self=Get WBP_UI_Button_Matchmaking()))
  ]}
 PlayFlushAudio {nodes [
   PlayFlushAudio() {is=title K2Node_FunctionEntry}
   > SpawnSound2D!(
    execute
    Sound:Sound Base Object Reference=Get MatchmakingFlushedSound() -> (MatchmakingFlushedSound:Sound Base Object Reference)
    VolumeMultiplier:float=1
    PitchMultiplier:float=1
    StartTime:float=0
    ConcurrencySettings:Sound Concurrency Object Reference
    bPersistAcrossLevelTransition:bool=true
    bAutoDestroy:bool=true
   ) -> (then, ReturnValue:Audio Component Object Reference)
  ]}
 PlayMatchmakingAudio {nodes [
   n1 = PlayMatchmakingAudio!() -> (then, bFlushed:bool=~, Time:float=~) {is=title K2Node_FunctionEntry}
   > n2 = Branch(Condition=@n1)
   n3 = Get FlushAudioTimer()
   n4 = Branch(execute=[@n7.then, @n2.then], Condition=Is Valid Timer Handle(Handle=@n3))
   > n5 = Clear and Invalidate Timer by Handle(Handle=@n3)
   n6 = Set Timer by Event(execute=[@n5.then, @n4.else], Delegate=Create Event(), Time=Get Time() -> (Time:float), bLooping=false)
   > Set FlushAudioTimer(FlushAudioTimer:Timer Handle Structure=@n6) -> (Output_Get:Timer Handle Structure)
   n7 = SpawnSound2D(
    execute=@n2.else
    Sound=Get MatchmakingMatchFoundSound() -> (MatchmakingMatchFoundSound:Sound Base Object Reference)
    VolumeMultiplier=1
    PitchMultiplier=1
    StartTime=0
    bPersistAcrossLevelTransition=true
    bAutoDestroy=true
   )
  ]}
 RemoveContentFromWidgetHost {nodes [
   n1 = Remove Content From Host Widget!() -> (
    then
    WidgetHostTag:"Gameplay Tag Structure (by ref)"
    ContentWidget:Widget Object Reference
    RemovalType:EWidgetSlottingRemovalType Enum=Unparent
   ) {is=title K2Node_FunctionEntry}
   > n2 = Branch(Condition=MatchesTag(TagOne=@n1, TagTwo=Get ExtensionHolderSlotTag(), bExactMatch=true))
   n3 = Get VerticalBox_ExtensionHolder()
   > n4 = RemoveChild(self=@n3, Content=@n1.ContentWidget)
   > SetPadding(self=SlotAsVerticalBoxSlot(Widget=@n3), InPadding_Left=0, InPadding_Top=0, InPadding_Right=0, InPadding_Bottom=0)
   > Return Node(ReturnValue:bool=@n4)
   Return Node(execute=@n2.else, ReturnValue=false)
  ]}
 Set Background Thumbnail {nodes [
   n1 = Set Background Thumbnail!() -> (then, HasThumbnail:bool=~) {is=title K2Node_FunctionEntry}
   > n2 = Branch(Condition=@n1)
   > n3 = Branch(Condition=NOT Boolean(A=Get IsPlaytimeOut()))
   > n4 = Branch(Condition=Is The Island A Tournament Island(self=Get FortMatchmakingViewModel()))
   > SetVisibility(self=Get BackgroundThumbnail(), InVisibility=SelfHitTestInvisible)
   SetMapName!(
    execute=@n4.else
    self
    MapName:Text=Island Sub Title!(self:Selected Island View Model Object Reference=Get FortMatchmakingViewModel()) -> (ReturnValue:Text)
   ) -> (then)
   > SetVisibility(self=Get BackgroundThumbnail(), InVisibility=SelfHitTestInvisible)
   SetVisibility(execute=[@n3.else, @n2.else], self=Get BackgroundThumbnail(), InVisibility=Collapsed)
  ]}
 Set Download Progress {nodes [
   n1 = Set Download Progress!() -> (then, Download Info:Matchmaking Download Info Structure) {is=title K2Node_FunctionEntry}
   > n2 = Set with Broadcast Download Info!(execute, Download Info:Matchmaking Download Info Structure=@n1) -> (
    then
    Output_Get_bIsDownloading:bool=~
    Output_Get_DownloadType:EFortUIDownloadType Enum
    Output_Get_Progress:float=~
    Output_Get_DownloadSize:Integer64
   )
   > SetProgress!(
    execute
    self:WBP UI Download Progress Object Reference=Get WBP_UI_DownloadProgress() -> (WBP_UI_DownloadProgress:WBP UI Download Progress Object Reference)
    Value:double=@n2.Output_Get_Progress
   ) -> (then)
  ]}
 Set Is Tournament {nodes [
   n1 = Set Is Tournament!() -> (then, Is Tournament:bool=~) {is=title K2Node_FunctionEntry}
   > n2 = Branch(Condition=IsNewTournamentLobbyWidgetEnabled!(self:Matchmaking UISubsystem Object Reference=Get MatchmakingUISubsystem()) -> (ReturnValue:bool))
   n3 = Get DecoyButton_TournamentDetails()
   > SetBottomButtonVisibility(self=@n3, ShouldButtonBeVisible=@n1)
   > SetScalarParameterValue(self=Get ThumbnailDynamicMaterial(), ParameterName=IsTournamentTexture, Value="To Float (Boolean)"(InBool=@n1))
   > SetDesiredSizeOverride!(
    execute
    self:Image Object Reference=Get BackgroundThumbnail()
    DesiredSize_X:double=32
    DesiredSize_Y:double=Select(Option 0:double=200, Option 1:double=282, Index=Get Is Tournament() -> (Is Tournament:bool)) -> (ReturnValue:double)
   ) -> (then)
   SetBottomButtonVisibility(execute=@n2.else, self=@n3, ShouldButtonBeVisible=false)
  ]}
 Set MM Addtional Data Text {nodes [
   Set MM Addtional Data Text!() -> (then, isAnimating:bool=~, DownloadInfo:Matchmaking Download Info Structure) {is=title K2Node_FunctionEntry}
   > n1 = Branch(Condition=Get isAnimating() -> (isAnimating:bool))
   > Return Node!(execute, FormattedText:Text=Get Previous MM Additional Data Text() -> (Previous MM Additional Data Text:Text))
   n2 = Get!() -> (
    DownloadInfo_bIsDownloading:bool=~
    DownloadInfo_DownloadType:EFortUIDownloadType Enum
    DownloadInfo_Progress:float=~
    DownloadInfo_DownloadSize:Integer64
   )
   n3 = Branch(
    execute=@n1.else
    Condition="Equal (Enum)"(A:EFortUIDownloadType Enum=@n2.DownloadInfo_DownloadType, B:EFortUIDownloadType Enum=PreparingDownload)
   )
   > Return Node()
   n4 = Branch(
    execute=@n3.else
    Condition="integer64 > integer64"(A:Integer64=@n2.DownloadInfo_DownloadSize, B:Integer64) -> (ReturnValue:bool=~) {is=title K2Node_PromotableOperator}
   )
   > Return Node(FormattedText=Format Text!(
     Format="LOCTEXT(\"of {size}\")"
     size:Text=Set Memory Format!(NumBytes:Integer64=@n2.DownloadInfo_DownloadSize) -> (FormattedMemory:Text) {is=title K2Node_MacroInstance}
    ) -> (Result))
   Return Node(execute=@n4.else)
  ]}
 Set MM Data Text {nodes [
   Set MM Data Text!() -> (
    then
    MatchmakingStatus:EFortUIMatchmakingStatus Enum=Unready
    ReadyPlayerCount:Byte
    TotalPlayerCount:Byte
    TimeElapsed:int
    DownloadProgress:double=~
    DownloadSize:Integer64
    isAnimating:bool=~
   ) {is=title K2Node_FunctionEntry}
   > n1 = Branch(Condition=Get isAnimating())
   > Return Node!(execute, PlayerCountFormattedText:Text=Get Previous MM Data Text() -> (Previous MM Data Text:Text))
   n2 = Switch on EFortUIDownloadType!(
    execute=@n1.else
    Selection:EFortUIDownloadType Enum=Get!() -> (
     Download Info_bIsDownloading:bool=~
     Download Info_DownloadType:EFortUIDownloadType Enum
     Download Info_Progress:float=~
     Download Info_DownloadSize:Integer64
    ).Download Info_DownloadType
   ) -> (NoDownload:Exec, DownloadWithProgress:Exec, PreparingDownload:Exec, QueuedDownload:Exec) {is=title K2Node_SwitchEnum}
   > n3 = Switch on EFortUIMatchmakingStatus(Selection=Get MatchmakingStatus() -> (MatchmakingStatus=InvalidIsland))
   Return Node(execute=[@n3.InvalidIsland, @n3.Unready])
   Return Node(execute=@n3.MatchFound)
   Return Node(
    execute=@n3.WaitingForPlayersReadyUp
    PlayerCountFormattedText=Format Text!(
     Format="LOCTEXT(\"{current}/{total}\")"
     current:Byte=Get ReadyPlayerCount() -> (ReadyPlayerCount:Byte=0)
     total:Byte=Get TotalPlayerCount!() -> (TotalPlayerCount:Byte=0)
    ) -> (Result)
   )
   n4 = "To Float (Integer)"(InInt=Get TimeElapsed() -> (TimeElapsed:int=0))
   n5 = Split(SourceString=Replace(SourceString=TimeSecondsToString!(InSeconds:float=@n4) -> (ReturnValue:String), From=.00), InStr=:, SearchDir=FromStart)
   n6 = "To String (Integer)"(InInt=String To Integer(InString=@n5))
   Return Node(
    execute=[@n3.FindingServer, @n2.PreparingDownload]
    PlayerCountFormattedText="To Text (String)"(InString=Append!(A=@n6, B=:, C:String=@n5.RightS) -> (ReturnValue))
   )
   n7 = Get DownloadProgress() -> (DownloadProgress:double)
   n8 = "Absolute (Float)"(A=@n7)
   n9 = Round(A=float * float(A="Clamp (Float)"(Value=@n7, Min=0, Max=@n8), B=100) {is=title K2Node_PromotableOperator}) {is=title K2Node_CallFunction}
   Return Node(
    execute=[@n2.DownloadWithProgress, @n2.QueuedDownload]
    PlayerCountFormattedText=Format Text!(Format="LOCTEXT(\"{progress}%\")", progress:double="To Float (Integer)"(InInt=@n9)) -> (Result)
   )
  ]}
 Set MM Status Text {nodes [
   Set MM Status Text!() -> (
    then
    Download Info:Matchmaking Download Info Structure
    MM State:EFortUIMatchmakingStatus Enum=InvalidIsland
    Queued Players:int
    isAnimating:bool=~
   ) {is=title K2Node_FunctionEntry}
   > n1 = Branch(Condition=Get isAnimating())
   > Return Node!(execute, Text:Text=Get Previous MM Status Text() -> (Previous MM Status Text:Text))
   n2 = Switch on EFortUIDownloadType(execute=@n1.else, Selection=Get() -> (Download Info_bIsDownloading=~, Download Info_Progress=~).Download Info_DownloadType)
   Return Node(execute=@n2.DownloadWithProgress, Text="LOCTEXT(\"Downloading\")")
   > n3 = Branch(Condition="integer > integer"(A=Get Queued Players() -> (Queued Players:int=0)))
   > Return Node(Text=Format Text!(Format="LOCTEXT(\"Queued Players: {Count}\")", Count:int=Get Queued Players() -> (Queued Players=0)) -> (Result))
   Return Node(
    execute=@n3.else
    Text=Select!(
     InvalidIsland:Text
     Unready:Text
     WaitingForPlayersReadyUp:Text="LOCTEXT(\"Waiting for Players\")"
     FindingServer:Text="LOCTEXT(\"Finding Server\")"
     MatchFound:Text
     Num:Text
     Index:EFortUIMatchmakingStatus Enum=Get MM State() -> (MM State:EFortUIMatchmakingStatus Enum=InvalidIsland)
    ) -> (ReturnValue:Text)
   )
   Return Node(execute=@n2.PreparingDownload, Text="LOCTEXT(\"Preparing Island\")")
   Return Node(execute=@n2.QueuedDownload, Text="LOCTEXT(\"Queued Download\")")
  ]}
 Set Map Rotation Date {nodes [
   Set Map Rotation Date!() -> (then, Date Time:Date Time Structure) {is=title K2Node_FunctionEntry}
   > n1 = Branch(Condition=Get isMatchFoundAnimationActive() -> (isMatchFoundAnimationActive:bool))
   > SetVisibility(self=Get MapRotationCountdown(), InVisibility=Collapsed)
   n2 = Get MapRotationCountdownTimerVM() -> (MapRotationCountdownTimerVM:Countdown Timer View Model Object Reference)
   ResetTimer(execute=@n1.else, self:Countdown Timer View Model Object Reference=@n2)
   > ClearCountdownData(self:Countdown Timer View Model Object Reference=@n2)
   n3 = Get FortMatchmakingViewModel()
   > n4 = Is Valid(InputObject=@n3)
   n5 = Get TournamentTimerVM() -> (TournamentTimerVM:Countdown Timer View Model Object Reference)
   ResetTimer(execute=[@n8.then, @n4.Is Not Valid], self=@n5)
   > n6 = ClearCountdownData(self=@n5)
   n7 = Branch(
    execute=[@n6.then, @n8.else]
    Condition="Not Equal (DateTime)"!(A:Date Time Structure=Get Date Time() -> (Date Time:Date Time Structure), B:Date Time Structure) -> (ReturnValue:bool)
   )
   > Start Rotation Timer!(execute, self, Date Time:Date Time Structure=Get Date Time()) -> (then) {is=title K2Node_CallFunction}
   SetVisibility(execute=@n7.else, self=Get MapRotationCountdown(), InVisibility=Collapsed)
   n8 = Branch(execute=@n4.Is Valid, Condition=NOT Boolean(A=Is The Island A Tournament Island(self=@n3)))
  ]}
 Set Match Found Island Widget {nodes [
   n1 = Set Match Found Island Widget!() -> (then, MatchFoundClass:User Widget Class Reference) {is=title K2Node_FunctionEntry}
   > n2 = Branch(Condition=IsValidClass!(Class:Object Class Reference=@n1) -> (ReturnValue:bool))
   > n3 = Set Match Found Class(Match Found Class:User Widget Class Reference=@n1) -> (Output_Get:User Widget Class Reference)
   n4 = Create Widget!(
    execute:Exec=[@n6.then, @n3.then]
    Class:User Widget Class Reference=Get Match Found Class() -> (Match Found Class:User Widget Class Reference)
    OwningPlayer:Player Controller Object Reference=GetOwningPlayer()
   ) -> (then:Exec, ReturnValue:User Widget Object Reference)
   > n5 = Set MatchFound Widget(MatchFound Widget:User Widget Object Reference=@n4) -> (Output_Get:User Widget Object Reference)
   > SetContent!(
    execute
    self:Content Widget Object Reference=Get NamedSlot_MatchFoundWidget() -> (NamedSlot_MatchFoundWidget:Named Slot Object Reference)
    Content:Widget Object Reference=@n5
   ) -> (then, ReturnValue:Panel Slot Object Reference)
   > RegisterTravelAnimationForMatchmakingFlow(self:Selected Island View Model Object Reference=Get FortMatchmakingViewModel())
   n6 = Set Match Found Class(execute=@n2.else, Match Found Class=&/WBP_UI_Lobby_MatchFound_Default.WBP_UI_Lobby_MatchFound_Default_C)
  ]}
 Set Matchmaking State {nodes [
   n1 = Set Matchmaking State!() -> (then, NewMatchmakingState:EFortUIMatchmakingStatus Enum=Unready) {is=title K2Node_FunctionEntry}
   > n2 = Branch(Condition=Get isMatchFoundAnimationActive())
   SetInvalidLoadingState!(execute=@n2.else, self, MatchmakingState:EFortUIMatchmakingStatus Enum=@n1) -> (then)
   > Set Current Matchmaking Status(Current Matchmaking Status:EFortUIMatchmakingStatus Enum=@n1) -> (Output_Get:EFortUIMatchmakingStatus Enum=InvalidIsland)
   > SetRenderOpacity(self=Get Content_Main(), InOpacity=1)
   > n3 = Switch on EFortUIMatchmakingStatus(Selection=Get NewMatchmakingState() -> (NewMatchmakingState:EFortUIMatchmakingStatus Enum=InvalidIsland))
   n4 = Branch(execute=[@n3.FindingServer, @n3.WaitingForPlayersReadyUp], Condition=Get IsStatusExpanded() -> (IsStatusExpanded:bool))
   > n5 = Animate Matchmaking Status() {is=title K2Node_CallFunction}
   Set IsStatusExpanded(execute=[@n6.then, @n5.then], IsStatusExpanded:bool=true) -> (Output_Get:bool)
   > ClearFlushAudio() {is=title K2Node_CallFunction}
   > RegisterTravelAnimationForMatchmakingFlow(self=Get FortMatchmakingViewModel())
   PlayAnimation(execute=@n4.else, InAnimation=Get Anim_StatusIntro() -> (Anim_StatusIntro:Widget Animation Object Reference))
   > n6 = "Update Matchmaking State Appearance "(NewMatchmakingState=Get Current Matchmaking Status() -> (Current Matchmaking Status=InvalidIsland))
   ClearFlushAudio(execute=@n3.InvalidIsland)
   > Set FlushClicked(FlushClicked=false)
   n7 = Branch(execute=@n3.MatchFound, Condition=Get IsStatusExpanded())
   > n8 = Animate Matchmaking Status()
   n9 = Cast To WBP_UI_Lobby_MatchFound_Base!(execute=[@n11.then, @n8.then], Object=Get MatchFound Widget() -> (MatchFound Widget:User Widget Object Reference)) -> (then, CastFailed, AsWBP UI Lobby Match Found Base:WBP UI Lobby Match Found Base Object Reference) {is=title K2Node_DynamicCast}
   > Play MatchFound Animation(self:WBP UI Lobby Match Found Base Object Reference=@n9)
   > Set IsStatusExpanded(IsStatusExpanded=false)
   n10 = Get TravelAnimationDuration!(self:Selected Island View Model Object Reference=Get FortMatchmakingViewModel()) -> (TravelAnimationDuration:float)
   > PlayMatchmakingAudio!(execute, self, bFlushed:bool=false, Time:float=@n10) -> (then) {is=title K2Node_CallFunction}
   > Branch(Condition=IsVisible(self=Get MapRotationCountdown()))
   > Set isMatchFoundAnimationActive(isMatchFoundAnimationActive:bool=true) -> (Output_Get:bool)
   PlayMatchmakingAudio(execute=@n9.CastFailed, bFlushed=true, Time=@n10)
   n11 = "Update Matchmaking State Appearance "(execute=@n7.else, NewMatchmakingState=Get NewMatchmakingState() -> (NewMatchmakingState=InvalidIsland))
   n12 = Get GroupError(self=Get FortMatchmakingViewModel())
   n13 = Branch(execute=@n3.Unready, Condition=IsValid(Object=@n12))
   > SetError!(execute, self, ErrorUI:Ready Up Error UI Object Reference=@n12) -> (then)
   > n14 = ClearFlushAudio()
   "Update Matchmaking State Appearance "(execute=[@n14.then, @n13.else], NewMatchmakingState=Get Current Matchmaking Status() -> (Current Matchmaking Status=InvalidIsland))
   > SetIsCancelButtonAnim!(execute, self:WBP UI Lobby Button Matchmaking Object Reference=Get WBP_UI_Button_Matchmaking(), IsCancel:bool=false) -> (then)
   > Set IsStatusExpanded(IsStatusExpanded=false)
   > Set FlushClicked(FlushClicked=false)
   n15 = Get WBP_UI_Button_Flush()
   > Branch(Condition=IsVisible(self=@n15))
   > SetVisibility(self=@n15, InVisibility=Collapsed)
   > Set CanStartMatchShouldBeVisible(CanStartMatchShouldBeVisible=false)
  ]}
 Set Status Icon {nodes [
   n1 = Set Status Icon!() -> (
    then
    MM Status:EFortUIMatchmakingStatus Enum=InvalidIsland
    Download Info:Matchmaking Download Info Structure
    Current Player Count:int
    Total Player Count:int
   ) {is=title K2Node_FunctionEntry}
   > n2 = Switch on EFortUIDownloadType(Selection=Get() -> (Download Info_bIsDownloading=~, Download Info_Progress=~).Download Info_DownloadType)
   > n3 = Switch on EFortUIMatchmakingStatus(Selection=@n1)
   Return Node!(
    execute=@n3.FindingServer
    Widget:Widget Object Reference=Get WBP_UIKit_Throbber() -> (WBP_UIKit_Throbber:WBP UIKit Throbber Object Reference)
   )
   n4 = Get Current Player Count() -> (Current Player Count:int=0)
   n5 = Get Total Player Count() -> (Total Player Count:int=0)
   Return Node(
    execute=@n3.WaitingForPlayersReadyUp
    Widget=Select!(
     Option 0:Widget Object Reference=Get WBP_UIKit_Throbber()
     Option 1:Widget Object Reference=Get WBP_UIKit_StatusIndicator() -> (WBP_UIKit_StatusIndicator:WBP UIKit Status Indicator Object Reference)
     Index:bool="Equal (Integer)"(A=@n4, B=@n5)
    ) -> (ReturnValue:Widget Object Reference)
   )
   Return Node(execute=@n2.PreparingDownload, Widget=Get WBP_UIKit_Throbber())
   n6 = Get WBP_UI_DownloadProgress()
   n7 = SetFeedbackVisibility!(execute=@n2.DownloadWithProgress, self:WBP UI Download Progress Object Reference=@n6, IsVisible:bool=true) -> (then)
   Return Node(execute=[@n8.then, @n7.then], Widget=@n6)
   n8 = SetFeedbackVisibility(execute=@n2.QueuedDownload, self=@n6, IsVisible=false)
  ]}
 SetCompactUIStyle {nodes [
   SetCompactUIStyle() {is=title K2Node_FunctionEntry}
   > Branch(Condition=IsUsingCompactUI())
   > n1 = Sequence!(execute) -> (then_0, then_1, then_2:Exec)
   n2 = Get WBP_UI_Button_ModeSelection()
   n3 = Get!(self:WBP UIKit Button Generic Base Object Reference=@n2) -> (IconPadding_Left:float=~, IconPadding_Top:float=~, IconPadding_Right:float=~, IconPadding_Bottom:float=~)
   > SetIconPadding!(
    execute
    self:WBP UIKit Button Generic Base Object Reference=@n2
    Padding_Left:float=@n3
    Padding_Top:float=@n3.IconPadding_Top
    Padding_Right:float=8
    Padding_Bottom:float=@n3.IconPadding_Bottom
   ) -> (then)
   > SetButtonPadding!(
    execute
    self:WBP UIKit Button Generic Base Object Reference=@n2
    Padding_Left:float=16
    Padding_Top:float=0
    Padding_Right:float=16
    Padding_Bottom:float=0
   ) -> (then)
   > SetTextPadding!(
    execute
    self:WBP UIKit Button Generic Base Object Reference=@n2
    TextPadding_Left:float=0
    TextPadding_Top:float=8
    TextPadding_Right:float=0
    TextPadding_Bottom:float=8
   ) -> (then)
   SetBadgeSize!(execute=@n1.then_2, self:WBP UIKit Badge Base Object Reference=Get MapRotationCountdown(), BadgeSize:E_UI_BadgeSize Enum=NewEnumerator1) -> (then)
   n4 = SlotAsGridSlot(Widget=Get Grid_RHighestScore() -> (Grid_RHighestScore:Grid Panel Object Reference))
   n5 = Get!(self:Grid Slot Object Reference=@n4) -> (Padding_Left:float=~, Padding_Top:float=~, Padding_Right:float=~, Padding_Bottom:float=~)
   SetPadding(
    execute=@n1.then_1
    self:Grid Slot Object Reference=@n4
    InPadding_Left=@n5
    InPadding_Top=8
    InPadding_Right=@n5.Padding_Right
    InPadding_Bottom=@n5.Padding_Bottom
   )
   n6 = SlotAsVerticalBoxSlot(Widget=Get WBP_UI_Lobby_ErrorState() -> (WBP_UI_Lobby_ErrorState:WBP UI Lobby Error State Object Reference))
   n7 = Get(self:Vertical Box Slot Object Reference=@n6) -> (Padding_Left=~, Padding_Top=~, Padding_Right=~, Padding_Bottom=~)
   > SetPadding(
    self:Vertical Box Slot Object Reference=@n6
    InPadding_Left=@n7
    InPadding_Top=8
    InPadding_Right=@n7.Padding_Right
    InPadding_Bottom=@n7.Padding_Bottom
   )
  ]}
 SetDebugEventText {nodes [
   n1 = SetDebugEventText!() -> (then, DebugEventText:Text) {is=title K2Node_FunctionEntry}
   > "SetText (Text)"(self=Get Text_DebugId() -> (Text_DebugId:Text Object Reference), InText=@n1)
   > SetVisibility(
    self=Get Text_DebugId()
    InVisibility=Select(Option 0:ESlateVisibility Enum=SelfHitTestInvisible, Option 1:ESlateVisibility Enum=Collapsed, Index=TextIsEmpty(InText=@n1)) -> (ReturnValue:ESlateVisibility Enum)
   )
  ]}
 SetDebugLeaderboardText {nodes [
   n1 = SetDebugLeaderboardText!() -> (then, DebugLeaderboardText:Text) {is=title K2Node_FunctionEntry}
   n2 = Get Text_DebugLBId() -> (Text_DebugLBId:Text Object Reference)
   > "SetText (Text)"(self=@n2, InText=@n1)
   > SetVisibility(self=@n2, InVisibility=Select(Option 0=SelfHitTestInvisible, Option 1=Collapsed, Index=TextIsEmpty(InText=@n1)))
  ]}
 SetError {nodes [
   SetError!() -> (then, ErrorUI:Ready Up Error UI Object Reference) {is=title K2Node_FunctionEntry}
   > CheckPlayerSittingOutTooltipVisibility() {is=title K2Node_CallFunction}
   > UpdateErrorStateWidget!(execute, self, ErrorUI:Ready Up Error UI Object Reference=Get ErrorUI()) -> (then)
   n1 = Get ErrorUI()
   > n2 = Is Valid(InputObject=@n1)
   n3 = Get WBP_UI_Button_Primary()
   > n4 = Set Text(Text=Play Button Text!(self:Ready Up Error UI Object Reference=@n1) -> (ReturnValue:Text), self=@n3)
   n5 = Len(S="To String (Text)"(InText=@n4))
   > SetWrapTextWidth!(
    execute
    self:Common Text Object Reference=Get TextBlock!(self:WBP UIKit Block Text Object Reference=Get Block_Text!(self:WBP UIKit Button CTA Base Object Reference=@n3) -> (Block_Text:WBP UIKit Block Text Object Reference)) -> (TextBlock:Common Text Object Reference)
    InWrapTextAt:int=SelectInt(A=0, B=260, bPickA="integer >= integer"(A=@n5, B=24))
   ) -> (then)
   > HandleSecondaryButtonTextChanged!(execute, self, SecondaryText:Text=Secondary Button Text!(self:Ready Up Error UI Object Reference=@n1) -> (ReturnValue:Text)) -> (then) {is=title K2Node_CallFunction}
   HandleSecondaryButtonTextChanged(execute=@n2.Is Not Valid)
  ]}
 SetInvalidLoadingState {nodes [
   n1 = SetInvalidLoadingState!() -> (then, MatchmakingState:EFortUIMatchmakingStatus Enum=InvalidIsland) {is=title K2Node_FunctionEntry}
   n2 = "Equal (Enum)"(A:EFortUIMatchmakingStatus Enum=@n1, B:EFortUIMatchmakingStatus Enum=InvalidIsland)
   > SetScalarParameterValue!(
    execute
    Collection:Material Parameter Collection Object Reference=Get MatchmakingCollectionParameter() -> (MatchmakingCollectionParameter:Material Parameter Collection Object Reference)
    ParameterName:Name=Get LoadingParameterName() -> (LoadingParameterName:Name=None)
    ParameterValue:float=Select(Option 0:float=0, Option 1:float=1, Index=@n2) -> (ReturnValue:float)
   ) -> (then) {is=title K2Node_CallMaterialParameterCollectionFunction}
   > Set ModeSelect Icon Visibility!(
    execute
    self:WBP UI Lobby Button Mode Selection Object Reference=Get WBP_UI_Button_ModeSelection()
    Visible:bool=Select(Option 0:bool=true, Option 1:bool=false, Index=@n2) -> (ReturnValue:bool)
   ) -> (then) {is=title K2Node_CallFunction}
  ]}
 SetIsLive {nodes [
   n1 = SetIsLive!() -> (then, IsLive:bool=~) {is=title K2Node_FunctionEntry}
   > SetActiveWidgetIndex(
    self:Common Visibility Switcher Object Reference=Get Switcher_Countdown()
    Index="To Integer (Boolean)"!(InBool:bool=@n1) -> (ReturnValue:int=0)
   )
  ]}
 SetIslandLogo {nodes [
   n1 = SetIslandLogo!() -> (then, LogoClass:User Widget Class Reference) {is=title K2Node_FunctionEntry}
   > n2 = Branch(Condition=IsValidClass(Class=@n1))
   > n3 = Create Widget(Class=@n1, OwningPlayer=GetOwningPlayer())
   > SetContent(self=Get NamedSlot_CustomLogoHeader() -> (NamedSlot_CustomLogoHeader:Named Slot Object Reference), Content=@n3)
   > SetActiveWidget(
    self=Get Switcher_GameIslandInfo() -> (Switcher_GameIslandInfo:Common Visibility Switcher Object Reference)
    Widget=Get NamedSlot_CustomLogoHeader()
   )
   SetActiveWidget(execute=@n2.else, self=Get Switcher_GameIslandInfo(), Widget=Get GameMode_SizeBox() -> (GameMode_SizeBox:Size Box Object Reference))
   > ClearChildren(self=Get NamedSlot_CustomLogoHeader())
  ]}
 SetIslandTitle {nodes [
   n1 = SetIslandTitle!() -> (then, TitleText:Text) {is=title K2Node_FunctionEntry}
   > Set Text(Text=@n1, self:Text Object Reference=Get GameIsland_Title() -> (GameIsland_Title:Common Text Object Reference))
  ]}
 SetMapName {nodes [
   n1 = SetMapName!() -> (then, MapName:Text) {is=title K2Node_FunctionEntry}
   n2 = TextIsEmpty(InText=@n1)
   > SetVisibility(
    self=Get Txt_MapName() -> (Txt_MapName:Common Text Object Reference)
    InVisibility=Select(Option 0:ESlateVisibility Enum=SelfHitTestInvisible, Option 1:ESlateVisibility Enum=Collapsed, Index=@n2) -> (ReturnValue:ESlateVisibility Enum)
   )
   > Set Text(Text=@n1, self=Get Txt_MapName())
   > n3 = Branch(Condition=Get bIsTournamentLive(self=Get FortMatchmakingViewModel()))
   > SetMaxDesiredHeight(self=Get GameMode_SizeBox(), InMaxDesiredHeight=Select(Option 0:float=80, Option 1:float=110, Index=@n2) -> (ReturnValue:float))
   ClearMaxDesiredHeight(execute=@n3.else, self:Size Box Object Reference=Get GameMode_SizeBox())
  ]}
 SetMapRotationEntry {nodes [
   n1 = SetMapRotationEntry!() -> (then, MapTexture:Texture Object Reference) {is=title K2Node_FunctionEntry}
   > Branch(Condition=IsValid(Object=@n1))
   > n2 = Set CurrentMapTexture(CurrentMapTexture:Texture Object Reference=@n1) -> (Output_Get:Texture Object Reference)
   > SetTextureParameterValue(self=Get ThumbnailDynamicMaterial(), ParameterName=TextureA, Value=@n2)
   > SetTextureParameterValue(self=Get ThumbnailDynamicMaterial(), ParameterName=TextureB, Value=@n2)
   > OnMapRotation() {is=title K2Node_CallFunction}
  ]}
 SetMatchMakingVM {nodes [
   n1 = SetMatchMakingVM!() -> (then, FortMatchmakingViewModel:Selected Island View Model Object Reference) {is=title K2Node_FunctionEntry}
   > Set FortMatchmakingViewModel(FortMatchmakingViewModel:Selected Island View Model Object Reference=@n1) -> (Output_Get:Selected Island View Model Object Reference)
   > Set Matchmaking State!(
    execute
    self
    NewMatchmakingState:EFortUIMatchmakingStatus Enum=Get MatchmakingStatus!(self:Selected Island View Model Object Reference=@n1) -> (MatchmakingStatus=InvalidIsland)
   ) -> (then) {is=title K2Node_CallFunction}
  ]}
 SetMemoryFormat {nodes [
   n1 = MakeLiteralInt64!(Value:Integer64=1000000) -> (ReturnValue:Integer64=0)
   n2 = Inputs!() -> (NumBytes:Integer64)
   n3 = "integer64 > integer64"(A=@n2, B=@n1) -> (ReturnValue=~)
   Outputs!(FormattedMemory:Text=AsMemory!(
     NumBytes:Integer64=Select(Option 0:Integer64=@n1, Option 1:Integer64=@n2, Index=@n3) -> (ReturnValue:Integer64)
     UnitStandard:EMemoryUnitStandard Enum=SI
     bUseGrouping:bool=false
     MinimumIntegralDigits:int=1
     MaximumIntegralDigits:int=324
     MinimumFractionalDigits:int=0
     MaximumFractionalDigits:int=SelectInt(A=2, B=0, bPickA=@n3)
    ) -> (ReturnValue:Text))
  ]}
 SetModeSelectionButtonTextStyle {nodes [
   SetModeSelectionButtonTextStyle() {is=title K2Node_FunctionEntry}
   n1 = Get WBP_UI_Button_ModeSelection()
   > SetMinDimensions!(execute, self:Common Button Base Object Reference=@n1, InMinWidth:int=0, InMinHeight:int=75) -> (then)
   > SetMaxDimensions!(execute, self:Common Button Base Object Reference=@n1, InMaxWidth:int=0, InMaxHeight:int=SelectInt(A=100, B=90, bPickA=IsUsingCompactUI())) -> (then)
   > SetClipping!(
    execute
    self:Widget Object Reference=Get Main_StackSize!(self:WBP UIKit Button Generic Base Object Reference=@n1) -> (Main_StackSize:Size Box Object Reference)
    InClipping:EWidgetClipping Enum=ClipToBounds
   ) -> (then)
   n2 = Get Block_Text(self:WBP UIKit Button Generic Base Object Reference=@n1)
   > SetWrapping!(execute, self:WBP UIKit Block Text Object Reference=@n2, IsAutoWrap:bool=true, WrapTextAt:int=0) -> (then)
   > SetTextOverflowPolicy!(execute, self:WBP UIKit Block Text Object Reference=@n2, OverflowPolicy:ETextOverflowPolicy Enum=MultilineEllipsis) -> (then)
  ]}
 SetModeSpecificWidgetVisibility {nodes [
   SetModeSpecificWidgetVisibility!() -> (then, isVisible:bool=~) {is=title K2Node_FunctionEntry}
   > n1 = Is Valid(InputObject=Get ModeSpecificWidget())
   > SetVisibility(
    self=Get ModeSpecificWidget()
    InVisibility=Select(Option 0:ESlateVisibility Enum=Collapsed, Option 1:ESlateVisibility Enum=Visible, Index=Get isVisible() -> (isVisible:bool)) -> (ReturnValue:ESlateVisibility Enum)
   )
   AddWidgetAlias_ModeSpecificWidget!(execute=@n1.Is Not Valid, self, IsVisible:bool=Get isVisible()) -> (then) {is=title K2Node_CallFunction}
  ]}
 SetNextEventTIme {nodes [
   SetNextEventTIme!() -> (then, DateTime:Date Time Structure) {is=title K2Node_FunctionEntry}
   > n1 = Is Valid(InputObject=Get TournamentTimerVM())
   n2 = Construct Object from Class(execute=@n1.Is Not Valid, Class=/Script/FortniteUI.FortCountdownTimerVM) -> (ReturnValue:Countdown Timer View Model Object Reference)
   > n3 = SetTournamentTimerVM!(execute, self, Viewmodel:Countdown Timer View Model Object Reference=@n2) -> (then)
   n4 = Get MapRotationCountdownTimerVM()
   ResetTimer(execute=[@n3.then, @n1.Is Valid], self=@n4)
   > ClearCountdownData(self=@n4)
   n5 = Get TournamentTimerVM()
   > ResetTimer(self=@n5)
   > ClearCountdownData(self=@n5)
   > n6 = Branch(Condition="Not Equal (DateTime)"(A=Get DateTime() -> (DateTime:Date Time Structure)))
   > Start TournamentTimer!(execute, self, Date Time:Date Time Structure=Get DateTime()) -> (then) {is=title K2Node_CallFunction}
   SetVisibility(execute=@n6.else, self=[Get MapRotationCountdown(), Get Grid_LiveRankedCountdown()], InVisibility=Collapsed)
  ]}
 SetPlay Button Enabled {nodes [
   SetPlay Button Enabled!() -> (then, IsPlayButtonLocked:bool=~) {is=title K2Node_FunctionEntry}
   > SetIsLocked!(
    execute
    self:Common Button Base Object Reference=Get WBP_UI_Button_Matchmaking()
    bInIsLocked:bool=Get IsPlayButtonLocked() -> (IsPlayButtonLocked:bool)
   ) -> (then)
  ]}
 SetPlayButtonAndBottomBar {nodes [
   n1 = SetPlayButtonAndBottomBar!() -> (
    then
    MatchmakingStatus:EFortUIMatchmakingStatus Enum=Unready
    Error UI:Ready Up Error UI Object Reference
    ReadyUp Status:Map of Unique Net Id Structures to EFortUIReadyUpStatus Enums
   ) {is=title K2Node_FunctionEntry}
   > n2 = Branch(Condition=IsValid(Object=@n1.Error UI))
   > n3 = Cast To BaseDownloadErrorUI!(execute, Object=Get Error UI() -> (Error UI:Ready Up Error UI Object Reference)) -> (then, CastFailed, AsBase Download Error UI:Base Download Error UI Object Reference) {is=title K2Node_DynamicCast}
   > BindToDownloadUpdated!(execute, self:Base Download Error UI Object Reference=@n3, OnDownloadInfo:"Delegate (by ref)"=Create Event()) -> (then)
   > n4 = Branch(Condition="Equal (Enum)"(
     A:EFortUIDownloadType Enum=Get Download Info Download Type() -> (Download Info Download Type:EFortUIDownloadType Enum=NoDownload)
     B:EFortUIDownloadType Enum=NoDownload
    ))
   > SetIsCancelButtonAnim(self=Get WBP_UI_Button_Matchmaking(), IsCancel=false)
   n5 = Play Button Text(self=Get Error UI())
   > Branch(Condition=NOT Boolean(A=TextIsEmpty(InText=@n5)))
   > Return Node!(execute, PlayButtonText:Text=Play Button Text(self=Get Error UI()))
   SetIsCancelButtonAnim(execute=[@n4.else, @n12.then], self=Get WBP_UI_Button_Matchmaking(), IsCancel=true)
   > n6 = Branch(Condition=ShouldBindCancelButtonToBottomBar!(self:Selected Island View Model Object Reference=Get FortMatchmakingViewModel()) -> (ReturnValue:bool))
   > n7 = Set with Broadcast Bottom Bar Button State!(
    execute
    Bottom Bar Button State:Data Table Row Handle Structure=Get Bottom Bar Button State Cancel() -> (Bottom Bar Button State Cancel:Data Table Row Handle Structure)
   ) -> (then, Output_Get:Data Table Row Handle Structure)
   Return Node(execute=[@n7.then, @n8.then], PlayButtonText="LOCTEXT(\"Cancel\")")
   n8 = Set with Broadcast Bottom Bar Button State(execute=@n6.else)
   SetIsCancelButtonAnim(execute=@n3.CastFailed, self=Get WBP_UI_Button_Matchmaking(), IsCancel=false)
   n9 = Play Button Text(self=Get Error UI())
   > n10 = Branch(Condition=NOT Boolean(A=TextIsEmpty(InText=@n9)))
   > Set with Broadcast Bottom Bar Button State(Bottom Bar Button State=Get Bottom Bar Button Redirect() -> (Bottom Bar Button Redirect:Data Table Row Handle Structure))
   > Return Node(PlayButtonText=Play Button Text(self=Get Error UI()))
   n11 = Get MatchmakingStatus!() -> (MatchmakingStatus=InvalidIsland)
   n12 = Branch(
    execute=[@n15.then, @n10.else]
    Condition=AND Boolean(A="Not Equal (Enum)"(A=@n11, B=Unready), B="Not Equal (Enum)"(A=@n11, B=InvalidIsland))
   )
   n13 = Find(
    TargetMap=Get ReadyUp Status() -> (ReadyUp Status:Map of Unique Net Id Structures to EFortUIReadyUpStatus Enums)
    Key=GetNetIDFromContextObject(ContextObject=GetOwningPlayer())
   ) -> (Value=Invalid)
   Set with Broadcast Bottom Bar Button State(
    execute=@n12.else
    Bottom Bar Button State=Find(
     TargetMap:Map of EFortUIReadyUpStatus Enums to Data Table Row Handle Structures=Get Bottom Bar Button Input() -> (Bottom Bar Button Input:Map of EFortUIReadyUpStatus Enums to Data Table Row Handle Structures=Invalid)
     Key:"EFortUIReadyUpStatus Enum (by ref)"=@n13
    ) -> (Value:Data Table Row Handle Structure)
   )
   > n14 = Switch on EFortUIReadyUpStatus!(execute, Selection:EFortUIReadyUpStatus Enum=@n13) -> (Invalid:Exec, NotPlaying:Exec, Play:Exec, Ready:Exec, Join:Exec, Rejoin:Exec, Watch:Exec) {is=title K2Node_SwitchEnum}
   Return Node(execute=@n14.Join, PlayButtonText="LOCTEXT(\"Join\")")
   Return Node(execute=@n14.Play, PlayButtonText="LOCTEXT(\"Play\")")
   Return Node(execute=[@n14.Ready, @n14.NotPlaying], PlayButtonText="LOCTEXT(\"Ready\")")
   Return Node(execute=@n14.Rejoin, PlayButtonText="LOCTEXT(\"Rejoin\")")
   Return Node(execute=@n14.Watch, PlayButtonText="LOCTEXT(\"Watch\")")
   n15 = SetIsCancelButtonAnim(execute=@n2.else, self=Get WBP_UI_Button_Matchmaking(), IsCancel=false)
  ]}
 SetPlaytimeOut {nodes [
   n1 = SetPlaytimeOut!() -> (then, IsPlaytimeOut:bool=~) {is=title K2Node_FunctionEntry}
   > n2 = Set IsPlaytimeOut(IsPlaytimeOut:bool=@n1) -> (Output_Get:bool)
   > SetIsEnabled(
    self=[Get WBP_UI_Button_Secondary(), Get WBP_UI_Button_ModeSelection(), Get ModeSpecificWidget(), Get VerticalBox_ExtensionHolder()]
    bInIsEnabled=Select(Option 0:bool=true, Option 1:bool=false, Index=@n2) -> (ReturnValue:bool)
   )
  ]}
 SetTournamentHighestScoreValue {nodes [
   n1 = SetTournamentHighestScoreValue!() -> (then, TournamentHighestScoreValue:int) {is=title K2Node_FunctionEntry}
   > "SetText (Text)"(self=Get Text_RHighestScoreValue() -> (Text_RHighestScoreValue:Common Text Object Reference), InText="To Text (Integer)"(Value=@n1))
   > ShowTournamentHighestScoreStats!(execute, self, bShowTournamentHighestScoreStats:bool="integer >= integer"(A=@n1)) -> (then)
  ]}
 SetTournamentMatchesAmount {nodes [
   n1 = SetTournamentMatchesAmount!() -> (then, Tournament Match Data_MatchesPlayed:int, Tournament Match Data_MatchCap:int) {is=title K2Node_FunctionEntry}
   n2 = "integer > integer"(A=@n1.Tournament Match Data_MatchCap)
   > ShowTournamentMatchesAmount!(execute, self, bShowTounramentMatchesAmount:bool=@n2) -> (then)
   > Branch(Condition=@n2)
   n3 = Get!() -> (Tournament Match Data_MatchesPlayed:int, Tournament Match Data_MatchCap:int)
   n4 = Format Text!(Format="LOCTEXT(\"{All Matches} Matches Played\")", All Matches:int=@n3.Tournament Match Data_MatchCap) -> (Result)
   n5 = "Equal (Integer)"(A=@n3.Tournament Match Data_MatchCap, B=1)
   n6 = SelectInt(A=1, B=int + int(A=@n3), bPickA="integer <= integer"(A=@n3))
   > "SetText (Text)"(
    self=Get Text_RMatchesAmount() -> (Text_RMatchesAmount:Common Text Object Reference)
    InText=SelectText(
     A=SelectText(A="LOCTEXT(\"1 Match Played\")", B=@n4, bPickA=@n5)
     B=Format Text!(Format="LOCTEXT(\"Match {Current Match} of {All Matches}\")", Current Match:int=@n6, All Matches=@n3.Tournament Match Data_MatchCap) -> (Result)
     bPickA="Equal (Integer)"(A=@n3, B=@n3.Tournament Match Data_MatchCap)
    )
   )
  ]}
 SetTournamentScoreValue {nodes [
   n1 = SetTournamentScoreValue!() -> (then, TournamentScoreValue:int) {is=title K2Node_FunctionEntry}
   > "SetText (Text)"(self=Get Text_RPointsValue() -> (Text_RPointsValue:Common Text Object Reference), InText="To Text (Integer)"(Value=@n1))
   > ShowTournamentPointsStats!(execute, self, bShowTournamentPointsStats:bool="integer >= integer"(A=@n1)) -> (then)
  ]}
 Setup Map Rotation Countdown Timer VM {nodes [
   n1 = Setup Map Rotation Countdown Timer VM!() -> (then, DataTime:Date Time Structure) {is=title K2Node_FunctionEntry}
   n2 = Get MapRotationCountdownTimerVM()
   > SetupViewModel!(
    execute
    self:Countdown Timer View Model Object Reference=@n2
    InFutureTime:"Date Time Structure (by ref)"=@n1
    InLiveText:"Text (by ref)"=GetEmptyText!() -> (ReturnValue:Text)
   ) -> (then)
   > SetTimerUnitDisplayType!(execute, self:Countdown Timer View Model Object Reference=@n2, InTimerUnitDisplayType:ETimespanUnitDisplayFormat Enum=Abbreviated) -> (then)
   > SetApproximateLongTimers!(execute, self:Countdown Timer View Model Object Reference=@n2, bInApproximateLongTimers:bool=false) -> (then)
   > StartTimer(self:Countdown Timer View Model Object Reference=@n2)
  ]}
 Setup Tournament Countdown Timer VM {nodes [
   n1 = Setup Tournament Countdown Timer VM!() -> (then, DataTime:Date Time Structure) {is=title K2Node_FunctionEntry}
   n2 = Get TournamentTimerVM()
   > SetupViewModel(self=@n2, InFutureTime=@n1, InLiveText=GetEmptyText())
   > SetTimerUnitDisplayType(self=@n2, InTimerUnitDisplayType=Abbreviated)
   > SetApproximateLongTimers(self=@n2, bInApproximateLongTimers=true)
   > SetShouldAlwaysDisplaySeconds!(execute, self:Countdown Timer View Model Object Reference=@n2, bInShouldAlwaysDisplaySeconds:bool=true) -> (then)
   > StartTimer(self=@n2)
  ]}
 ShowTournamentHighestScoreStats {nodes [
   n1 = ShowTournamentHighestScoreStats!() -> (then, bShowTournamentHighestScoreStats:bool=~) {is=title K2Node_FunctionEntry}
   > SetVisibility(
    self=[Get Text_RHighestScore() -> (Text_RHighestScore:Common Text Object Reference), Get Text_RHighestScoreValue()]
    InVisibility=Select(Option 0:ESlateVisibility Enum=Collapsed, Option 1:ESlateVisibility Enum=SelfHitTestInvisible, Index=@n1) -> (ReturnValue:ESlateVisibility Enum)
   )
  ]}
 ShowTournamentMatchesAmount {nodes [
   n1 = ShowTournamentMatchesAmount!() -> (then, bShowTounramentMatchesAmount:bool=~) {is=title K2Node_FunctionEntry}
   > SetVisibility(self=Get Text_RMatchesAmount(), InVisibility=Select(Option 0=Collapsed, Option 1=SelfHitTestInvisible, Index=@n1))
  ]}
 ShowTournamentPointsStats {nodes [
   n1 = ShowTournamentPointsStats!() -> (then, bShowTournamentPointsStats:bool=~) {is=title K2Node_FunctionEntry}
   > SetVisibility(
    self=[Get Text_RPointsValue(), Get Text_RPoints() -> (Text_RPoints:Common Text Object Reference)]
    InVisibility=Select(Option 0=Collapsed, Option 1=SelfHitTestInvisible, Index=@n1)
   )
  ]}
 ShowTournamentStats {nodes [
   n1 = ShowTournamentStats!() -> (then, bShowTournamentStats:bool=~) {is=title K2Node_FunctionEntry}
   n2 = NOT Boolean(A=Get IsPlaytimeOut())
   > SetVisibility(
    self=Get Grid_RankStats() -> (Grid_RankStats:Grid Panel Object Reference)
    InVisibility=Select(Option 0=Collapsed, Option 1=SelfHitTestInvisible, Index=AND Boolean(A=@n1, B=@n2))
   )
  ]}
 TellLobbyToUpdateBottomButtonVisibility {nodes [
   n1 = TellLobbyToUpdateBottomButtonVisibility!() -> (then, IsBottomBarButtonVisible:bool=~) {is=title K2Node_FunctionEntry}
   > n2 = FindParentWidgetOfType(StartingWidget=Self-Reference(), Type=/Script/FortniteUI.AthenaLobbyBase) -> (ReturnValue:Athena Lobby Base Object Reference)
   > n3 = Cast To AthenaLobby!(execute, Object=@n2) -> (then, CastFailed, AsAthena Lobby:Athena Lobby Object Reference) {is=title K2Node_DynamicCast}
   > UpdateBottomBarPlayButtonVisibility!(execute, self:Athena Lobby Object Reference=@n3, IsVisible:bool=@n1) -> (then)
  ]}
 Update Download State Visuals {nodes [
   n1 = Update Download State Visuals!() -> (then, Download Info:Matchmaking Download Info Structure) {is=title K2Node_FunctionEntry}
   > n2 = Set with Broadcast Download Info(Download Info=@n1) -> (Output_Get_bIsDownloading=~, Output_Get_Progress=~)
   > n3 = Set Download Info Download Type(Download Info Download Type:EFortUIDownloadType Enum=@n2.Output_Get_DownloadType) -> (Output_Get:EFortUIDownloadType Enum=NoDownload)
   > n4 = Branch(Condition=@n2)
   SetVisibility(
    execute=[@n4.then, @n7.DownloadWithProgress, @n7.PreparingDownload, @n7.QueuedDownload]
    self=Get Text_MMData()
    InVisibility=HitTestInvisible
   )
   > SetVisibility(self=Get Grid_StatusInfo() -> (Grid_StatusInfo:Grid Panel Object Reference), InVisibility=HitTestInvisible)
   > SetIsCancelButtonAnim(self=Get WBP_UI_Button_Matchmaking(), IsCancel=true)
   n5 = Branch(execute=@n4.else, Condition=NOT Boolean(A="Equal (Enum)"(A=@n3, B=NoDownload)))
   > n6 = Branch(Condition=Get IsStatusExpanded())
   n7 = Switch on EFortUIDownloadType(
    execute=[@n6.then, @n8.then]
    Selection=Get!() -> (
     Download Info_bIsDownloading:bool=~
     Download Info_DownloadType:EFortUIDownloadType Enum
     Download Info_Progress:float=~
     Download Info_DownloadSize:Integer64
    ).Download Info_DownloadType
   )
   PlayAnimation(execute=@n6.else, InAnimation=Get Anim_StatusIntro())
   > n8 = Set IsStatusExpanded(IsStatusExpanded=true)
   "Update Matchmaking State Appearance "(execute=@n5.else, NewMatchmakingState=Get Current Matchmaking Status() -> (Current Matchmaking Status=InvalidIsland))
  ]}
 Update Island Changed {nodes [
   Update Island Changed!() -> (then, LinkId_Mnemonic:String, LinkId_Version:int) {is=title K2Node_FunctionEntry}
   > CloseMatchmakingSettings() {is=title K2Node_CallFunction}
   > n1 = Branch(Condition=IsEmpty(InString=Get!() -> (LinkId_Mnemonic:String, LinkId_Version:int)))
   Set isMatchFoundAnimationActive(execute=@n1.else, isMatchFoundAnimationActive=false)
   > On Island Changed() {is=title K2Node_CallFunction}
   > n2 = Sequence()
   n3 = Get FortMatchmakingViewModel()
   > n4 = Branch(Condition=HasLogoWidget!(self:Selected Island View Model Object Reference=@n3) -> (ReturnValue:bool))
   > LoadLogoWidget!(execute, self:Selected Island View Model Object Reference=@n3, OnWidgetLoaded:"Delegate (by ref)"=Create Event()) -> (then)
   SetIslandLogo!(execute=@n4.else, self, LogoClass:User Widget Class Reference) -> (then) {is=title K2Node_CallFunction}
   n5 = Get FortMatchmakingViewModel()
   n6 = Branch(execute=@n2.then_1, Condition=HasMatchFoundWidget!(self:Selected Island View Model Object Reference=@n5) -> (ReturnValue:bool))
   > LoadMatchFoundWidget!(execute, self:Selected Island View Model Object Reference=@n5, OnWidgetLoaded:"Delegate (by ref)"=Create Event()) -> (then)
   Set Match Found Island Widget!(execute=@n6.else, self, MatchFoundClass:User Widget Class Reference) -> (then) {is=title K2Node_CallFunction}
   Set Background Thumbnail(execute=@n2.then_2, HasThumbnail=HasThumbnail(self=Get FortMatchmakingViewModel())) {is=title K2Node_CallFunction}
   > Set PrevMapTexture()
  ]}
 Update Map Rotation Badge {nodes [
   n1 = Update Map Rotation Badge!() -> (then, In Date Text:Text) {is=title K2Node_FunctionEntry}
   n2 = Get FortMatchmakingViewModel()
   > SetBadgeText!(
    execute
    self:WBP UIKit Badge Base Object Reference=Get MapRotationCountdown()
    Text:Text=Format Text!(
     Format="LOCTEXT(\"{TimerTitle} {Timestamp}\")"
     TimerTitle:Text=Timer Title!(self:Selected Island View Model Object Reference=@n2) -> (ReturnValue:Text)
     Timestamp:Text=@n1
    ) -> (Result)
   ) -> (then)
  ]}
 "Update Matchmaking State Appearance " {nodes [
   n1 = "Update Matchmaking State Appearance "!() -> (then, NewMatchmakingState:EFortUIMatchmakingStatus Enum=Unready) {is=title K2Node_FunctionEntry}
   > n2 = Branch(Condition="Equal (Enum)"(A:EFortUIMatchmakingStatus Enum=@n1, B:EFortUIMatchmakingStatus Enum=FindingServer))
   n3 = Get FortMatchmakingViewModel()
   > BindToMatchFound!(execute, self:Selected Island View Model Object Reference=@n3, OnMatchFound:"Delegate (by ref)"=Create Event()) -> (then)
   > n4 = BindToMatchFlushed!(execute, self:Selected Island View Model Object Reference=@n3, OnMatchFlushed:"Delegate (by ref)"=Create Event()) -> (then)
   SetActiveWidget(
    execute=[@n4.then, @n6.then]
    self=Get Switcher_MainOrMatchFound()
    Widget=Select(
     Option 0:Widget Object Reference=Get Content_Main()
     Option 1:Widget Object Reference=Get SB_MatchFoundWidget()
     Index="Equal (Enum)"(A=@n1, B=MatchFound)
    ) -> (ReturnValue:Widget Object Reference)
   )
   n5 = Select!(
    InvalidIsland:ESlateVisibility Enum=Collapsed
    Unready:ESlateVisibility Enum=Collapsed
    WaitingForPlayersReadyUp:ESlateVisibility Enum=HitTestInvisible
    FindingServer:ESlateVisibility Enum=HitTestInvisible
    MatchFound:ESlateVisibility Enum=Collapsed
    Num:ESlateVisibility Enum=Collapsed
    Index:EFortUIMatchmakingStatus Enum=@n1
   ) -> (ReturnValue:ESlateVisibility Enum)
   > SetVisibility(self=Get Grid_StatusInfo(), InVisibility=@n5)
   > SetVisibility(
    self=Get Text_MMData()
    InVisibility=Select!(
     NoDownload:ESlateVisibility Enum=@n5
     DownloadWithProgress:ESlateVisibility Enum=@n5
     PreparingDownload:ESlateVisibility Enum=Visible
     QueuedDownload:ESlateVisibility Enum=Visible
     Index:EFortUIDownloadType Enum=Get!() -> (
      Download Info_bIsDownloading:bool=~
      Download Info_DownloadType:EFortUIDownloadType Enum
      Download Info_Progress:float=~
      Download Info_DownloadSize:Integer64
     ).Download Info_DownloadType
    ) -> (ReturnValue)
   )
   > Set with Broadcast isAnimating(isAnimating=false)
   SetVisibility(execute=@n2.else, self=Get WBP_UI_Button_Flush(), InVisibility=Collapsed)
   > n6 = Set CanStartMatchShouldBeVisible(CanStartMatchShouldBeVisible=false)
  ]}
 Update Tournament Timer {nodes [
   n1 = Update Tournament Timer!() -> (then, In Date Text:Text) {is=title K2Node_FunctionEntry}
   > n2 = Branch(Condition=Get bIsTournamentLive(self=Get FortMatchmakingViewModel()))
   > "SetText (Text)"(
    self=Get Text_LiveTournamentCountdown() -> (Text_LiveTournamentCountdown:Common Text Object Reference)
    InText=Format Text!(Format="LOCTEXT(\"{Timestamp} Remaining\")", Timestamp=Get In Date Text() -> (In Date Text:Text)) -> (Result)
   )
   SetBadgeText(execute=@n2.else, self=Get MapRotationCountdown(), Text=Format Text(Format="LOCTEXT(\"{Timestamp} Until Start\")", Timestamp=@n1))
  ]}
 UpdateErrorStateWidget {nodes [
   n1 = UpdateErrorStateWidget!() -> (then, ErrorUI:Ready Up Error UI Object Reference) {is=title K2Node_FunctionEntry}
   > n2 = Branch(Condition=IsValid(Object=@n1))
   n3 = Error Text!(self:Ready Up Error UI Object Reference=Get ErrorUI()) -> (ReturnValue:Text)
   > n4 = Branch(Condition=NOT Boolean(A=TextIsEmpty(InText=@n3)))
   n5 = Get WBP_UI_Lobby_ErrorState()
   > SetVisibility(self=@n5, InVisibility=HitTestInvisible)
   > Set Error!(execute, self:WBP UI Lobby Error State Object Reference=@n5, ErrorUI:Ready Up Error UI Object Reference=Get ErrorUI()) -> (then) {is=title K2Node_CallFunction}
   > SetShowBackground!(execute, self:WBP UI Lobby Error State Object Reference=Get WBP_UI_Lobby_ErrorState(), bIsVisible:bool=NOT Boolean(A=Get IsPlaytimeOut())) -> (then)
   SetVisibility(execute=[@n4.else, @n2.else], self=@n5, InVisibility=Collapsed)
  ]}
 UpdateLobbyBottomBarButtonInput {nodes [
   n1 = UpdateLobbyBottomBarButtonInput!() -> (then, BottomBarButtonState:Data Table Row Handle Structure) {is=title K2Node_FunctionEntry}
   > n2 = FindParentWidgetOfType(StartingWidget=Self-Reference(), Type=&AthenaLobbyBase)
   > n3 = Cast To AthenaLobby(Object=@n2)
   > UpdateBottomBarPlayButtonText!(execute, self:Athena Lobby Object Reference=@n3, BottomBarPlayState:Data Table Row Handle Structure=@n1) -> (then)
  ]}
 UpdateModeSelectionButton {nodes [
   UpdateModeSelectionButton!() -> (then, Info:Matchmaking Settings Button Info Structure) {is=title K2Node_FunctionEntry}
   n1 = Get!() -> (Info_TitleText:Text, Info_PrimaryText:Text, Info_SecondaryText:Text, Info_bIsEnabled:bool=~, Info_bIsVisible:bool=~)
   n2 = "To String (Text)"(InText=@n1.Info_PrimaryText)
   n3 = Select!(Option 0:String=", ", Option 1:String, Index:bool=TextIsEmpty(InText=@n1.Info_SecondaryText)) -> (ReturnValue:String)
   n4 = "To String (Text)"(InText=@n1.Info_SecondaryText)
   n5 = "To Text (String)"(InString=Append!(A=Replace(SourceString=@n2, From=" - ", To=", "), B=Append(A=@n3, B=@n4)) -> (ReturnValue))
   > SetText!(execute, self:WBP UIKit Block Text Object Reference=Get Block_Text(self=Get WBP_UI_Button_ModeSelection()), Text:Text=@n5) -> (then)
   n6 = NOT Boolean(A=TextIsEmpty(InText=@n5))
   n7 = Get() -> (Info_bIsEnabled=~, Info_bIsVisible=~)
   n8 = Get WBP_UI_Button_ModeSelection()
   > Set Visibility(
    Visibility=Select(Option 0:ESlateVisibility Enum=Collapsed, Option 1:ESlateVisibility Enum=Visible, Index=AND Boolean(A=@n7.Info_bIsVisible, B=@n6)) -> (ReturnValue:ESlateVisibility Enum)
    self=@n8
   )
   n9 = Get IsPlaytimeOut()
   n10 = Get() -> (Info_bIsEnabled=~, Info_bIsVisible=~)
   > SetIsLocked(self=@n8, bInIsLocked=AND Boolean(A=NOT Boolean(A=@n10.Info_bIsEnabled), B=NOT Boolean(A=@n9)))
  ]}
]
variables [
 AthenaLobby {type=AthenaLobby_C, default=null}
 Bottom Bar Button Input {default {
   invalid {dataTable=null, rowName=None}
   join {dataTable=&InputActionDataTable, rowName=Join}
   play {dataTable=&InputActionDataTable, rowName=PlayMatchmaking}
   ready {dataTable=&InputActionDataTable, rowName=ReadyMatchmaking}
   rejoin {dataTable=&InputActionDataTable, rowName=RejoinMatchmaking}
   watch {dataTable=&InputActionDataTable, rowName=WatchMatchmaking}
  }}
 Bottom Bar Button Predownload {type=DataTableRowHandle, default {dataTable=&InputActionDataTable, rowName=PlayMatchmaking}}
 Bottom Bar Button Redirect {type=DataTableRowHandle, default {dataTable=&InputActionDataTable, rowName=ViewDetailsMatchmaking}}
 Bottom Bar Button Start Match {type=DataTableRowHandle, default {dataTable=&InputActionDataTable, rowName=CustomStartMatch}}
 Bottom Bar Button State {type=DataTableRowHandle, default {dataTable=null, rowName=None}}
 Bottom Bar Button State Cancel {type=DataTableRowHandle, default {dataTable=&InputActionDataTable, rowName=Cancel}}
 CanStartMatchShouldBeVisible {type=boolean}
 Current Matchmaking Status {type=EFortUIMatchmakingStatus, default=InvalidIsland}
 Current Rotation Time {type=DateTime, default=0001.01.01-00.00.00}
 CurrentMapTexture {type=Texture, default=null}
 Download Info {type=MatchmakingDownloadInfo, default {bIsDownloading=false, downloadSize=0, downloadType=NoDownload, progress=-1}}
 Download Info Download Type {type=EFortUIDownloadType, default=NoDownload}
 ExtensionHolderSlotTag {type=GameplayTag, default {tagName=DynamicUI.Frontend.Fort.Athena.Matchmaking.Slot}}
 FlushAudioTimer {type=TimerHandle}
 FlushClicked {type=boolean}
 HabaneroDisplayData {type=FortHabaneroDisplayData, default=/Game/UI/Habanero/HabaneroDisplayData}
 In Live Text {type=string, default=""}
 IsAnimating {type=boolean}
 IsBottomBarVisible {type=boolean, default=true}
 IsCreatetimeOut {type=boolean}
 isMatchFoundAnimationActive {type=boolean}
 IsPlaytimeOut {type=boolean}
 IsStatusExpanded {type=boolean}
 Last Focused Widget {type=Widget, default=null}
 Match Found Class {type=Class, default=null}
 MatchFound Widget {type=UserWidget, default=null}
 MatchFoundslot {type=PanelSlot, default=null}
 MatchmakingCollectionParameter {type=MaterialParameterCollection, default=/EcosystemMatchmakingUI/MaterialFunctions/MPC_UI_Matchmaking_GlobalVisualValues}
 MatchmakingFlushedSound {type=SoundBase, default=/Game/Sounds/Fort_UI_Sounds/TrayNotifications/SC_TrayNotification_MatchFlush}
 MatchmakingMatchFoundSound {type=SoundBase, default=/Game/Sounds/Fort_UI_Sounds/TrayNotifications/SC_TrayNotification_MatchFound}
 ModeSpecificWidget {type=UserWidget, default=null}
 PlayButtonNewText {type=string, default=""}
 Previous MM Additional Data Text {type=string, default=""}
 Previous MM Data Text {type=string, default=""}
 Previous MM Status Text {type=string, default=""}
 PrevMapTexture {type=Texture, default=null}
 RankInfoOverrides {type=FortCompeteUIHabaneroTierOverrideList, default=/Game/UI/Habanero/DA_RankDisplayDataOverride}
 SettingsModal {type=CommonActivatableWidget, default=null}
 TempText {type=string, default=""}
 ThumbnailDynamicMaterial {type=MaterialInstanceDynamic}
]
widgets {
 info {inheritedWidgetCount=0, parentClass=&CommonActivatableWidget, rootWidgetClass=&VerticalBox, widgetCount=63}
 widgets [
  VerticalBox_0 {widgetClassPath=&VerticalBox, widget=VerticalBox_0}
  WBP_PluginStub_Info {
   widgetClassPath=&/WBP_PluginStub_Info
   bIsVariable=true
   parent=VerticalBox_0
   widget=WBP_PluginStub_Info
   slot {class=VerticalBoxSlot, properties {padding=<ltrb 0 0 0 16>}}
  }
  SB_Width {
   widgetClassPath=&SizeBox
   parent=VerticalBox_0
   widget=SB_Width
   properties {bOverride_WidthOverride=true, widthOverride=360}
   slot {class=VerticalBoxSlot, properties {horizontalAlignment=HAlign_Left}}
  }
  Grid_MatchmakingContainer {
   widgetClassPath=&GridPanel
   parent=SB_Width
   widget=Grid_MatchmakingContainer
   properties {columnFill [1], rowFill [1]}
   slot {class=SizeBoxSlot}
  }
  WBP_CaptureForPostBufferUpdate {
   widgetClassPath=&WBP_CaptureForPostBufferUpdate
   parent=Grid_MatchmakingContainer
   slot {class=GridSlot, properties {layer=-5}}
  }
  UIKit_Backplate {
   widgetClassPath=&WBP_UIKit_Backplate
   bIsVariable=true
   parent=Grid_MatchmakingContainer
   widget=UIKit_Backplate
   properties {brightness=Dark, cornerRadius=Tab}
   slot {class=GridSlot}
  }
  BackgroundThumbnail {
   widgetClassPath=&Image
   bIsVariable=true
   parent=Grid_MatchmakingContainer
   widget=BackgroundThumbnail
   properties {
    visibility=Collapsed
    brush {
     bIsDynamicallyLoaded=false
     drawAs=Image
     imageType=FullColor
     mirroring=NoMirror
     resourceName=None
     resourceObject=/EcosystemMatchmakingUI/MaterialInstances/MI_UI_Matchmaking_Thumbnail
     tiling=NoTile
     imageSize=<32 200>
     margin=<ltrb 0 0 0 0>
     outlineSettings {
      bUseBrushTransparency=false
      roundingType=HalfHeightRadius
      width=0
      color {colorUseRule=UseColor_Specified, specifiedColor=<rgba 0 0 0 0>}
      cornerRadii=<0 0 0 0>
     }
     tintColor {colorUseRule=UseColor_Specified, specifiedColor=<rgba 1 1 1 1>}
     uVRegion {bIsValid=false, max=<0 0>, min=<0 0>}
    }
   }
   slot {class=GridSlot, properties {layer=1, verticalAlignment=VAlign_Top}}
  }
  Grid_Content {
   widgetClassPath=&GridPanel
   parent=Grid_MatchmakingContainer
   widget=Grid_Content
   properties {columnFill [1], rowFill [0,0]}
   slot {class=GridSlot, properties {layer=5, padding=<ltrb 16 24 16 16>}}
  }
  GP_Title {
   widgetClassPath=&Overlay
   parent=Grid_Content
   widget=GP_Title
   properties {bOverride_Cursor=true}
   slot {class=GridSlot, properties {padding=<ltrb 0 0 0 8>}}
  }
  Min_Height {
   widgetClassPath=&Spacer
   parent=GP_Title
   widget=Min_Height
   properties {size=<1 128>}
   slot {class=OverlaySlot, properties {horizontalAlignment=HAlign_Fill, verticalAlignment=VAlign_Fill}}
  }
  VB_MapName {
   widgetClassPath=&VerticalBox
   parent=GP_Title
   widget=VB_MapName
   properties {clipping=ClipToBounds}
   slot {class=OverlaySlot, properties {horizontalAlignment=HAlign_Fill, verticalAlignment=VAlign_Center}}
  }
  Text_RMatchesAmount {
   widgetClassPath=&CommonTextBlock
   bIsVariable=true
   parent=VB_MapName
   widget=Text_RMatchesAmount
   properties {
    clipping=ClipToBounds
    justification=Center
    scrollStyle=&TextScrollStyle-Base_C
    text=$Match 2 of 5
    visibility=Collapsed
    font {
     bForceMonospaced=false
     bMaterialIsStencil=false
     fontMaterial=/EcosystemMatchmakingUI/MaterialInstances/MI_UI_Matchmaking_Title
     fontObject=&HeadingNow
     letterSpacing=0
     monospacedWidth=1
     size=15.75
     skewAmount=0
     typefaceFontName=75 Medium
     outlineSettings {
      bApplyOutlineToDropShadows=false
      bMiteredCorners=false
      bSeparateFillAlpha=false
      outlineMaterial=null
      outlineSize=0
      outlineColor=<rgba 0 0 0 1>
     }
    }
    shadowOffset=<0 0>
   }
   slot {class=VerticalBoxSlot, properties {verticalAlignment=VAlign_Top, padding=<ltrb 0 0 0 4>}}
  }
  Switcher_GameIslandInfo {
   widgetClassPath=&CommonVisibilitySwitcher
   bIsVariable=true
   parent=VB_MapName
   widget=Switcher_GameIslandInfo
   slot {class=VerticalBoxSlot}
  }
  GameMode_SizeBox {
   widgetClassPath=&SizeBox
   bIsVariable=true
   parent=Switcher_GameIslandInfo
   widget=GameMode_SizeBox
   properties {maxDesiredHeight=110}
   slot {class=CommonVisibilitySwitcherSlot, properties {verticalAlignment=VAlign_Top}}
  }
  GameIsland_Title {
   widgetClassPath=&CommonTextBlock
   bIsVariable=true
   parent=GameMode_SizeBox
   widget=GameIsland_Title
   properties {
    autoWrapText=true
    bIsScrollingEnabled=false
    justification=Center
    scrollStyle=&TextScrollStyle-Base_C
    text=BATTLE ROYALE ZERO BUILD
    textOverflowPolicy=MultilineEllipsis
    font {
     bForceMonospaced=false
     bMaterialIsStencil=false
     fontMaterial=&MI_UI_Matchmaking_Title
     fontObject=&HeadingNow
     letterSpacing=-40
     monospacedWidth=1
     size=21
     skewAmount=0
     typefaceFontName=86 Bold Italic
     outlineSettings {
      bApplyOutlineToDropShadows=false
      bMiteredCorners=false
      bSeparateFillAlpha=false
      outlineMaterial=null
      outlineSize=0
      outlineColor=<rgba 0 0 0 1>
     }
    }
    shadowOffset=<0 0>
   }
   slot {class=SizeBoxSlot}
  }
  NamedSlot_CustomLogoHeader {
   widgetClassPath=&NamedSlot
   bIsVariable=true
   parent=Switcher_GameIslandInfo
   widget=NamedSlot_CustomLogoHeader
   slot {class=CommonVisibilitySwitcherSlot}
  }
  Txt_MapName {
   widgetClassPath=&CommonTextBlock
   bIsVariable=true
   parent=VB_MapName
   widget=Txt_MapName
   properties {
    clipping=ClipToBounds
    justification=Center
    scrollStyle=&TextScrollStyle-Base_C
    text=$MapName
    visibility=Collapsed
    font {
     bForceMonospaced=false
     bMaterialIsStencil=false
     fontMaterial=&MI_UI_Matchmaking_Title
     fontObject=&HeadingNow
     letterSpacing=0
     monospacedWidth=1
     size=15.75
     skewAmount=0
     typefaceFontName=75 Medium
     outlineSettings {
      bApplyOutlineToDropShadows=false
      bMiteredCorners=false
      bSeparateFillAlpha=false
      outlineMaterial=null
      outlineSize=0
      outlineColor=<rgba 0 0 0 1>
     }
    }
    shadowOffset=<0 0>
   }
   slot {class=VerticalBoxSlot, properties {verticalAlignment=VAlign_Top, padding=<ltrb 0 4 0 0>}}
  }
  Switcher_MainOrMatchFound {
   widgetClassPath=&CommonVisibilitySwitcher
   bIsVariable=true
   parent=Grid_Content
   widget=Switcher_MainOrMatchFound
   slot {class=GridSlot, properties {row=1, padding=<ltrb 0 8 0 0>}}
  }
  Content_Main {
   widgetClassPath=&VerticalBox
   bIsVariable=true
   parent=Switcher_MainOrMatchFound
   widget=Content_Main
   slot {class=CommonVisibilitySwitcherSlot}
  }
  Grid_RankStats {
   widgetClassPath=&GridPanel
   bIsVariable=true
   parent=Content_Main
   widget=Grid_RankStats
   properties {visibility=Collapsed, columnFill [1]}
   slot {class=VerticalBoxSlot, properties {horizontalAlignment=HAlign_Center, padding=<ltrb 0 0 0 24>}}
  }
  Text_RPointsValue {
   widgetClassPath=&CommonTextBlock
   bIsVariable=true
   parent=Grid_RankStats
   widget=Text_RPointsValue
   properties {
    autoWrapText=true
    justification=Center
    scrollStyle=&TextScrollStyle-Base_C
    text=$000
    font {
     bForceMonospaced=false
     bMaterialIsStencil=false
     fontMaterial=null
     fontObject=&HeadingNow
     letterSpacing=-40
     monospacedWidth=1
     size=32.25
     skewAmount=0
     typefaceFontName=86 Bold Italic
     outlineSettings {
      bApplyOutlineToDropShadows=false
      bMiteredCorners=false
      bSeparateFillAlpha=false
      outlineMaterial=null
      outlineSize=0
      outlineColor=<rgba 0 0 0 1>
     }
    }
    shadowOffset=<0 0>
   }
   slot {class=GridSlot, properties {horizontalAlignment=HAlign_Center, verticalAlignment=VAlign_Top}}
  }
  Text_RPoints {
   widgetClassPath=&CommonTextBlock
   bIsVariable=true
   parent=Grid_RankStats
   widget=Text_RPoints
   properties {
    autoWrapText=true
    justification=Center
    scrollStyle=&TextScrollStyle-Base_C
    text=pts
    colorAndOpacity {colorUseRule=UseColor_Specified, specifiedColor=<rgba 0.5776 0.5776 0.5776 1>}
    font {
     bForceMonospaced=false
     bMaterialIsStencil=false
     fontMaterial=null
     fontObject=&HeadingNow
     letterSpacing=10
     monospacedWidth=1
     size=15.75
     skewAmount=0
     typefaceFontName=75 Medium
     outlineSettings {
      bApplyOutlineToDropShadows=false
      bMiteredCorners=false
      bSeparateFillAlpha=false
      outlineMaterial=null
      outlineSize=0
      outlineColor=<rgba 0 0 0 1>
     }
    }
    shadowOffset=<0 0>
   }
   slot {class=GridSlot, properties {horizontalAlignment=HAlign_Center, row=1, verticalAlignment=VAlign_Top}}
  }
  Grid_RHighestScore {
   widgetClassPath=&GridPanel
   bIsVariable=true
   parent=Grid_RankStats
   widget=Grid_RHighestScore
   properties {columnFill [0,1], rowFill [1]}
   slot {class=GridSlot, properties {horizontalAlignment=HAlign_Center, row=2, padding=<ltrb 0 24 0 0>}}
  }
  Text_RHighestScore {
   widgetClassPath=&CommonTextBlock
   bIsVariable=true
   parent=Grid_RHighestScore
   widget=Text_RHighestScore
   properties {
    mobileFontSizeMultiplier=1.14
    text="Highest Score: "
    font {
     bForceMonospaced=false
     bMaterialIsStencil=false
     fontMaterial=null
     fontObject=&HeadingNow
     letterSpacing=0
     monospacedWidth=1
     size=15.75
     skewAmount=0
     typefaceFontName=75 Medium
     outlineSettings {
      bApplyOutlineToDropShadows=false
      bMiteredCorners=false
      bSeparateFillAlpha=false
      outlineMaterial=null
      outlineSize=0
      outlineColor=<rgba 0 0 0 1>
     }
    }
    shadowOffset=<0 0>
   }
   slot {class=GridSlot}
  }
  Text_RHighestScoreValue {
   widgetClassPath=&CommonTextBlock
   bIsVariable=true
   parent=Grid_RHighestScore
   widget=Text_RHighestScoreValue
   properties {
    autoWrapText=true
    mobileFontSizeMultiplier=1.14
    scrollStyle=&TextScrollStyle-Base_C
    text=$000
    colorAndOpacity {colorUseRule=UseColor_Specified, specifiedColor=<rgba 0.5776 0.5776 0.5776 1>}
    font {
     bForceMonospaced=false
     bMaterialIsStencil=false
     fontMaterial=null
     fontObject=&HeadingNow
     letterSpacing=0
     monospacedWidth=1
     size=15.75
     skewAmount=0
     typefaceFontName=75 Medium
     outlineSettings {
      bApplyOutlineToDropShadows=false
      bMiteredCorners=false
      bSeparateFillAlpha=false
      outlineMaterial=null
      outlineSize=0
      outlineColor=<rgba 0 0 0 1>
     }
    }
    shadowOffset=<0 0>
   }
   slot {class=GridSlot, properties {column=1}}
  }
  Grid_StatusInfo {
   widgetClassPath=&GridPanel
   bIsVariable=true
   parent=Content_Main
   widget=Grid_StatusInfo
   properties {visibility=HitTestInvisible, columnFill [1,0], rowFill [1]}
   slot {class=VerticalBoxSlot, properties {padding=<ltrb 16 0 16 24>}}
  }
  VB_TextLeft {widgetClassPath=&VerticalBox, parent=Grid_StatusInfo, widget=VB_TextLeft, slot {class=GridSlot}}
  Text_MMStatus {
   widgetClassPath=&CommonTextBlock
   bIsVariable=true
   parent=VB_TextLeft
   widget=Text_MMStatus
   properties {
    autoWrapText=true
    bIsScrollingEnabled=false
    scrollStyle=&TextScrollStyle-Base_C
    text=$Status
    colorAndOpacity {colorUseRule=UseColor_Specified, specifiedColor=<rgba 1 1 1 0.7>}
    font {
     bForceMonospaced=false
     bMaterialIsStencil=false
     fontMaterial=null
     fontObject=&NotoSans
     letterSpacing=0
     monospacedWidth=1
     size=24
     skewAmount=0
     typefaceFontName=NotoSans-Regular
     outlineSettings {
      bApplyOutlineToDropShadows=false
      bMiteredCorners=false
      bSeparateFillAlpha=false
      outlineMaterial=null
      outlineSize=0
      outlineColor=<rgba 0 0 0 1>
     }
    }
    shadowOffset=<0 0>
   }
   slot {class=VerticalBoxSlot, properties {verticalAlignment=VAlign_Top}}
  }
  Text_MMData {
   widgetClassPath=&CommonTextBlock
   bIsVariable=true
   parent=VB_TextLeft
   widget=Text_MMData
   properties {
    scrollStyle=&TextScrollStyle-Base_C
    text=00:00
    font {
     bForceMonospaced=false
     bMaterialIsStencil=false
     fontMaterial=null
     fontObject=&HeadingNow
     letterSpacing=0
     monospacedWidth=1
     size=22.5
     skewAmount=0
     typefaceFontName=75 Medium
     outlineSettings {
      bApplyOutlineToDropShadows=false
      bMiteredCorners=false
      bSeparateFillAlpha=false
      outlineMaterial=null
      outlineSize=0
      outlineColor=<rgba 0 0 0 1>
     }
    }
    shadowOffset=<0 0>
   }
   slot {class=VerticalBoxSlot, properties {verticalAlignment=VAlign_Top}}
  }
  Text_MMAdditionalData {
   widgetClassPath=&CommonTextBlock
   bIsVariable=true
   parent=VB_TextLeft
   widget=Text_MMAdditionalData
   properties {
    bAutoCollapseWithEmptyText=true
    scrollStyle=&TextScrollStyle-Base_C
    text=of 6.9 MB
    colorAndOpacity {colorUseRule=UseColor_Specified, specifiedColor=<rgba 1 1 1 0.7>}
    font {
     bForceMonospaced=false
     bMaterialIsStencil=false
     fontMaterial=null
     fontObject=&HeadingNow
     letterSpacing=0
     monospacedWidth=1
     size=12
     skewAmount=0
     typefaceFontName=75 Medium
     outlineSettings {
      bApplyOutlineToDropShadows=false
      bMiteredCorners=false
      bSeparateFillAlpha=false
      outlineMaterial=null
      outlineSize=0
      outlineColor=<rgba 0 0 0 1>
     }
    }
    shadowOffset=<0 0>
   }
   slot {class=VerticalBoxSlot, properties {verticalAlignment=VAlign_Top}}
  }
  Switcher_StatusIcons {
   widgetClassPath=&CommonVisibilitySwitcher
   bIsVariable=true
   parent=Grid_StatusInfo
   widget=Switcher_StatusIcons
   properties {activeWidgetIndex=1}
   slot {class=GridSlot, properties {column=1, padding=<ltrb 16 0 0 0>}}
  }
  WBP_UIKit_StatusIndicator {
   widgetClassPath=/FortUILibrary/UIKit/StatusIndicator/WBP_UIKit_StatusIndicator
   bIsVariable=true
   parent=Switcher_StatusIcons
   widget=WBP_UIKit_StatusIndicator
   properties {icon=/FortUILibrary/Textures/Icons/T_UI_Icon_UIKit_Check, overrideStatusIndicatorVisibility=true}
   slot {class=CommonVisibilitySwitcherSlot, properties {horizontalAlignment=HAlign_Center}}
  }
  WBP_UIKit_Throbber {
   widgetClassPath=/FortUILibrary/UIKit/Throbber/WBP_UIKit_Throbber
   bIsVariable=true
   parent=Switcher_StatusIcons
   widget=WBP_UIKit_Throbber
   slot {class=CommonVisibilitySwitcherSlot, properties {horizontalAlignment=HAlign_Right, verticalAlignment=VAlign_Center}}
  }
  WBP_UI_DownloadProgress {
   widgetClassPath=/FortUILibrary/UIKit/ProgressBar/WBP_UI_DownloadProgress
   bIsVariable=true
   parent=Switcher_StatusIcons
   widget=WBP_UI_DownloadProgress
   slot {class=CommonVisibilitySwitcherSlot, properties {verticalAlignment=VAlign_Center}}
  }
  Slot_ModeSpecificButtonSlot {
   widgetClassPath=&NamedSlot
   bIsVariable=true
   parent=Content_Main
   widget=Slot_ModeSpecificButtonSlot
   slot {class=VerticalBoxSlot, properties {padding=<ltrb 0 0 0 16>}}
  }
  Overlay_0 {
   widgetClassPath=&Overlay
   parent=Content_Main
   widget=Overlay_0
   slot {class=VerticalBoxSlot, properties {padding=<ltrb 0 0 0 16>}}
  }
  WBP_UI_Button_ModeSelection {
   widgetClassPath=&/WBP_UI_Lobby_Button_ModeSelection
   bIsVariable=true
   parent=Overlay_0
   widget=WBP_UI_Button_ModeSelection
   properties {
    backgroundMaterial=/EcosystemMatchmakingUI/MaterialInstances/MI_Matchmaking_CTA_Background_Secondary
    minHeightOverride=90
    mobileMinHeight=80
    buttonPadding=<ltrb 12 12 12 12>
    font {
     bForceMonospaced=false
     bMaterialIsStencil=false
     fontMaterial=/FortUILibrary/UIKit/Buttons/MaterialInstances/CTAButtons/MI_UIKit_TextLoud
     fontObject=&HeadingNow
     letterSpacing=10
     monospacedWidth=1
     size=18
     skewAmount=0
     typefaceFontName=75 Medium
     outlineSettings {
      bApplyOutlineToDropShadows=false
      bMiteredCorners=false
      bSeparateFillAlpha=false
      outlineMaterial=null
      outlineSize=0
      outlineColor=<rgba 0 0 0 1>
     }
    }
    iconPadding=<ltrb 0 0 8 0>
    secondFont {
     bForceMonospaced=false
     bMaterialIsStencil=false
     fontMaterial=&MI_UIKit_TextLoud
     fontObject=&HeadingNow
     letterSpacing=10
     monospacedWidth=1
     size=15.75
     skewAmount=0
     typefaceFontName=75 Medium
     outlineSettings {
      bApplyOutlineToDropShadows=false
      bMiteredCorners=false
      bSeparateFillAlpha=false
      outlineMaterial=null
      outlineSize=0
      outlineColor=<rgba 0 0 0 1>
     }
    }
    textPadding=<ltrb 0 2 8 0>
   }
   slot {class=OverlaySlot, properties {horizontalAlignment=HAlign_Fill, verticalAlignment=VAlign_Fill}}
  }
  CommonVisualAttachment_97 {
   widgetClassPath=&CommonVisualAttachment
   parent=Overlay_0
   widget=CommonVisualAttachment_97
   properties {contentAnchor=<-0.05 0.8>}
   slot {class=OverlaySlot, properties {horizontalAlignment=HAlign_Right, verticalAlignment=VAlign_Center}}
  }
  WBP_PilgrimCoachmark_CuratedSetlist {
   widgetClassPath=/FortniteCoreUI/UI/Coachmarks/WBP_PilgrimCoachmark_CuratedSetlist
   bIsVariable=true
   parent=CommonVisualAttachment_97
   widget=WBP_PilgrimCoachmark_CuratedSetlist
   slot {class=SizeBoxSlot}
  }
  VerticalBox_ExtensionHolder {
   widgetClassPath=&VerticalBox
   bIsVariable=true
   parent=Content_Main
   widget=VerticalBox_ExtensionHolder
   slot {class=VerticalBoxSlot}
  }
  WBP_UI_Button_Flush {
   widgetClassPath=&/WBP_UI_Lobby_Button_Matchmaking
   bIsVariable=true
   parent=Content_Main
   widget=WBP_UI_Button_Flush
   properties {
    backgroundMaterial=&MI_Matchmaking_CTA_Background_Secondary
    minHeight=0
    minHeightOverride=56
    mobileMinHeight=80
    text=Start Match
    visibility=Collapsed
    font {
     bForceMonospaced=false
     bMaterialIsStencil=false
     fontMaterial=&MI_UIKit_TextLoud
     fontObject=&HeadingNow
     letterSpacing=10
     monospacedWidth=1
     size=18
     skewAmount=0
     typefaceFontName=76 Bold
     outlineSettings {
      bApplyOutlineToDropShadows=false
      bMiteredCorners=false
      bSeparateFillAlpha=false
      outlineMaterial=null
      outlineSize=0
      outlineColor=<rgba 0 0 0 1>
     }
    }
   }
   slot {class=VerticalBoxSlot, properties {padding=<ltrb 0 0 0 16>}}
  }
  CommonVisualAttachment_1 {
   widgetClassPath=&CommonVisualAttachment
   parent=Content_Main
   widget=CommonVisualAttachment_1
   properties {contentAnchor=<-0.05 0.55>, renderTransformPivot=<0.5 0.8>}
   slot {class=VerticalBoxSlot, properties {horizontalAlignment=HAlign_Right, verticalAlignment=VAlign_Center}}
  }
  WBP_Lobby_Coachmark_AFKSittingOut {
   widgetClassPath=/FortniteCoreUI/UI/Coachmarks/WBP_Lobby_Coachmark_AFKSittingOut
   bIsVariable=true
   parent=CommonVisualAttachment_1
   widget=WBP_Lobby_Coachmark_AFKSittingOut
   slot {class=SizeBoxSlot}
  }
  Switcher_PrimaryButton {
   widgetClassPath=&CommonVisibilitySwitcher
   bIsVariable=true
   parent=Content_Main
   widget=Switcher_PrimaryButton
   slot {class=VerticalBoxSlot}
  }
  WBP_UI_Button_Matchmaking {
   widgetClassPath=&WBP_UI_Lobby_Button_Matchmaking
   bIsVariable=true
   parent=Switcher_PrimaryButton
   widget=WBP_UI_Button_Matchmaking
   properties {minHeight=0, minHeightOverride=80, mobileMinHeight=80, useSecondTextBlock=true}
   slot {class=CommonVisibilitySwitcherSlot}
  }
  WBP_UI_Button_Primary {
   widgetClassPath=&WBP_UIKit_Button_Regular
   bIsVariable=true
   parent=Switcher_PrimaryButton
   widget=WBP_UI_Button_Primary
   properties {
    minHeight=56
    mobileFontSizeMultiplierOverride=1.25
    overrideDefaultMobileFontSizeMultiplier=true
    overrideSecondaryTextSize=true
    overrideSecondaryTextSizeMultiplier=0.7
    text=$Primary
   }
   slot {class=CommonVisibilitySwitcherSlot}
  }
  WBP_UI_Button_Secondary {
   widgetClassPath=/FortUILibrary/UIKit/Buttons/WBP_UIKit_Button_Quiet
   bIsVariable=true
   parent=Content_Main
   widget=WBP_UI_Button_Secondary
   properties {text=$Secondary, useHideBackground=false, visibility=Collapsed}
   slot {class=VerticalBoxSlot, properties {padding=<ltrb 0 16 0 0>}}
  }
  SB_MatchFoundWidget {
   widgetClassPath=&SizeBox
   bIsVariable=true
   parent=Switcher_MainOrMatchFound
   widget=SB_MatchFoundWidget
   properties {bOverride_MaxDesiredHeight=true, maxDesiredHeight=182, minDesiredHeight=182}
   slot {class=CommonVisibilitySwitcherSlot, properties {horizontalAlignment=HAlign_Center, verticalAlignment=VAlign_Center}}
  }
  NamedSlot_MatchFoundWidget {
   widgetClassPath=&NamedSlot
   bIsVariable=true
   parent=SB_MatchFoundWidget
   widget=NamedSlot_MatchFoundWidget
   properties {visibility=HitTestInvisible}
   slot {class=SizeBoxSlot, properties {horizontalAlignment=HAlign_Center, verticalAlignment=VAlign_Top}}
  }
  VA_MapRotationTimer {
   widgetClassPath=&FortVisualAttachment
   parent=Grid_MatchmakingContainer
   widget=VA_MapRotationTimer
   properties {bOverride_MinDesiredWidth=true, minDesiredWidth=212, contentAnchor=<0.5 0.5>}
   slot {class=GridSlot, properties {horizontalAlignment=HAlign_Center, layer=20, verticalAlignment=VAlign_Top}}
  }
  Switcher_Countdown {
   widgetClassPath=&CommonVisibilitySwitcher
   bIsVariable=true
   parent=VA_MapRotationTimer
   widget=Switcher_Countdown
   slot {class=SizeBoxSlot, properties {padding=<ltrb 0 2 0 0>}}
  }
  VerticalBox_1 {widgetClassPath=&VerticalBox, parent=Switcher_Countdown, widget=VerticalBox_1, slot {class=CommonVisibilitySwitcherSlot}}
  WidgetSlottingHostOverlay_Countdown {
   widgetClassPath=/Script/WidgetSlotting.WidgetSlottingHostOverlay
   parent=VerticalBox_1
   widget=WidgetSlottingHostOverlay_Countdown
   properties {hostTag {tagName=DynamicUI.Frontend.Fort.Athena.Matchmaking.PowerHourTimer.Slot}}
   slot {class=VerticalBoxSlot}
  }
  MapRotationCountdown {
   widgetClassPath=/FortUILibrary/UIKit/Badge/WBP_UIKit_Badge
   bIsVariable=true
   parent=VerticalBox_1
   widget=MapRotationCountdown
   properties {
    badgeText=$MapRotationTimer
    badgeType=Text
    indicator Type=Utility
    indicatorConfig=Icon
    indicatorIcon=/FortUILibrary/Textures/MaterialInstances/MI_UIKit_Icon_UpdatePending
    pixelSnapping=Disabled
    visibility=Collapsed
   }
   slot {class=VerticalBoxSlot}
  }
  Grid_LiveRankedCountdown {
   widgetClassPath=&GridPanel
   bIsVariable=true
   parent=Switcher_Countdown
   widget=Grid_LiveRankedCountdown
   properties {visibility=Collapsed, columnFill [1], rowFill [1]}
   slot {class=CommonVisibilitySwitcherSlot}
  }
  Image_LiveHighlightBackground {
   widgetClassPath=&Image
   parent=Grid_LiveRankedCountdown
   widget=Image_LiveHighlightBackground
   properties {brush {
     bIsDynamicallyLoaded=false
     drawAs=Image
     imageType=FullColor
     mirroring=NoMirror
     resourceName=None
     resourceObject=/EcosystemMatchmakingUI/MaterialInstances/MI_UI_MatchmakingHighlightBackground
     tiling=NoTile
     imageSize=<0 24>
     margin=<ltrb 0 0 0 0>
     outlineSettings {
      bUseBrushTransparency=false
      roundingType=HalfHeightRadius
      width=0
      color {colorUseRule=UseColor_Specified, specifiedColor=<rgba 0 0 0 0>}
      cornerRadii=<0 0 0 0>
     }
     tintColor {colorUseRule=UseColor_Specified, specifiedColor=<rgba 1 1 1 1>}
     uVRegion {bIsValid=false, max=<0 0>, min=<0 0>}
    }}
   slot {class=GridSlot}
  }
  Grid_LiveRankedCountdownContent {
   widgetClassPath=&GridPanel
   parent=Grid_LiveRankedCountdown
   widget=Grid_LiveRankedCountdownContent
   slot {class=GridSlot, properties {horizontalAlignment=HAlign_Center, verticalAlignment=VAlign_Center, padding=<ltrb 12 0 12 0>}}
  }
  Text_HighlightLive {
   widgetClassPath=&CommonTextBlock
   parent=Grid_LiveRankedCountdownContent
   widget=Text_HighlightLive
   properties {
    applyLineHeightToBottomLine=false
    bAutoCollapseWithEmptyText=true
    justification=Center
    mobileFontSizeMultiplier=1.14
    scrollStyle=&TextScrollStyle-Base_C
    text="Live "
    textTransformPolicy=ToUpper
    font {
     bForceMonospaced=false
     bMaterialIsStencil=false
     fontMaterial=null
     fontObject=&HeadingNow
     letterSpacing=0
     monospacedWidth=1
     size=15.75
     skewAmount=0
     typefaceFontName=86 Bold Italic
     outlineSettings {
      bApplyOutlineToDropShadows=false
      bMiteredCorners=false
      bSeparateFillAlpha=false
      outlineMaterial=null
      outlineSize=0
      outlineColor=<rgba 0 0 0 1>
     }
    }
    shadowOffset=<0 0>
   }
   slot {class=GridSlot, properties {horizontalAlignment=HAlign_Left, verticalAlignment=VAlign_Center}}
  }
  Text_LiveTournamentCountdown {
   widgetClassPath=&CommonTextBlock
   bIsVariable=true
   parent=Grid_LiveRankedCountdownContent
   widget=Text_LiveTournamentCountdown
   properties {
    applyLineHeightToBottomLine=false
    bAutoCollapseWithEmptyText=true
    justification=Center
    mobileFontSizeMultiplier=1.41
    scrollStyle=&TextScrollStyle-Base_C
    text=$MapRotationTimer
    font {
     bForceMonospaced=false
     bMaterialIsStencil=false
     fontMaterial=null
     fontObject=&HeadingNow
     letterSpacing=0
     monospacedWidth=1
     size=12.75
     skewAmount=0
     typefaceFontName=74 Regular
     outlineSettings {
      bApplyOutlineToDropShadows=false
      bMiteredCorners=false
      bSeparateFillAlpha=false
      outlineMaterial=null
      outlineSize=0
      outlineColor=<rgba 0 0 0 1>
     }
    }
    shadowOffset=<0 0>
   }
   slot {class=GridSlot, properties {column=1, horizontalAlignment=HAlign_Left, verticalAlignment=VAlign_Center, padding=<ltrb 4 0 0 0>}}
  }
  DecoyButton_TournamentDetails {
   widgetClassPath=/Game/UI/Frontend/BottomBar/WBP_BottomBarDecoyButton
   bIsVariable=true
   parent=Grid_MatchmakingContainer
   widget=DecoyButton_TournamentDetails
   properties {
    bDisplayInputActionWhenNotInteractable=false
    holdData=&CommonUI_HoldData_Base_C
    visibility=Collapsed
    triggeringInputAction {dataTable=/Game/UI/Frontend/Lobby/Habanero/HabaneroData/DT_HabaneroUIInputActions, rowName=TournamentDetails}
   }
   slot {class=GridSlot, properties {horizontalAlignment=HAlign_Left, verticalAlignment=VAlign_Top}}
  }
  WBP_UI_Lobby_ErrorState {
   widgetClassPath=&/WBP_UI_Lobby_ErrorState
   bIsVariable=true
   parent=VerticalBox_0
   widget=WBP_UI_Lobby_ErrorState
   slot {class=VerticalBoxSlot, properties {horizontalAlignment=HAlign_Left, padding=<ltrb 0 16 0 0>}}
  }
  Text_DebugLBId {
   widgetClassPath=&TextBlock
   bIsVariable=true
   parent=VerticalBox_0
   widget=Text_DebugLBId
   properties {
    autoWrapText=true
    justification=Center
    text="[DEBUG] LB Key:<None>"
    visibility=Collapsed
    wrapTextAt=500
    wrappingPolicy=AllowPerCharacterWrapping
    font {
     bForceMonospaced=false
     bMaterialIsStencil=false
     fontMaterial=null
     fontObject=&Roboto
     letterSpacing=0
     monospacedWidth=1
     size=12
     skewAmount=0
     typefaceFontName=Bold
     outlineSettings {
      bApplyOutlineToDropShadows=false
      bMiteredCorners=false
      bSeparateFillAlpha=false
      outlineMaterial=null
      outlineSize=3
      outlineColor=<rgba 0 0 0 1>
     }
    }
   }
   slot {class=VerticalBoxSlot, properties {horizontalAlignment=HAlign_Left, verticalAlignment=VAlign_Center}}
  }
  Text_DebugId {
   widgetClassPath=&TextBlock
   bIsVariable=true
   parent=VerticalBox_0
   widget=Text_DebugId
   properties {
    autoWrapText=true
    text=EventId: Event EventWindowId: EventWindow
    visibility=Collapsed
    wrapTextAt=500
    wrappingPolicy=AllowPerCharacterWrapping
    font {
     bForceMonospaced=false
     bMaterialIsStencil=false
     fontMaterial=null
     fontObject=&Roboto
     letterSpacing=0
     monospacedWidth=1
     size=12
     skewAmount=0
     typefaceFontName=Bold
     outlineSettings {
      bApplyOutlineToDropShadows=false
      bMiteredCorners=false
      bSeparateFillAlpha=false
      outlineMaterial=null
      outlineSize=3
      outlineColor=<rgba 0 0 0 1>
     }
    }
   }
   slot {class=VerticalBoxSlot, properties {horizontalAlignment=HAlign_Left, verticalAlignment=VAlign_Top}}
  }
 ]
}
